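# Player AnimBlueprint for the AnnoyedMop item, overriding Fortnite item animation layers, chiefly upper body and final pose.
# The upper-body layer runs an equip state machine that plays the equip-end animation.
# The final-pose layer selects by bIsCombo04 between an aim offset and a jump-AOE/in-air-attack state machine, then inertializes.
# Landing switches on bIsFalling between a per-bone fall blend and landing clip; AOE states use TEMP clips.
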AnimBlueprint /AnnoyedMop/Assets/Animations/AnnoyedMop_Player_AnimBP
target_skeleton=&Fortnite_M_Avg_Player_Skeleton
graphs [
 AnimGraph {nodes [Output Pose() {layerGroup=None}]}
 EventGraph {links [], nodes []}
 ItemAimOffset {nodes [
   Input Pose!() -> (Pose, AimOffsetAlpha:double, Yaw:double, Pitch:double) {name=InPose_AimOffset}
   Input Pose!() -> (Pose) {name=InPose_PreAimOffsets}
   Output Pose() {layerGroup=Item Override}
  ]}
 ItemFallAdditive {nodes [
   Input Pose() {name=InPoseItemFallAdditive}
   Output Pose() {layerGroup=Item Override}
  ]}
 ItemFallLandAdditive {nodes [
   Input Pose() {name=InPoseItemFallLandAdditive}
   Output Pose() {layerGroup=Item Override}
  ]}
 ItemFallLevel2Override {nodes [
   Input Pose() {name=InPose_FallLevel2}
   Output Pose() {layerGroup=Item Override}
  ]}
 ItemFinalPoseOverride {nodes [
   n1 = AimOffset Player 'ColdDay_AnnoyedMop_AO'!(
    X:float=Get!() -> (AdjustedAim_ResultingYaw:float=~, AdjustedAim_ResultingPitch:float=~, AdjustedAim_ZeroOutPitchWeightCurveName:Name).AdjustedAim_ResultingPitch
    BasePose:Pose Link Structure=Use cached pose 'Final Pose Cached'()
    Alpha:float=Get Combo04AOAlpha() -> (Combo04AOAlpha:float)
   ) -> (Pose:Pose Link Structure) {blendSpace=/AnnoyedMop/Assets/Animations/AO/ColdDay_AnnoyedMop_AO, alphaBoolBlend {blendOption=Linear}}
   AnimGraphNode_StateMachine_1 = FP AOE+InAirAttack SM() {is=title AnimGraphNode_StateMachine}
   n2 = Get bIsCombo04() -> (bIsCombo04:bool)
   Output Pose(Result=Inertialization!(Source:Pose Link Structure=Blend Poses by bool(BlendPose_0=@n1, BlendPose_1=@AnimGraphNode_StateMachine_1, bActiveValue=@n2) {blendType=Linear}) -> (Pose:Pose Link Structure) {is=title AnimGraphNode_Inertialization}) {layerGroup=Item Override}
   Save cached pose 'Final Pose Cached'(Pose=Input Pose() {name=InFinalPose}) {cacheName=Final Pose Cached}
  ]}
 FP AOE+InAirAttack SM {
  entry=FP Wait
  path=ItemFinalPoseOverride.AnimGraphNode_StateMachine_1.FP AOE+InAirAttack SM
  ids=id
  nodes [
   FP Jump AOE Exit() {is=title AnimStateAliasNode, also {aliases [FP Jump AOE Intro, FP Jump AOE Loop]}}
   FP To Jump AOE Intro() {is=title AnimStateAliasNode, also {aliases [FP InAirAttack, FP InAirAttack Landing, FP Wait]}}
   InAirAttack Exit() {is=title AnimStateAliasNode, also {aliases [FP InAirAttack Landing]}}
   AnimStateNode_0 = FP Wait() {
    is=title AnimStateNode
    stateEntered {guid=#1}
    stateFullyBlended {guid=#2}
    stateLeft {guid=#3}
   }
   AnimStateNode_1 = FP Jump AOE Intro() {is=title AnimStateNode}
   AnimStateNode_2 = FP Jump AOE Loop() {is=title AnimStateNode}
   AnimStateNode_5 = FP InAirAttack Landing() {
    is=title AnimStateNode
    stateEntered {guid=#4}
    stateFullyBlended {guid=#5}
    stateLeft {guid=#6}
   }
   AnimStateNode_6 = FP InAirAttack() {
    is=title AnimStateNode
    stateEntered {guid=#7}
    stateFullyBlended {guid=#8}
    stateLeft {guid=#9}
   }
   AnimStateTransitionNode_1 = AnimStateTransitionNode() {bAutomaticRuleBasedOnSequencePlayerInState=true, crossfadeDuration=0.1}
   AnimStateTransitionNode_11 = AnimStateTransitionNode() {
    crossfadeDuration=0.1
    transitionEnd {guid=#10}
    transitionInterrupt {guid=#11}
    transitionStart {guid=#12}
   }
   AnimStateTransitionNode_2 = AnimStateTransitionNode() {crossfadeDuration=0}
   AnimStateTransitionNode_3 = AnimStateTransitionNode() {
    crossfadeDuration=0
    priorityOrder=10
    transitionEnd {guid=#13}
    transitionInterrupt {guid=#14}
    transitionStart {guid=#15}
   }
   AnimStateTransitionNode_4 = AnimStateTransitionNode() {priorityOrder=10}
   AnimStateTransitionNode_5 = AnimStateTransitionNode() {
    bAutomaticRuleBasedOnSequencePlayerInState=true
    crossfadeDuration=0.4
    transitionEnd {guid=#16}
    transitionInterrupt {guid=#17}
    transitionStart {guid=#18}
   }
   AnimStateTransitionNode_6 = AnimStateTransitionNode() {crossfadeDuration=0, priorityOrder=10}
   AnimStateTransitionNode_9 = AnimStateTransitionNode() {
    crossfadeDuration=0.25
    priorityOrder=10
    transitionEnd {guid=#19}
    transitionInterrupt {guid=#20}
    transitionStart {guid=#21}
   }
  ]
  links [
   {
    dst=FP InAirAttack Landing
    rule=ItemFinalPoseOverride.AnimGraphNode_StateMachine_1.FP AOE+InAirAttack SM.AnimStateTransitionNode_11.AnimationTransitionGraph_0
    src=FP InAirAttack
    via=AnimStateTransitionNode_11
   }
   {
    dst=FP Wait
    rule=ItemFinalPoseOverride.AnimGraphNode_StateMachine_1.FP AOE+InAirAttack SM.AnimStateTransitionNode_4.Transition
    src=FP InAirAttack
    via=AnimStateTransitionNode_4
   }
   {
    dst=FP Wait
    rule=ItemFinalPoseOverride.AnimGraphNode_StateMachine_1.FP AOE+InAirAttack SM.AnimStateTransitionNode_5.AnimationTransitionGraph_0
    src=FP InAirAttack Landing
    via=AnimStateTransitionNode_5
   }
   {
    dst=FP Wait
    rule=ItemFinalPoseOverride.AnimGraphNode_StateMachine_1.FP AOE+InAirAttack SM.AnimStateTransitionNode_6.AnimationTransitionGraph_1
    src=FP Jump AOE Exit
    via=AnimStateTransitionNode_6
   }
   {
    dst=FP Jump AOE Loop
    rule=ItemFinalPoseOverride.AnimGraphNode_StateMachine_1.FP AOE+InAirAttack SM.AnimStateTransitionNode_1.Transition
    src=FP Jump AOE Intro
    via=AnimStateTransitionNode_1
   }
   {
    dst=FP Jump AOE Intro
    rule=ItemFinalPoseOverride.AnimGraphNode_StateMachine_1.FP AOE+InAirAttack SM.AnimStateTransitionNode_2.Transition
    src=FP To Jump AOE Intro
    via=AnimStateTransitionNode_2
   }
   {
    dst=FP InAirAttack
    rule=ItemFinalPoseOverride.AnimGraphNode_StateMachine_1.FP AOE+InAirAttack SM.AnimStateTransitionNode_3.AnimationTransitionGraph_0
    src=FP Wait
    via=AnimStateTransitionNode_3
   }
   {
    dst=FP Wait
    rule=ItemFinalPoseOverride.AnimGraphNode_StateMachine_1.FP AOE+InAirAttack SM.AnimStateTransitionNode_9.AnimationTransitionGraph_0
    src=InAirAttack Exit
    via=AnimStateTransitionNode_9
   }
  ]
 }
 FP Wait {
  path=ItemFinalPoseOverride.AnimGraphNode_StateMachine_1.FP AOE+InAirAttack SM.AnimStateNode_0.FP Wait
  nodes [Output Animation Pose(Result=Use cached pose 'Final Pose Cached'())]
 }
 FP Jump AOE Intro {
  path=ItemFinalPoseOverride.AnimGraphNode_StateMachine_1.FP AOE+InAirAttack SM.AnimStateNode_1.FP Jump AOE Intro
  nodes [Output Animation Pose(Result=Sequence Player 'Fire_Bending_TEMP_GroundAOE_Intro'() {bLoopAnimation=false, sequence=/AnnoyedMop/Assets/Animations/TEMP/Fire_Bending_TEMP_GroundAOE_Intro})]
 }
 FP Jump AOE Loop {
  path=ItemFinalPoseOverride.AnimGraphNode_StateMachine_1.FP AOE+InAirAttack SM.AnimStateNode_2.FP Jump AOE Loop
  nodes [Output Animation Pose(Result=Sequence Player 'Fire_Bending_TEMP_GroundAOE_Loop'() {sequence=/AnnoyedMop/Assets/Animations/TEMP/Fire_Bending_TEMP_GroundAOE_Loop})]
 }
 FP InAirAttack Landing {
  path=ItemFinalPoseOverride.AnimGraphNode_StateMachine_1.FP AOE+InAirAttack SM.AnimStateNode_5.FP InAirAttack Landing
  nodes [
   n1 = Sequence Player 'AnnoyedMop_InAirAttack_FallTransition'() {bLoopAnimation=false, groupName=InAirAttack, method=SyncGroup, sequence=&/AnnoyedMop_InAirAttack_FallTransition}
   n2 = Layered blend per bone(BasePose=Use cached pose 'Final Pose Cached'(), BlendPoses_0=@n1) {layerSetup [{branchFilters [{boneName=pelvis}, {blendDepth=-1, boneName=thigh_r}, {blendDepth=-1, boneName=thigh_l}]}]}
   n3 = Sequence Player 'AnnoyedMop_InAirAttack_Landing'() {bLoopAnimation=false, groupName=InAirAttack, method=SyncGroup, sequence=&/AnnoyedMop_InAirAttack_Landing}
   n4 = Get bIsFalling()
   Output Animation Pose(Result=Two Way Blend(
     A=Blend Poses by bool(BlendPose_0=@n2, BlendPose_1=@n3, bActiveValue=@n4) {blendType=Linear}
     B=Use cached pose 'Final Pose Cached'()
     Alpha=0.01
    ) {blendNode {alphaBoolBlend {blendOption=Linear}}})
  ]
 }
 FP InAirAttack {
  path=ItemFinalPoseOverride.AnimGraphNode_StateMachine_1.FP AOE+InAirAttack SM.AnimStateNode_6.FP InAirAttack
  nodes [Output Animation Pose(Result=Use cached pose 'Final Pose Cached'())]
 }
 Transition {path=ItemFinalPoseOverride.AnimGraphNode_StateMachine_1.FP AOE+InAirAttack SM.AnimStateTransitionNode_1.Transition, nodes [Result()]}
 AnimationTransitionGraph_0 {
  path=ItemFinalPoseOverride.AnimGraphNode_StateMachine_1.FP AOE+InAirAttack SM.AnimStateTransitionNode_11.AnimationTransitionGraph_0
  nodes [Result(bCanEnterTransition=Get bIsInAirAttackLanding() -> (bIsInAirAttackLanding:bool))]
 }
 Transition {
  path=ItemFinalPoseOverride.AnimGraphNode_StateMachine_1.FP AOE+InAirAttack SM.AnimStateTransitionNode_2.Transition
  nodes [Result(bCanEnterTransition=Get bIsJumpAOEStart() -> (bIsJumpAOEStart:bool))]
 }
 AnimationTransitionGraph_0 {
  path=ItemFinalPoseOverride.AnimGraphNode_StateMachine_1.FP AOE+InAirAttack SM.AnimStateTransitionNode_3.AnimationTransitionGraph_0
  nodes [Result(bCanEnterTransition=Get bIsInAirAttack() -> (bIsInAirAttack:bool))]
 }
 Transition {
  path=ItemFinalPoseOverride.AnimGraphNode_StateMachine_1.FP AOE+InAirAttack SM.AnimStateTransitionNode_4.Transition
  nodes [Result(bCanEnterTransition=NOT Boolean(A=Get bIsInAirAttack()))]
 }
 AnimationTransitionGraph_0 {
  path=ItemFinalPoseOverride.AnimGraphNode_StateMachine_1.FP AOE+InAirAttack SM.AnimStateTransitionNode_5.AnimationTransitionGraph_0
  nodes [Result()]
 }
 AnimationTransitionGraph_1 {
  path=ItemFinalPoseOverride.AnimGraphNode_StateMachine_1.FP AOE+InAirAttack SM.AnimStateTransitionNode_6.AnimationTransitionGraph_1
  nodes [Result(bCanEnterTransition=Get bIsJumpAOELand() -> (bIsJumpAOELand:bool))]
 }
 AnimationTransitionGraph_0 {
  path=ItemFinalPoseOverride.AnimGraphNode_StateMachine_1.FP AOE+InAirAttack SM.AnimStateTransitionNode_9.AnimationTransitionGraph_0
  nodes [Result(bCanEnterTransition=NOT Boolean(A=Get bIsInAirAttackLanding()))]
 }
 ItemFullBodyAimOffsets {nodes [
   Input Pose!() -> (Pose, FullBody_AimOffsetAlpha:double, Aim_Yaw:double, Aim_Pitch:double) {name=InPoseFullBodyAimOffsets}
   Input Pose!() -> (Pose) {name=InPosePreFullBodyAimOffsets}
   Output Pose() {layerGroup=Item Override}
  ]}
 ItemFullBodyOverride {nodes [
   Input Pose() {name=InPoseFullBody}
   Output Pose() {layerGroup=Item Override}
  ]}
 ItemIdleOverride {nodes [
   Input Pose() {name=InPose_Idle}
   Output Pose() {layerGroup=Item Override}
  ]}
 ItemJetPackJumpAdditive {nodes [Output Pose() {layerGroup=Item Override}]}
 ItemJetPackStartAdditive {nodes [Output Pose() {layerGroup=Item Override}]}
 ItemJumpLoopAdditive {nodes [
   Input Pose() {name=InPoseItemJumpLoopAdditive}
   Output Pose() {layerGroup=Item Override}
  ]}
 ItemJumpOverride {nodes [
   Input Pose() {name=InPoseJump}
   Output Pose() {layerGroup=Item Override}
  ]}
 ItemJumpUpAdditive {nodes [
   Input Pose() {name=InPoseItemJumpUpAdditive}
   Output Pose() {layerGroup=Item Override}
  ]}
 ItemPreAimOffsets {nodes [
   Input Pose() {name=InPosePreAimOffsets}
   Output Pose() {layerGroup=Item Override}
  ]}
 ItemPreIK {nodes [
   Input Pose() {name=InPose_PreIK}
   Output Pose() {layerGroup=Item Override}
  ]}
 ItemRightHandIK {nodes [
   Input Pose() {name=InPosePreRightHandIK}
   Input Pose() {name=InPosePostRightHandIK}
   Output Pose() {layerGroup=Item Override}
  ]}
 ItemSkeletalControl {nodes [
   Input Pose()
   Output Pose() {layerGroup=Item Override}
  ]}
 ItemSwimJumpFallLoop {nodes [Output Pose() {layerGroup=Item Override}]}
 ItemSwimJumpSurfaceEnd {nodes [Output Pose() {layerGroup=Item Override}]}
 ItemSwimJumpSurfaceLoop {nodes [Output Pose() {layerGroup=Item Override}]}
 ItemUpperBody {nodes [
   AnimGraphNode_StateMachine_0 = UB Equip SM() {is=title AnimGraphNode_StateMachine}
   Output Pose(Result=@AnimGraphNode_StateMachine_0) {layerGroup=Item Override}
   Save cached pose 'UB Pose Cached'(Pose=Input Pose!() -> (Pose, InputParam:Fort Anim Input Adjusted Aim Structure) {name=InPoseUpperBody}) {cacheName=UB Pose Cached}
  ]}
 UB Equip SM {
  entry=UB Wait
  path=ItemUpperBody.AnimGraphNode_StateMachine_0.UB Equip SM
  ids=id
  nodes [
   AnimStateNode_0 = UB Wait() {is=title AnimStateNode}
   AnimStateNode_1 = UB Equip() {is=title AnimStateNode}
   AnimStateTransitionNode_0 = AnimStateTransitionNode() {blendMode=Linear, crossfadeDuration=0.3}
   AnimStateTransitionNode_1 = AnimStateTransitionNode()
   AnimStateTransitionNode_2 = AnimStateTransitionNode() {bAutomaticRuleBasedOnSequencePlayerInState=true, crossfadeDuration=0.4}
  ]
  links [
   {
    dst=UB Wait
    rule=ItemUpperBody.AnimGraphNode_StateMachine_0.UB Equip SM.AnimStateTransitionNode_1.Transition
    src=UB Equip
    via=AnimStateTransitionNode_1
   }
   {
    dst=UB Wait
    rule=ItemUpperBody.AnimGraphNode_StateMachine_0.UB Equip SM.AnimStateTransitionNode_2.Transition
    src=UB Equip
    via=AnimStateTransitionNode_2
   }
   {
    dst=UB Equip
    rule=ItemUpperBody.AnimGraphNode_StateMachine_0.UB Equip SM.AnimStateTransitionNode_0.Transition
    src=UB Wait
    via=AnimStateTransitionNode_0
   }
  ]
 }
 UB Wait {
  path=ItemUpperBody.AnimGraphNode_StateMachine_0.UB Equip SM.AnimStateNode_0.UB Wait
  nodes [Output Animation Pose(Result=Use cached pose 'UB Pose Cached'())]
 }
 UB Equip {
  path=ItemUpperBody.AnimGraphNode_StateMachine_0.UB Equip SM.AnimStateNode_1.UB Equip
  nodes [Output Animation Pose(Result=Sequence Player 'AnnoyedMop_Equip_End'() {bLoopAnimation=false, sequence=&/AnnoyedMop_Equip_End})]
 }
 Transition {
  path=ItemUpperBody.AnimGraphNode_StateMachine_0.UB Equip SM.AnimStateTransitionNode_0.Transition
  nodes [Result(bCanEnterTransition=Get bIsEquipMontagePlaying() -> (bIsEquipMontagePlaying:bool))]
 }
 Transition {
  path=ItemUpperBody.AnimGraphNode_StateMachine_0.UB Equip SM.AnimStateTransitionNode_1.Transition
  nodes [Result(bCanEnterTransition=Get bIsActionInput() -> (bIsActionInput:bool))]
 }
 Transition {path=ItemUpperBody.AnimGraphNode_StateMachine_0.UB Equip SM.AnimStateTransitionNode_2.Transition, nodes [Result()]}
 ItemZipLineStartAdditive {nodes [Output Pose() {layerGroup=Item Override}]}
]
variables []
