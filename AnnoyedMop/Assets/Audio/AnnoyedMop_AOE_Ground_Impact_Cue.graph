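# Sound cue for the ground impact of the AnnoyedMop AOE slam attack, on a melee weapon sound class.
# A mixer blends a dominant near layer with a much quieter distant layer, each with its own attenuation asset.
# Each layer randomly picks from equally weighted land waves and passes the result through a modulator for variation.
# A concurrency override caps simultaneous instances.
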SoundCue /AnnoyedMop/Assets/Audio/AnnoyedMop_AOE_Ground_Impact_Cue
root=Mixer_0
nodes [
 {id=Attenuation_0, children [Modulator_3], properties {attenuationSettings=/AnnoyedMop/Assets/Audio/Settings/AnnoyedMop_AOE_Attack_Slam_Att}}
 {
  id=Attenuation_1
  children [Modulator_0]
  properties {attenuationSettings=/AnnoyedMop/Assets/Audio/Settings/AnnoyedMop_AOE_Attack_Slam_Distant_Att}
 }
 {id=Mixer_0, children [Attenuation_0, Attenuation_1], properties {inputVolume [1.2,0.15]}}
 {id=Modulator_0, children [Random_1]}
 {id=Modulator_3, children [Random_0]}
 {id=Random_0, children [WavePlayer_2,WavePlayer_3], properties {weights [1,1]}}
 {id=Random_1, children [WavePlayer_1,WavePlayer_5,WavePlayer_6], properties {weights [1,1,1]}}
 {id=WavePlayer_1, properties {soundWaveAssetPtr=&/IP_AnnoyedMop_AOE_Land_Distant_01}}
 {id=WavePlayer_2, properties {soundWaveAssetPtr=&/IP_AnnoyedMop_AOE_Land_01}}
 {id=WavePlayer_3, properties {soundWaveAssetPtr=&/IP_AnnoyedMop_AOE_Land_02}}
 {id=WavePlayer_5, properties {soundWaveAssetPtr=&/IP_AnnoyedMop_AOE_Land_Distant_02}}
 {id=WavePlayer_6, properties {soundWaveAssetPtr=&/IP_AnnoyedMop_AOE_Land_Distant_03}}
]
properties {
 bOverrideConcurrency=true
 maxDistance=12500
 soundClassObject=&Fort_WeaponMelee_PSM
 volumeMultiplier=1.5
 concurrencyOverrides {maxCount=3}
}
