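# Sound cue for the airborne phase of AnnoyedMop's AOE attack, blending looping ambience with a falling sound.
# A mixer balances four modulated layers: a random pick of three projectile loops, a fire loop, a boosted wind loop, and a random pick of three non-looping fall waves.
# It uses a dedicated jump-attack attenuation asset, a melee weapon sound class, and overridden concurrency to cap simultaneous instances.
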SoundCue /AnnoyedMop/Assets/Audio/AnnoyedMop_AOE_InAir_Loop_Cue
root=Mixer_0
nodes [
 {id=Mixer_0, children [Modulator_0,Modulator_1,Modulator_2,Modulator_3], properties {inputVolume [0.7,0.7,1.5,1]}}
 {id=Modulator_0, children [Random_0]}
 {id=Modulator_1, children [WavePlayer_0]}
 {id=Modulator_2, children [WavePlayer_2]}
 {id=Modulator_3, children [Random_2]}
 {id=Random_0, children [WavePlayer_1,WavePlayer_3,WavePlayer_4], properties {weights [1,1,1]}}
 {id=Random_2, children [WavePlayer_5,WavePlayer_6,WavePlayer_7], properties {weights [1,1,1]}}
 {id=WavePlayer_0, properties {bLooping=true, soundWaveAssetPtr=&/IP_AnnoyedMop_AOE_InAir_Fire_Loop}}
 {id=WavePlayer_1, properties {bLooping=true, soundWaveAssetPtr=&/IP_AnnoyedMop_Projectile_Loop_01}}
 {id=WavePlayer_2, properties {bLooping=true, soundWaveAssetPtr=&/IP_AnnoyedMop_AOE_InAir_Wind_Loop}}
 {id=WavePlayer_3, properties {bLooping=true, soundWaveAssetPtr=&/IP_AnnoyedMop_Projectile_Loop_02}}
 {id=WavePlayer_4, properties {bLooping=true, soundWaveAssetPtr=&/IP_AnnoyedMop_Projectile_Loop_03}}
 {id=WavePlayer_5, properties {soundWaveAssetPtr=&/IP_AnnoyedMop_AOE_InAir_Fall_01}}
 {id=WavePlayer_6, properties {soundWaveAssetPtr=&/IP_AnnoyedMop_AOE_InAir_Fall_02}}
 {id=WavePlayer_7, properties {soundWaveAssetPtr=&/IP_AnnoyedMop_AOE_InAir_Fall_03}}
]
properties {
 attenuationSettings=/AnnoyedMop/Assets/Audio/Settings/AnnoyedMop_AOE_Attack_Jump_InAir_Loop_Att
 bOverrideConcurrency=true
 maxDistance=8500
 soundClassObject=&Fort_WeaponMelee_PSM
 volumeMultiplier=1
 concurrencyOverrides {maxCount=3}
}
