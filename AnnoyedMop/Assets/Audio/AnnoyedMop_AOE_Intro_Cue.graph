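# Sound cue for the intro of the AnnoyedMop AOE attack, filed under the melee weapon sound class.
# A mixer layers an intro wave with a louder launch "Anime" wave.
# Each layer passes through its own modulator, with the launch layer pitched up in a randomized range.
# A jump-attack attenuation asset handles falloff, and a concurrency override caps simultaneous instances.
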SoundCue /AnnoyedMop/Assets/Audio/AnnoyedMop_AOE_Intro_Cue
root=Mixer_0
nodes [
 {id=Mixer_0, children [Modulator_0,Modulator_1], properties {inputVolume [1,2]}}
 {id=Modulator_0, children [WavePlayer_0]}
 {id=Modulator_1, children [WavePlayer_1], properties {pitchMax=1.3, pitchMin=1.2}}
 {id=WavePlayer_0, properties {soundWaveAssetPtr=&/IP_AnnoyedMop_AOE_Intro_01}}
 {id=WavePlayer_1, properties {soundWaveAssetPtr=&/IP_AnnoyedMop_AOE_Launch_Anime}}
]
properties {
 attenuationSettings=/AnnoyedMop/Assets/Audio/Settings/AnnoyedMop_AOE_Attack_Jump_Att
 bOverrideConcurrency=true
 maxDistance=8500
 soundClassObject=&Fort_WeaponMelee_PSM
 concurrencyOverrides {maxCount=3}
}
