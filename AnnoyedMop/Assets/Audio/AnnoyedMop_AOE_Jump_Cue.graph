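# Sound cue for the jump of the Annoyed Mop's area-of-effect attack.
# A Random node picks among three jump sound waves with equal weight and feeds a Modulator at the root.
# It plays through the melee weapon sound class with a dedicated attack-jump attenuation asset, a volume boost, and an overridden concurrency limit on simultaneous instances.
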SoundCue /AnnoyedMop/Assets/Audio/AnnoyedMop_AOE_Jump_Cue
root=Modulator_0
nodes [
 {id=Modulator_0, children [Random_0]}
 {id=Random_0, children [WavePlayer_3,WavePlayer_1,WavePlayer_0], properties {weights [1,1,1]}}
 {id=WavePlayer_0, properties {soundWaveAssetPtr=&/IP_AnnoyedMop_AOE_Jump_02}}
 {id=WavePlayer_1, properties {soundWaveAssetPtr=&/IP_AnnoyedMop_AOE_Jump_03}}
 {id=WavePlayer_3, properties {soundWaveAssetPtr=&/IP_AnnoyedMop_AOE_Jump_01}}
]
properties {
 attenuationSettings=/AnnoyedMop/Assets/Audio/Settings/AnnoyedMop_AOE_Attack_Jump_Att
 bOverrideConcurrency=true
 maxDistance=8500
 soundClassObject=&Fort_WeaponMelee_PSM
 volumeMultiplier=1.2
 concurrencyOverrides {maxCount=3}
}
