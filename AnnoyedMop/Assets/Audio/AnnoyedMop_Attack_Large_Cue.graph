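# Sound cue for the AnnoyedMop large attack, layering a heavy fire sound with a fire-ignition burst.
# A Mixer plays two equal-weight Random picks together, one from four HeavyFire waves and one from three shared Curie fire medium-start waves.
# A Modulator at the root adds slight pitch and volume variation.
# The cue uses the large-attack attenuation, a melee weapon sound class and the AnnoyedMop attacks concurrency set.
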SoundCue /AnnoyedMop/Assets/Audio/AnnoyedMop_Attack_Large_Cue
root=Modulator_0
nodes [
 {id=Mixer_0, children [Random_0,Random_2], properties {inputVolume [1,1]}}
 {id=Modulator_0, children [Mixer_0], properties {pitchMax=1.1, pitchMin=0.9, volumeMax=0.95, volumeMin=0.9}}
 {id=Random_0, children [WavePlayer_2,WavePlayer_0,WavePlayer_1,WavePlayer_3], properties {weights [1,1,1,1]}}
 {id=Random_2, children [WavePlayer_5,WavePlayer_8,WavePlayer_9], properties {weights [1,1,1]}}
 {id=WavePlayer_0, properties {soundWaveAssetPtr=&/IP_AnnoyedMop_HeavyFire_02}}
 {id=WavePlayer_1, properties {soundWaveAssetPtr=&/IP_AnnoyedMop_HeavyFire_03}}
 {id=WavePlayer_2, properties {soundWaveAssetPtr=&/IP_AnnoyedMop_HeavyFire_01}}
 {id=WavePlayer_3, properties {soundWaveAssetPtr=&/IP_AnnoyedMop_HeavyFire_04}}
 {id=WavePlayer_5, properties {soundWaveAssetPtr=/Game/Sounds/Fort_GamePlay_Sounds/Curie_Fires/Curie_Fire_MediumStart_05}}
 {id=WavePlayer_8, properties {soundWaveAssetPtr=/Game/Sounds/Fort_GamePlay_Sounds/Curie_Fires/Curie_Fire_MediumStart_06}}
 {id=WavePlayer_9, properties {soundWaveAssetPtr=/Game/Sounds/Fort_GamePlay_Sounds/Curie_Fires/Curie_Fire_MediumStart_04}}
]
properties {
 attenuationSettings=/AnnoyedMop/Assets/Audio/Settings/AnnoyedMop_Attack_Large_Att
 maxDistance=8500
 soundClassObject=&Fort_WeaponMelee_PSM
 volumeMultiplier=1.75
 concurrencyOverrides {resolutionRule=PreventNew}
 concurrencySet [/AnnoyedMop/Assets/Audio/Settings/AnnoyedMop_Attacks_Conc]
}
