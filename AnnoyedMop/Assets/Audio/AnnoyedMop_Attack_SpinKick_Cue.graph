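# Sound cue for the AnnoyedMop spin kick attack, layering fire, fire crackle and leg whoosh sounds.
# A mixer blends the three layers with crackle quieter, each pitch-varied by a Modulator over random wave picks.
# Per-layer Quality Level nodes use full variation pools in their first two slots and three-wave subsets, excluded from branch culling, in the last two.
# Shared AnnoyedMop attack attenuation and concurrency assets govern playback.
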SoundCue /AnnoyedMop/Assets/Audio/AnnoyedMop_Attack_SpinKick_Cue
root=Mixer_1
nodes [
 {id=Mixer_1, children [Modulator_0,Modulator_1,Modulator_2], properties {inputVolume [1.5,0.5,1.5]}}
 {id=Modulator_0, children [QualityLevel_0], properties {pitchMax=1.3, pitchMin=0.9}}
 {id=Modulator_1, children [QualityLevel_1], properties {pitchMax=1.2, pitchMin=1}}
 {id=Modulator_2, children [QualityLevel_2], properties {pitchMax=1.3, pitchMin=0.9}}
 {id=QualityLevel_0, children [Random_2,Random_2,Random_3,Random_3]}
 {id=QualityLevel_1, children [Random_4,Random_4,Random_5,Random_5]}
 {id=QualityLevel_2, children [Random_1,Random_1,Random_6,Random_6]}
 {
  id=Random_1
  children [WavePlayer_1, WavePlayer_3, WavePlayer_4, WavePlayer_5, WavePlayer_6, WavePlayer_22]
  properties {weights [1,1,1,1,1,1]}
 }
 {
  id=Random_2
  children [WavePlayer_0, WavePlayer_7, WavePlayer_8, WavePlayer_9, WavePlayer_10, WavePlayer_11, WavePlayer_12, WavePlayer_13]
  properties {weights [1,1,1,1,1,1,1,1]}
 }
 {id=Random_3, children [WavePlayer_11, WavePlayer_12, WavePlayer_13], properties {bShouldExcludeFromBranchCulling=true, weights [1,1,1]}}
 {
  id=Random_4
  children [WavePlayer_14, WavePlayer_15, WavePlayer_16, WavePlayer_17, WavePlayer_18, WavePlayer_19, WavePlayer_20, WavePlayer_21]
  properties {weights [1,1,1,1,1,1,1,1]}
 }
 {id=Random_5, children [WavePlayer_19, WavePlayer_20, WavePlayer_21], properties {bShouldExcludeFromBranchCulling=true, weights [1,1,1]}}
 {id=Random_6, children [WavePlayer_5, WavePlayer_6, WavePlayer_22], properties {bShouldExcludeFromBranchCulling=true, weights [1,1,1]}}
 {id=WavePlayer_0, properties {soundWaveAssetPtr=&/IP_AnnoyedMop_BasicFire_01}}
 {id=WavePlayer_1, properties {soundWaveAssetPtr=&/IP_AnnoyedMop_LegWhoosh_01}}
 {id=WavePlayer_10, properties {soundWaveAssetPtr=&/IP_AnnoyedMop_BasicFire_05}}
 {id=WavePlayer_11, properties {soundWaveAssetPtr=&/IP_AnnoyedMop_BasicFire_06}}
 {id=WavePlayer_12, properties {soundWaveAssetPtr=&/IP_AnnoyedMop_BasicFire_07}}
 {id=WavePlayer_13, properties {soundWaveAssetPtr=&/IP_AnnoyedMop_BasicFire_08}}
 {id=WavePlayer_14, properties {soundWaveAssetPtr=&/IP_AnnoyedMop_BasicFire_Crackel_01}}
 {id=WavePlayer_15, properties {soundWaveAssetPtr=&/IP_AnnoyedMop_BasicFire_Crackel_02}}
 {id=WavePlayer_16, properties {soundWaveAssetPtr=&/IP_AnnoyedMop_BasicFire_Crackel_03}}
 {id=WavePlayer_17, properties {soundWaveAssetPtr=&/IP_AnnoyedMop_BasicFire_Crackel_04}}
 {id=WavePlayer_18, properties {soundWaveAssetPtr=&/IP_AnnoyedMop_BasicFire_Crackel_05}}
 {id=WavePlayer_19, properties {soundWaveAssetPtr=&/IP_AnnoyedMop_BasicFire_Crackel_06}}
 {id=WavePlayer_20, properties {soundWaveAssetPtr=&/IP_AnnoyedMop_BasicFire_Crackel_07}}
 {id=WavePlayer_21, properties {soundWaveAssetPtr=&/IP_AnnoyedMop_BasicFire_Crackel_08}}
 {id=WavePlayer_22, properties {soundWaveAssetPtr=&/IP_AnnoyedMop_LegWhoosh_06}}
 {id=WavePlayer_3, properties {soundWaveAssetPtr=&/IP_AnnoyedMop_LegWhoosh_02}}
 {id=WavePlayer_4, properties {soundWaveAssetPtr=&/IP_AnnoyedMop_LegWhoosh_03}}
 {id=WavePlayer_5, properties {soundWaveAssetPtr=&/IP_AnnoyedMop_LegWhoosh_04}}
 {id=WavePlayer_6, properties {soundWaveAssetPtr=&/IP_AnnoyedMop_LegWhoosh_05}}
 {id=WavePlayer_7, properties {soundWaveAssetPtr=&/IP_AnnoyedMop_BasicFire_02}}
 {id=WavePlayer_8, properties {soundWaveAssetPtr=&/IP_AnnoyedMop_BasicFire_03}}
 {id=WavePlayer_9, properties {soundWaveAssetPtr=&/IP_AnnoyedMop_BasicFire_04}}
]
properties {
 attenuationSettings=/AnnoyedMop/Assets/Audio/Settings/AnnoyedMop_Attacks_Att
 maxDistance=8500
 soundClassObject=&Fort_WeaponMelee_PSM
 volumeMultiplier=1
 concurrencyOverrides {resolutionRule=PreventNew}
 concurrencySet [/AnnoyedMop/Assets/Audio/Settings/AnnoyedMop_Attacks_Conc]
}
