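# Looping ambient Sound Cue for the AnnoyedMop drop, likely heard while the item lies in the world as loot.
# A single looping Wave Player feeds the output directly, with slightly lowered volume and pitch.
# Shared weapon-drop attenuation, loot ambient sound class and loot ambient concurrency govern falloff, mixing and voice limits.
# Local attenuation and concurrency override values are stored, but their override switches remain off.
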SoundCue /AnnoyedMop/Assets/Audio/AnnoyedMop_Drop_Ambient_Loop_Cue
root=WavePlayer_1
nodes [{id=WavePlayer_1, properties {bLooping=true, soundWaveAssetPtr=&/IP_AnnoyedMop_Ambient_Loop}}]
properties {
 attenuationSettings=&Fort_Weapon_Drop_Att
 maxDistance=3000
 pitchMultiplier=0.99
 soundClassObject=/Game/Sounds/FortSoundClassesAndModes/Fort_Loot_Ambient_Loop
 volumeMultiplier=0.8
 attenuationOverrides {
  bEnableOcclusion=true
  falloffDistance=600
  occlusionTraceChannel=ECC_GameTraceChannel11
  occlusionVolumeAttenuation=0.2
  attenuationShapeExtents {x=250}
 }
 concurrencyOverrides {maxCount=3, resolutionRule=StopLowestPriority, voiceStealReleaseTime=2}
 concurrencySet [/Game/Sounds/Concurrency_Rules/Loot_Ambient_Conc]
}
