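# Sound cue for the Annoyed Mop's drop impact, played as the item lands.
# A Random node picks among four equally weighted landing waves and feeds a Modulator at default settings for subtle variation.
# It uses weapon-drop attenuation and the loot-collect sound class, with a slight volume boost.
# A concurrency override caps simultaneous instances and blocks new plays once the limit is reached.
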SoundCue /AnnoyedMop/Assets/Audio/AnnoyedMop_Drop_Impact_Cue
root=Modulator_0
nodes [
 {id=Modulator_0, children [Random_0]}
 {id=Random_0, children [WavePlayer_1,WavePlayer_2,WavePlayer_3,WavePlayer_4], properties {weights [1,1,1,1]}}
 {id=WavePlayer_1, properties {soundWaveAssetPtr=&/IP_AnnoyedMop_DropLand_01}}
 {id=WavePlayer_2, properties {soundWaveAssetPtr=&/IP_AnnoyedMop_DropLand_02}}
 {id=WavePlayer_3, properties {soundWaveAssetPtr=&/IP_AnnoyedMop_DropLand_03}}
 {id=WavePlayer_4, properties {soundWaveAssetPtr=&/IP_AnnoyedMop_DropLand_04}}
]
properties {
 attenuationSettings=&Fort_Weapon_Drop_Att
 bOverrideConcurrency=true
 maxDistance=3000
 soundClassObject=&Fort_Loot_Collect
 volumeMultiplier=1.2
 attenuationOverrides {falloffDistance=1000, attenuationShapeExtents {x=0}}
 concurrencyOverrides {maxCount=3, resolutionRule=PreventNew}
}
