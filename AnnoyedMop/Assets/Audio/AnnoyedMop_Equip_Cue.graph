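# Sound cue for equipping the AnnoyedMop, classed as melee weapon audio with player-equip attenuation and an owner-limited concurrency that stops the oldest instance.
# A mixer layers a random basic-fire variant at reduced volume, a random hand whoosh and a looping equip sound, all after the same short fixed delay.
# An enveloper gradually fades the loop, while the root modulator randomly raises the pitch of the whole mix.
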SoundCue /AnnoyedMop/Assets/Audio/AnnoyedMop_Equip_Cue
root=Modulator_0
nodes [
 {id=Delay_0, children [Random_2], properties {delayMax=0.25, delayMin=0.25}}
 {id=Delay_1, children [Random_3], properties {delayMax=0.25, delayMin=0.25}}
 {id=Delay_2, children [WavePlayer_0], properties {delayMax=0.25, delayMin=0.25}}
 {
  id=Enveloper_0
  children [Modulator_1]
  properties {pitchCurve {editorCurveData {keys [{time=0}]}}, volumeCurve {editorCurveData {keys [<at 0 1>,<at 4.035 0.913>,<at 15 0.2>]}}}
 }
 {id=Mixer_0, children [Delay_1,Delay_0,Enveloper_0], properties {inputVolume [0.5,1,1]}}
 {id=Modulator_0, children [Mixer_0], properties {pitchMax=1.25, pitchMin=1.2}}
 {id=Modulator_1, children [Delay_2]}
 {id=Random_2, children [WavePlayer_4,WavePlayer_6,WavePlayer_7,WavePlayer_8], properties {weights [1,1,1,1]}}
 {id=Random_3, children [WavePlayer_9, WavePlayer_10, WavePlayer_11, WavePlayer_12], properties {weights [1,1,1,1]}}
 {id=WavePlayer_0, properties {bLooping=true, soundWaveAssetPtr=&/P_AnnoyedMop_Equip_Loop}}
 {id=WavePlayer_10, properties {soundWaveAssetPtr=&/IP_AnnoyedMop_BasicFire_02}}
 {id=WavePlayer_11, properties {soundWaveAssetPtr=&/IP_AnnoyedMop_BasicFire_03}}
 {id=WavePlayer_12, properties {soundWaveAssetPtr=&/IP_AnnoyedMop_BasicFire_04}}
 {id=WavePlayer_4, properties {soundWaveAssetPtr=&/IP_AnnoyedMop_HandWhoosh_01}}
 {id=WavePlayer_6, properties {soundWaveAssetPtr=&/IP_AnnoyedMop_HandWhoosh_02}}
 {id=WavePlayer_7, properties {soundWaveAssetPtr=&/IP_AnnoyedMop_HandWhoosh_03}}
 {id=WavePlayer_8, properties {soundWaveAssetPtr=&/IP_AnnoyedMop_HandWhoosh_04}}
 {id=WavePlayer_9, properties {soundWaveAssetPtr=&/IP_AnnoyedMop_BasicFire_01}}
]
properties {
 attenuationSettings=/Game/Sounds/Attenuation/Fort_Player_Equip_Generic_Att
 maxDistance=1800
 soundClassObject=&Fort_WeaponMelee
 volumeMultiplier=1
 concurrencyOverrides {bLimitToOwner=true, maxCount=2, resolutionRule=StopOldest}
}
