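# Sound cue for the heavy attack windup of AnnoyedMop, routed to a melee weapon sound class with attack attenuation.
# A mixer layers three pitch-modulated branches: a random windup, a random fire crackle and a random hand whoosh.
# Quality-level nodes switch the crackle and whoosh layers between full variant pools and three-wave subsets to scale variety.
# A concurrency override caps simultaneous instances.
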SoundCue /AnnoyedMop/Assets/Audio/AnnoyedMop_Heavy_Windup_Cue
root=Mixer_0
nodes [
 {id=Mixer_0, children [Modulator_0,Modulator_1,Modulator_2], properties {inputVolume [2,2,1.5]}}
 {id=Modulator_0, children [Random_4], properties {pitchMax=1.5, pitchMin=0.9}}
 {id=Modulator_1, children [QualityLevel_0], properties {pitchMax=1, pitchMin=1}}
 {id=Modulator_2, children [QualityLevel_1], properties {pitchMax=0.95, pitchMin=0.9}}
 {id=QualityLevel_0, children [Random_0,Random_0,Random_1,Random_1]}
 {id=QualityLevel_1, children [Random_2,Random_2,Random_3,Random_3]}
 {
  id=Random_0
  children [WavePlayer_1,WavePlayer_2,WavePlayer_3,WavePlayer_4,WavePlayer_5,WavePlayer_6,WavePlayer_7,WavePlayer_8]
  properties {weights [1,1,1,1,1,1,1,1]}
 }
 {id=Random_1, children [WavePlayer_6,WavePlayer_7,WavePlayer_8], properties {weights [1,1,1]}}
 {
  id=Random_2
  children [WavePlayer_9, WavePlayer_10, WavePlayer_11, WavePlayer_12, WavePlayer_13, WavePlayer_14]
  properties {weights [1,1,1,1,1,1]}
 }
 {id=Random_3, children [WavePlayer_12, WavePlayer_13, WavePlayer_14], properties {weights [1,1,1]}}
 {id=Random_4, children [WavePlayer_0, WavePlayer_15, WavePlayer_16], properties {weights [1,1,1]}}
 {id=WavePlayer_0, properties {soundWaveAssetPtr=&/IP_AnnoyedMop_Heavy_Windup_01}}
 {id=WavePlayer_1, properties {soundWaveAssetPtr=&/IP_AnnoyedMop_BasicFire_Crackel_01}}
 {id=WavePlayer_10, properties {soundWaveAssetPtr=&/IP_AnnoyedMop_HandWhoosh_02}}
 {id=WavePlayer_11, properties {soundWaveAssetPtr=&/IP_AnnoyedMop_HandWhoosh_03}}
 {id=WavePlayer_12, properties {soundWaveAssetPtr=&/IP_AnnoyedMop_HandWhoosh_04}}
 {id=WavePlayer_13, properties {soundWaveAssetPtr=&/IP_AnnoyedMop_HandWhoosh_05}}
 {id=WavePlayer_14, properties {soundWaveAssetPtr=&/IP_AnnoyedMop_HandWhoosh_06}}
 {id=WavePlayer_15, properties {soundWaveAssetPtr=&/IP_AnnoyedMop_Heavy_Windup_02}}
 {id=WavePlayer_16, properties {soundWaveAssetPtr=&/IP_AnnoyedMop_Heavy_Windup_03}}
 {id=WavePlayer_2, properties {soundWaveAssetPtr=&/IP_AnnoyedMop_BasicFire_Crackel_02}}
 {id=WavePlayer_3, properties {soundWaveAssetPtr=&/IP_AnnoyedMop_BasicFire_Crackel_03}}
 {id=WavePlayer_4, properties {soundWaveAssetPtr=&/IP_AnnoyedMop_BasicFire_Crackel_04}}
 {id=WavePlayer_5, properties {soundWaveAssetPtr=&/IP_AnnoyedMop_BasicFire_Crackel_05}}
 {id=WavePlayer_6, properties {soundWaveAssetPtr=&/IP_AnnoyedMop_BasicFire_Crackel_06}}
 {id=WavePlayer_7, properties {soundWaveAssetPtr=&/IP_AnnoyedMop_BasicFire_Crackel_07}}
 {id=WavePlayer_8, properties {soundWaveAssetPtr=&/IP_AnnoyedMop_BasicFire_Crackel_08}}
 {id=WavePlayer_9, properties {soundWaveAssetPtr=&/IP_AnnoyedMop_HandWhoosh_01}}
]
properties {
 attenuationSettings=/AnnoyedMop/Assets/Audio/Settings/AnnoyedMop_Attacks_Att
 bOverrideConcurrency=true
 maxDistance=8500
 soundClassObject=&Fort_WeaponMelee_PSM
 volumeMultiplier=0.8
 concurrencyOverrides {maxCount=3}
}
