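# Sound cue for the Annoyed Mop's leg whoosh, classed as a melee weapon sound with attack attenuation.
# A quality-level switch sends its first two slots to a random pick of six whoosh waves and its last two to the final three of those waves, with that smaller branch excluded from branch culling.
# A root modulator varies pitch, and a concurrency override caps simultaneous instances.
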SoundCue /AnnoyedMop/Assets/Audio/AnnoyedMop_LegWhoosh_Cue
root=Modulator_0
nodes [
 {id=Modulator_0, children [QualityLevel_0], properties {pitchMax=1.3, pitchMin=0.9}}
 {id=QualityLevel_0, children [Random_2,Random_2,Random_3,Random_3]}
 {
  id=Random_2
  children [WavePlayer_2, WavePlayer_7, WavePlayer_8, WavePlayer_9, WavePlayer_10, WavePlayer_11]
  properties {weights [1,1,1,1,1,1]}
 }
 {id=Random_3, children [WavePlayer_9, WavePlayer_10, WavePlayer_11], properties {bShouldExcludeFromBranchCulling=true, weights [1,1,1]}}
 {id=WavePlayer_10, properties {soundWaveAssetPtr=&/IP_AnnoyedMop_LegWhoosh_05}}
 {id=WavePlayer_11, properties {soundWaveAssetPtr=&/IP_AnnoyedMop_LegWhoosh_06}}
 {id=WavePlayer_2, properties {soundWaveAssetPtr=&/IP_AnnoyedMop_LegWhoosh_01}}
 {id=WavePlayer_7, properties {soundWaveAssetPtr=&/IP_AnnoyedMop_LegWhoosh_02}}
 {id=WavePlayer_8, properties {soundWaveAssetPtr=&/IP_AnnoyedMop_LegWhoosh_03}}
 {id=WavePlayer_9, properties {soundWaveAssetPtr=&/IP_AnnoyedMop_LegWhoosh_04}}
]
properties {
 attenuationSettings=/AnnoyedMop/Assets/Audio/Settings/AnnoyedMop_Attacks_Att
 bOverrideConcurrency=true
 maxDistance=8500
 pitchMultiplier=0.8
 soundClassObject=&Fort_WeaponMelee_PSM
 volumeMultiplier=1.8
 concurrencyOverrides {maxCount=3}
}
