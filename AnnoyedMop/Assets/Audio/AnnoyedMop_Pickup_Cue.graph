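# Sound cue that plays the pickup sound for the AnnoyedMop item.
# A random node picks one of four equally weighted wave variations and feeds a modulator that adds slight pitch variation.
# The cue lowers volume, raises base pitch, and uses shared Fort weapon-pickup attenuation, the loot-collect sound class and the PickupMaximum_Oldest concurrency set.
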SoundCue /AnnoyedMop/Assets/Audio/AnnoyedMop_Pickup_Cue
root=Modulator_0
nodes [
 {id=Modulator_0, children [Random_1], properties {pitchMax=1.02, pitchMin=0.99, volumeMin=1}}
 {id=Random_1, children [WavePlayer_0,WavePlayer_1,WavePlayer_2,WavePlayer_3], properties {weights [1,1,1,1]}}
 {id=WavePlayer_0, properties {soundWaveAssetPtr=&/IP_AnnoyedMop_Pickup_01}}
 {id=WavePlayer_1, properties {soundWaveAssetPtr=&/IP_AnnoyedMop_Pickup_02}}
 {id=WavePlayer_2, properties {soundWaveAssetPtr=&/IP_AnnoyedMop_Pickup_03}}
 {id=WavePlayer_3, properties {soundWaveAssetPtr=&/IP_AnnoyedMop_Pickup_04}}
]
properties {
 attenuationSettings=&Fort_Weapon_Pickup_Att
 maxDistance=3500
 pitchMultiplier=1.1
 soundClassObject=&Fort_Loot_Collect
 volumeMultiplier=0.6
 concurrencyOverrides {maxCount=1, resolutionRule=PreventNew}
 concurrencySet [&PickupMaximum_Oldest]
}
