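# Sound cue for AnnoyedMop projectile impacts, layering impact waves with a quieter, briefly delayed flame-trap "Hot Feet Bounce" sound.
# Each layer runs through a Modulator into a Quality Level switch that picks randomly from five variations, or from three in the last two quality slots.
# The reduced random pools are excluded from branch culling.
# A dedicated attenuation asset and two concurrency assets, global and limited, govern playback in the Fort_Impacts sound class.
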SoundCue /AnnoyedMop/Assets/Audio/AnnoyedMop_Projectile_Impacts_Cue
root=Mixer_0
nodes [
 {id=Delay_0, children [Modulator_2], properties {delayMax=0.2, delayMin=0.2}}
 {id=Mixer_0, children [Delay_0,Modulator_0], properties {inputVolume [0.5,3]}}
 {id=Modulator_0, children [QualityLevel_1]}
 {id=Modulator_2, children [QualityLevel_0]}
 {id=QualityLevel_0, children [Random_1,Random_1,Random_2,Random_2]}
 {id=QualityLevel_1, children [Random_0,Random_0,Random_3,Random_3]}
 {id=Random_0, children [WavePlayer_0,WavePlayer_1,WavePlayer_2,WavePlayer_3,WavePlayer_6], properties {weights [1,1,1,1,1]}}
 {id=Random_1, children [WavePlayer_4, WavePlayer_5, WavePlayer_8, WavePlayer_9, WavePlayer_10], properties {weights [1,1,1,1,1]}}
 {id=Random_2, children [WavePlayer_8, WavePlayer_9, WavePlayer_10], properties {bShouldExcludeFromBranchCulling=true, weights [1,1,1]}}
 {id=Random_3, children [WavePlayer_2,WavePlayer_3,WavePlayer_6], properties {bShouldExcludeFromBranchCulling=true, weights [1,1,1]}}
 {id=WavePlayer_0, properties {soundWaveAssetPtr=&/IP_AnnoyedMop_Projectile_Impact_01}}
 {id=WavePlayer_1, properties {soundWaveAssetPtr=&/IP_AnnoyedMop_Projectile_Impact_02}}
 {id=WavePlayer_10, properties {soundWaveAssetPtr=/Game/Sounds/Fort_Traps/Flame_Trap/Hot_Feet_Bounce_05}}
 {id=WavePlayer_2, properties {soundWaveAssetPtr=&/IP_AnnoyedMop_Projectile_Impact_03}}
 {id=WavePlayer_3, properties {soundWaveAssetPtr=&/IP_AnnoyedMop_Projectile_Impact_04}}
 {id=WavePlayer_4, properties {soundWaveAssetPtr=/Game/Sounds/Fort_Traps/Flame_Trap/Hot_Feet_Bounce_01}}
 {id=WavePlayer_5, properties {soundWaveAssetPtr=/Game/Sounds/Fort_Traps/Flame_Trap/Hot_Feet_Bounce_02}}
 {id=WavePlayer_6, properties {soundWaveAssetPtr=&/IP_AnnoyedMop_Projectile_Impact_05}}
 {id=WavePlayer_8, properties {soundWaveAssetPtr=/Game/Sounds/Fort_Traps/Flame_Trap/Hot_Feet_Bounce_03}}
 {id=WavePlayer_9, properties {soundWaveAssetPtr=/Game/Sounds/Fort_Traps/Flame_Trap/Hot_Feet_Bounce_04}}
]
properties {
 attenuationSettings=/AnnoyedMop/Assets/Audio/Settings/AnnoyedMop_Projectile_Impact_Att
 bDebug=true
 maxDistance=7500
 soundClassObject=&Fort_Impacts
 volumeMultiplier=1.3
 concurrencyOverrides {bLimitToOwner=true, maxCount=1, resolutionRule=PreventNew, volumeScale=0.5}
 concurrencySet [/AnnoyedMop/Assets/Audio/Settings/AnnoyedMop_Impacts_Global_Conc, /AnnoyedMop/Assets/Audio/Settings/AnnoyedMop_Impacts_Limited_Conc]
}
