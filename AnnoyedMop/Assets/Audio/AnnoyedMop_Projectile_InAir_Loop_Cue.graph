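# Looping in-air sound cue for the AnnoyedMop projectile, routed to the melee weapon sound class.
# A Random node picks one of three equally weighted looping waves, and an Enveloper fades its volume up from near silence over a quarter second.
# A root Modulator applies default pitch and volume variation.
# An attenuation asset shapes distance falloff, concurrency is capped at three, and the loop keeps playing when silent.
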SoundCue /AnnoyedMop/Assets/Audio/AnnoyedMop_Projectile_InAir_Loop_Cue
root=Modulator_0
nodes [
 {
  id=Enveloper_0
  children [Random_0]
  properties {
   pitchCurve {editorCurveData {keys [{time=0}]}}
   volumeCurve {editorCurveData {keys [
      {interpMode=RCIM_Cubic, tangentMode=RCTM_User, time=0, value=0.01}
      {arriveTangent=-0.019, interpMode=RCIM_Cubic, leaveTangent=-0.019, tangentMode=RCTM_User, time=0.25, value=1}
     ]}}
  }
 }
 {id=Modulator_0, children [Enveloper_0]}
 {id=Random_0, children [WavePlayer_0,WavePlayer_1,WavePlayer_2], properties {weights [1,1,1]}}
 {id=WavePlayer_0, properties {bLooping=true, soundWaveAssetPtr=&/IP_AnnoyedMop_Projectile_Loop_01}}
 {id=WavePlayer_1, properties {bLooping=true, soundWaveAssetPtr=&/IP_AnnoyedMop_Projectile_Loop_02}}
 {id=WavePlayer_2, properties {bLooping=true, soundWaveAssetPtr=&/IP_AnnoyedMop_Projectile_Loop_03}}
]
properties {
 attenuationSettings=/AnnoyedMop/Assets/Audio/Settings/AnnoyedMop_Projectile_InAir_Loop_Att
 bOverrideConcurrency=true
 maxDistance=4500
 soundClassObject=&Fort_WeaponMelee
 virtualizationMode=PlayWhenSilent
 volumeMultiplier=1
 concurrencyOverrides {maxCount=3}
}
