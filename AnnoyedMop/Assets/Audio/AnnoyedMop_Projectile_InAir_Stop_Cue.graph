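# Sound cue for the stop sound of the AnnoyedMop projectile in flight, classed as melee weapon audio.
# A quality-level node picks randomly from six waves in its first two slots and only the last three waves in the other two.
# The result passes through a short volume fade-in envelope and a default Modulator.
# It reuses the in-air loop's attenuation, caps concurrency and keeps playing when silent.
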SoundCue /AnnoyedMop/Assets/Audio/AnnoyedMop_Projectile_InAir_Stop_Cue
root=Modulator_0
nodes [
 {
  id=Enveloper_0
  children [QualityLevel_0]
  properties {
   pitchCurve {editorCurveData {keys [{time=0}]}}
   volumeCurve {editorCurveData {keys [
      {interpMode=RCIM_Cubic, tangentMode=RCTM_User, time=0, value=0}
      {arriveTangent=-0.019, interpMode=RCIM_Cubic, leaveTangent=-0.019, tangentMode=RCTM_User, time=0.16, value=1}
     ]}}
  }
 }
 {id=Modulator_0, children [Enveloper_0]}
 {id=QualityLevel_0, children [Random_0,Random_0,Random_1,Random_1]}
 {id=Random_0, children [WavePlayer_1,WavePlayer_2,WavePlayer_3,WavePlayer_4,WavePlayer_5,WavePlayer_6], properties {weights [1,1,1,1,1,1]}}
 {id=Random_1, children [WavePlayer_4,WavePlayer_5,WavePlayer_6], properties {weights [1,1,1]}}
 {id=WavePlayer_1, properties {soundWaveAssetPtr=&/IP_AnnoyedMop_Projectile_InAir_Stop_01}}
 {id=WavePlayer_2, properties {soundWaveAssetPtr=&/IP_AnnoyedMop_Projectile_InAir_Stop_02}}
 {id=WavePlayer_3, properties {soundWaveAssetPtr=&/IP_AnnoyedMop_Projectile_InAir_Stop_03}}
 {id=WavePlayer_4, properties {soundWaveAssetPtr=&/IP_AnnoyedMop_Projectile_InAir_Stop_04}}
 {id=WavePlayer_5, properties {soundWaveAssetPtr=&/IP_AnnoyedMop_Projectile_InAir_Stop_05}}
 {id=WavePlayer_6, properties {soundWaveAssetPtr=&/IP_AnnoyedMop_Projectile_InAir_Stop_06}}
]
properties {
 attenuationSettings=/AnnoyedMop/Assets/Audio/Settings/AnnoyedMop_Projectile_InAir_Loop_Att
 bOverrideConcurrency=true
 maxDistance=4500
 soundClassObject=&Fort_WeaponMelee
 virtualizationMode=PlayWhenSilent
 volumeMultiplier=0.4
 concurrencyOverrides {maxCount=3}
}
