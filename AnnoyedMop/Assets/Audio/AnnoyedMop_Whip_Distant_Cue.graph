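# Sound cue for the distant whip-by of an AnnoyedMop projectile, classed as a ranged weapon projectile effect.
# A quality-level switch picks randomly among five wave variations in its first two slots and a three-wave subset in the last two.
# A neutral modulator roots the graph, while cue settings raise pitch and lower volume.
# It borrows shared bullet-whip attenuation and whiz-by concurrency, with virtualization disabled.
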SoundCue /AnnoyedMop/Assets/Audio/AnnoyedMop_Whip_Distant_Cue
root=Modulator_0
nodes [
 {id=Modulator_0, children [QualityLevel_1], properties {pitchMax=1, pitchMin=1, volumeMax=1, volumeMin=1}}
 {id=QualityLevel_1, children [Random_0,Random_0,Random_3,Random_3]}
 {
  id=Random_0
  children [WavePlayer_1, WavePlayer_2, WavePlayer_3, WavePlayer_4, WavePlayer_12]
  properties {bShouldExcludeFromBranchCulling=true, weights [1,1,1,1,1]}
 }
 {id=Random_3, children [WavePlayer_3, WavePlayer_4, WavePlayer_12], properties {bShouldExcludeFromBranchCulling=true, weights [1,1,1]}}
 {id=WavePlayer_1, properties {soundWaveAssetPtr=&/IP_AnnoyedMop_Projectile_Whip_Distant_01}}
 {id=WavePlayer_12, properties {soundWaveAssetPtr=&/IP_AnnoyedMop_Projectile_Whip_Distant_05}}
 {id=WavePlayer_2, properties {soundWaveAssetPtr=&/IP_AnnoyedMop_Projectile_Whip_Distant_02}}
 {id=WavePlayer_3, properties {soundWaveAssetPtr=&/IP_AnnoyedMop_Projectile_Whip_Distant_03}}
 {id=WavePlayer_4, properties {soundWaveAssetPtr=&/IP_AnnoyedMop_Projectile_Whip_Distant_04}}
]
properties {
 attenuationSettings=/Game/Athena/Sounds/Weapons/BulletWhips/bullet_whip_Att.Bullet_Whip_Att
 bBypassVolumeScaleForPriority=true
 maxDistance=2097152
 pitchMultiplier=1.2
 priority=0.5
 soundClassObject=&Fort_WeaponRanged_Projectile_Effects
 virtualizationMode=Disabled
 volumeMultiplier=0.4
 attenuationOverrides {bAttenuate=false, stereoSpread=0}
 concurrencySet [/Game/Sounds/Concurrency_Rules/Athena_BulletWhizBy_Conc]
}
