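# Unlit, two-sided translucent Niagara mesh-particle material that draws flame-like spline trails, apparently for the AnnoyedMop equip effect.
# SplineThicken sets strip width from dynamic parameters, with panning noise along the normal adding wobble that scales along U.
# Distorted, scrolling fire texture is eroded by dissolve noise and a particle-driven U mask to form opacity, depth-faded through a quality switch.
# Particle color times dissolve noise feeds emissive adjustments.
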Material /AnnoyedMop/Assets/VFX/Materials/M_AnnoyedMop_EquipSpline
graph {
 properties {
  bUsedWithNiagaraMeshParticles=true
  blendMode=BLEND_Translucent
  pixelDepthOffsetMode=PDOM_Legacy
  shadingModel=MSM_Unlit
  twoSided=true
  usedShadingModels=Unlit
 }
 nodes [
  n1 = TextureCoordinate()
  n2 = DynamicParameter() {parameterIndex=2, defaultValue=<rgba 0 0.01 1 1>, paramNames [Rand, Width Tip, 2, 3]}
  n3 = Panner(Coordinate=Multiply(A=TextureCoordinate() {vTiling=0.4}) {constB=0.5}) {speedX=0.25}
  n4 = Multiply(
   A=TextureSample(UVs=Add(A=@n2.Rand, B=@n3)) {samplerType=SAMPLERTYPE_Grayscale, texture=/FortWater/Textures/LowResBlurredNoise}.R
   B=VertexNormalWS()
  )
  n5 = DynamicParameter() {defaultValue=<rgba 0 1 1 1>, paramNames [Width,1,2,3]}
  Add(
   A=MaterialFunctionCall(
    WidthBase=DynamicParameter() {defaultValue=<rgba 0 1 1 1>, paramNames [Width,1,2,3]}.Width
    WidthTip=DynamicParameter() {parameterIndex=2, defaultValue=<rgba 0 0.01 1 1>, paramNames [Rand, Width Tip, 2, 3]}.Width Tip
    Expand U or V UV Channel=StaticBool()
   ) {materialFunction=&SplineThicken}.WorldPositionOffset
   B=Multiply(A=ComponentMask(None=@n1) {r=true}, B=Multiply(A=@n4, B=@n5.2))
  )
  Multiply()
  n6 = TextureCoordinate()
  Multiply(
   A=SphereMask(A=ComponentMask(None=@n6) {g=true, r=true}, B=Constant() {r=0.5}) {attenuationRadius=0.5, hardnessPercent=0}
   B=DynamicParameter() {defaultValue=<rgba 0 1 1 1>, paramNames [Width,1,2,3]}.1
  ) {constB=0.01}
  n7 = ParticleColor()
  n8 = TextureCoordinate() {uTiling=1.5}
  n9 = Panner(Coordinate=Multiply(A=TextureCoordinate() {vTiling=0.4}, B=ScalarParameter() {parameterName=DistortionScale}) {constB=2}) {speedX=0.25}
  n10 = TextureSample(UVs=@n9) {texture=/AnnoyedMop/Assets/VFX/Textures/T_ColdDay_Fire_Trail_Dissolve}
  n11 = Multiply(A=Subtract(A=Multiply(A=@n10.R) {constB=2}), B=ScalarParameter() {parameterName=DistortionStrength}) {constB=0.1}
  n12 = Time()
  n13 = DynamicParameter() {parameterIndex=2, defaultValue=<rgba 0 0.01 1 1>, paramNames [Rand, Width Tip, 2, 3]}
  n14 = ScalarParameter() {defaultValue=-0.85, parameterName=UV Speed}
  n15 = Constant()
  n16 = TextureSample(UVs=Panner(Coordinate=Add(A=@n8, B=@n11), Time=Add(A=@n12, B=@n13.Rand), Speed=AppendVector(A=@n14, B=@n15)) {bFractionalPart=true}) {texture=/Game/Effects/Fort_Effects/Textures/T_PlantStand_Fire}
  n17 = Multiply(A=@n16.R) {constB=0.3}
  n18 = Multiply(A=@n16.G) {constB=0.3}
  n19 = DynamicParameter() {parameterIndex=2, defaultValue=<rgba 0 0.01 1 1>, paramNames [Rand, Width Tip, 2, 3]}
  n20 = Multiply(A=TextureCoordinate() {vTiling=0.4}, B=ScalarParameter() {parameterName=DistortionScale}) {constB=2}
  n21 = TextureSample(UVs=Panner(Coordinate=Add(A=@n19.Rand, B=@n20)) {speedX=0.25}) {texture=&T_ColdDay_Fire_Trail_Dissolve}
  n22 = Multiply(A=DynamicParameter() {defaultValue=<rgba 0 1 1 1>, paramNames [Width,1,2,3]}.3, B=@n21.R)
  n23 = TextureCoordinate()
  n24 = OneMinus(None=SphereMask(
    A=ComponentMask(None=@n23) {r=true}
    B=DynamicParameter() {parameterIndex=1, paramNames [Thickness/OpacityBar, GlowOpacity, WobbleStrength, NoiseOpacity]}.Thickness/OpacityBar
   ) {attenuationRadius=0.7, hardnessPercent=25})
  n25 = Subtract(A=Add(A=Add(A=@n17, B=@n18), B=Multiply(A=@n16.B) {constB=0.3}), B=Saturate(None=Add(A=@n22, B=@n24)))
  n26 = VertexColor()
  n27 = Multiply(A=@n7.A, B=Saturate(None=Multiply(A=@n25, B=@n26.R)))
  NamedRerouteDeclaration(None=@n27) {name=Opacity_Flames, nodeColor=<rgba 1 0.7969 0 1>}
  NamedRerouteDeclaration(None=@n10.R) {name=DistortionNoise, nodeColor=<rgba 0 0.8594 1 1>}
  ParticleRandom()
  QualitySwitch(Default=DepthFade(Opacity=@n27) {fadeDistanceDefault=25}, Low=@n27, Medium=@n27)
  Reroute(None=MaterialFunctionCall(Emissive=Multiply(A=@n7.RGB, B=@n21.R)) {materialFunction=&MF_FX_EmissiveAdjustments}.Result)
  SphereMask()
  TextureSample(UVs=@n9) {samplerType=SAMPLERTYPE_Grayscale, texture=&LowResBlurredNoise}
 ]
}
