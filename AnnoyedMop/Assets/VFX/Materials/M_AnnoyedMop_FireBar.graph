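# Unlit, two-sided translucent fire bar VFX material for Niagara ribbons, mesh particles and skeletal meshes.
# World position offset combines optional sine-pulsed SplineThicken, panned-noise turbulence and local-space scaling.
# A panning fire texture, with flowmap swirl and noise displacement UV distortion, tints its four channels with separate core, fire and glow colors.
# Opacity combines dissolve-noise erosion, noise-varied depth fade, view-angle fade and particle alpha.
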Material /AnnoyedMop/Assets/VFX/Materials/M_AnnoyedMop_FireBar
graph {
 properties {
  bUsedWithMeshDeformer=true
  bUsedWithNiagaraMeshParticles=true
  bUsedWithNiagaraRibbons=true
  bUsedWithSkeletalMesh=true
  blendMode=BLEND_Translucent
  pixelDepthOffsetMode=PDOM_Legacy
  shadingModel=MSM_Unlit
  twoSided=true
  usedShadingModels=Unlit
 }
 nodes [
  n1 = ScalarParameter() {defaultValue=200, parameterName=WPO}
  n2 = Multiply(A=Time())
  n3 = ComponentMask(None=TextureCoordinate()) {r=true}
  n4 = Cosine(None=Multiply(A=Add(A=@n2, B=@n3), B=ScalarParameter() {defaultValue=3, parameterName=Sine}) {constB=5})
  n5 = Multiply(A=Multiply(A=@n4) {constB=0.3}, B=ScalarParameter() {defaultValue=3, parameterName=Sine Affect Rate}) {constB=3}
  n6 = DynamicParameter() {parameterIndex=1, defaultValue=<rgba 1 1 20 10>, paramNames [Random Ofs, Swirl Strength, WPO Side Intensity, WPO Wave Intensity]}
  n7 = VectorParameter() {group=WPO, parameterName=WPO Side UV, sortPriority=6, defaultValue=<rgba 0.5 0.25 0.5 -0.1>}
  n8 = TextureCoordinate()
  n9 = StaticSwitchParameter(True=Add(A=NamedRerouteUsage(), B=@n8), False=@n8) {parameterName=Use WPos UV}
  n10 = Multiply(A=@n9, B=AppendVector(A=@n7.R, B=@n7.R))
  n11 = AppendVector(A=@n7.B, B=@n7.B)
  n12 = TextureSample(UVs=Add(A=AppendVector(A=@n6.Random Ofs, B=@n6.Random Ofs), B=Panner(Coordinate=@n10, Speed=@n11))) {samplerType=SAMPLERTYPE_Normal, texture=/Game/Characters/Player/Female/Medium/Bodies/F_MED_Bistro_Astronaut/FX/T_GaussianNoise_Normal}
  n13 = ComponentMask(None=@n12.RGB) {g=true, r=true}
  n14 = ScalarParameter() {defaultValue=20, parameterName=WPO Side Intensity}
  n15 = DynamicParameter() {parameterIndex=1, defaultValue=<rgba 1 1 20 10>, paramNames [Random Ofs, Swirl Strength, WPO Side Intensity, WPO Wave Intensity]}
  n16 = VectorParameter() {group=WPO, parameterName=WPO Wave UV, sortPriority=6, defaultValue=<rgba 1 0.25 0 1>}
  n17 = Multiply(A=@n9, B=AppendVector(A=@n16.R, B=@n16.R))
  n18 = AppendVector(A=@n16.B, B=@n16.B)
  n19 = TextureSample(UVs=Add(A=AppendVector(A=@n15.Random Ofs, B=@n15.Random Ofs), B=Panner(Coordinate=@n17, Speed=@n18))) {texture=&T_Smooth_tile}
  n20 = ScalarParameter() {defaultValue=10, parameterName=WPO Wave Intensity}
  n21 = TextureCoordinate()
  n22 = LinearInterpolate(
   B=AppendVector(A=Multiply(A=@n13, B=@n14) {constB=10}, B=Multiply(A=@n19.R, B=@n20) {constB=10})
   Alpha=StaticSwitchParameter(True=ComponentMask(None=@n21) {r=true}, False=Constant() {r=1}) {defaultValue=true, parameterName=Turb Fade Mask}
  )
  n23 = Add(
   A=MaterialFunctionCall(WidthBase=Add(A=@n1, B=@n5), WidthTip=ScalarParameter() {defaultValue=200.015, parameterName=WPO TIP}, Expand U or V UV Channel=StaticBool()) {materialFunction=&SplineThicken}.WorldPositionOffset
   B=@n22
  )
  n24 = TextureCoordinate()
  n25 = Multiply(A=Subtract(A=ComponentMask(None=@n24) {r=true}) {constB=0.5}) {constB=2}
  n26 = Multiply(A=OneMinus(None=Abs(None=@n25)), B=@n23)
  n27 = ScalarParameter() {parameterName=Edge Control Lerp}
  n28 = Subtract(A=WorldPosition().XYZ, B=ActorPositionWS())
  Add(
   A=StaticSwitchParameter(True=LinearInterpolate(A=@n23, B=@n26, Alpha=@n27), False=@n22) {parameterName=USE Thicken}
   B=Multiply(A=Transform(None=@n28) {transformSourceType=TRANSFORMSOURCE_World, transformType=TRANSFORM_Local}, B=ScalarParameter() {parameterName=Scale}) {constB=-0.9}
  )
  Add(A=@n4)
  n29 = VectorParameter() {parameterName=Core Color, defaultValue=<rgba 1 0.3792 0.0539 1>}
  n30 = NamedRerouteUsage()
  n31 = ComponentMask(None=Transform(None=WorldPosition().XYZ) {transformSourceType=TRANSFORMSOURCE_World, transformType=TRANSFORM_Local}) {g=true, r=true}
  n32 = AppendVector(A=ScalarParameter() {defaultValue=0.005, parameterName=World UV Scale}, B=Constant())
  n33 = TextureCoordinate()
  n34 = Add(A=NamedRerouteDeclaration(None=Multiply(A=@n31, B=@n32) {constB=0.01}) {name=WPosUV, nodeColor=<rgba 1 0 0.9609 1>}, B=@n33)
  n35 = ScalarParameter() {defaultValue=2, parameterName=Main U Scale}
  n36 = Constant() {r=1}
  n37 = VectorParameter() {parameterName=Dissolve UV, defaultValue=<rgba 0.5 0.4 1 1>}
  n38 = Multiply(
   A=Multiply(A=StaticSwitchParameter(True=@n34, False=@n33) {parameterName=Use WPos UV}, B=AppendVector(A=@n35, B=@n36))
   B=AppendVector(A=@n37.B, B=@n37.B)
  )
  n39 = Constant2Vector() {r=0.2}
  n40 = ScalarParameter() {defaultValue=1, parameterName=Main Fire Pan}
  n41 = Time()
  n42 = Add(A=Panner(Coordinate=Add(A=@n38, B=@n39), Time=Multiply(A=@n40, B=@n41)) {speedX=-1}, B=NamedRerouteUsage())
  n43 = MaterialFunctionCall(Float4=TextureSampleParameter2D(UVs=Add(A=@n30, B=@n42)) {parameterName=Tex_Default1, texture=/AnnoyedMop/Assets/VFX/Textures/T_ColdDay_Fire_Trail00B}.RGBA) {materialFunction=&BreakOutFloat4Components}
  n44 = VectorParameter() {parameterName=Core Color2, defaultValue=<rgba 1 0.8284 0.2855 1>}
  n45 = Multiply(A=@n43.B, B=@n43.B)
  n46 = VectorParameter() {parameterName=Fire Color, defaultValue=<rgba 0.9323 0.0777 0 1>}
  n47 = VectorParameter() {parameterName=Glow Color, defaultValue=<rgba 1 0.0167 0 1>}
  n48 = Add(
   A=Add(A=Multiply(A=@n29.RGB, B=@n43.R), B=Multiply(A=@n44.RGB, B=@n45))
   B=Add(A=Multiply(A=@n43.G, B=@n46.RGB), B=Multiply(A=@n47.RGB, B=@n43.A))
  )
  MaterialFunctionCall(Emissive=@n48) {materialFunction=&MF_FX_EmissiveAdjustments}
  n49 = ParticleColor()
  Multiply(A=@n49.RGB, B=@n48)
  n50 = Constant() {r=1}
  n51 = TextureCoordinate()
  n52 = StaticSwitchParameter(True=Add(A=NamedRerouteUsage(), B=@n51), False=@n51) {parameterName=Use WPos UV}
  n53 = ComponentMask(None=@n37.RGB) {g=true, r=true}
  n54 = TextureSampleParameter2D(UVs=Add(A=@n42, B=Multiply(A=@n52, B=@n53))) {parameterName=Dissolve Noise Tex, texture=/AnnoyedMop/Assets/VFX/Textures/T_ColdDay_Fire_Trail_Dissolve}
  n55 = LinearInterpolate(
   A=ScalarParameter() {defaultValue=100, parameterName=DepthFade_Low}
   B=ScalarParameter() {defaultValue=10, parameterName=DepthFade_high}
   Alpha=@n54.R
  )
  n56 = Subtract(A=@n48, B=Multiply(A=@n54.R, B=ScalarParameter() {defaultValue=2, parameterName=Dissolve}))
  n57 = Multiply(A=Saturate(None=ComponentMask(None=@n56) {r=true}), B=NamedRerouteUsage())
  n58 = MaterialFunctionCall("FadeSensitivity (optional)"=ScalarParameter() {defaultValue=0.3, parameterName=Angle Fade}) {materialFunction=&FadeBasedOnViewAngle}
  Multiply(A=@n49.A, B=Multiply(A=DepthFade(Opacity=@n50, FadeDistance=@n55), B=Multiply(A=@n57, B=@n58.Result)))
  n59 = Multiply(A=TextureCoordinate()) {constB=0.5}
  n60 = MaterialFunctionCall(Float2=Abs(None=Sine(None=@n59))) {materialFunction=&BreakOutFloat2Components}
  n61 = Saturate(None=Multiply(A=Multiply(A=@n60.R, B=@n60.R)) {constB=2})
  NamedRerouteDeclaration(None=Multiply(A=@n61, B=@n61)) {name=Square Mask, nodeColor=<rgba 0 1 0.6016 1>}
  n62 = ComponentMask(None=TextureCoordinate()) {r=true}
  n63 = ScalarParameter() {defaultValue=1, parameterName=Swirl Gradient Lerp}
  n64 = Time()
  n65 = ScalarParameter() {defaultValue=0.5, parameterName=Swirl Speed}
  n66 = TextureCoordinate()
  n67 = Multiply(
   A=AppendVector(A=ScalarParameter() {defaultValue=2, parameterName=Swirl UV Scale U}, B=ScalarParameter() {defaultValue=0.9, parameterName=Swirl UV Scale V})
   B=Add(A=Multiply(A=@n64, B=@n65), B=StaticSwitchParameter(True=@n34, False=@n66) {parameterName=Use WPos UV})
  )
  n68 = Subtract(A=ComponentMask(None=TextureSample(UVs=@n67) {samplerType=SAMPLERTYPE_LinearColor, texture=/Game/VisualThreatManager/StormVisuals/LargeClouds/T_Swirl_Flowmap}.RGB) {g=true, r=true}) {constB=0.5}
  n69 = ScalarParameter() {defaultValue=0.5, parameterName=Swirl Strength}
  NamedRerouteDeclaration(None=Multiply(A=LinearInterpolate(B=@n62, Alpha=@n63) {constA=1}, B=Multiply(A=@n68, B=@n69))) {name=Swirl UV, nodeColor=<rgba 0 1 0.9297 1>}
  NamedRerouteDeclaration(None=Multiply(A=Subtract(A=@n13) {constB=0.5}, B=ScalarParameter() {defaultValue=0.1, parameterName=Displacement1 Insnsity})) {name=Displacement1, nodeColor=<rgba 0 0.5078 1 1>}
  ObjectPositionWS()
 ]
}
