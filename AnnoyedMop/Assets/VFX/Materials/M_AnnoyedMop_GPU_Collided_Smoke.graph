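# Unlit translucent Niagara sprite material for AnnoyedMop collided smoke, defaulting to rock sprite textures.
# Color multiplies texture, randomized tint and time-of-day mids through emissive adjustments.
# Below a particle-speed threshold, it computes an offset from randomized spread minus age-powered gravity, apparently faking a post-collision splash and fall.
# Opacity combines texture mask, tint and particle alpha with a depth fade whose distance varies with mask and speed.
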Material /AnnoyedMop/Assets/VFX/Materials/M_AnnoyedMop_GPU_Collided_Smoke
graph {
 properties {bUsedWithNiagaraSprites=true, blendMode=BLEND_Translucent, pixelDepthOffsetMode=PDOM_Legacy, shadingModel=MSM_Unlit, usedShadingModels=Unlit}
 nodes [
  n1 = Saturate(None=Divide(A=ParticleSpeed(), B=ScalarParameter() {defaultValue=2, group=Speed, parameterName=Speed}))
  n2 = Constant()
  n3 = VectorParameter() {group=Fall, parameterName=HowFarToFallZ, defaultValue=<rgba 20 20 50 0>}
  n4 = Transform(None=@n3.RGB) {transformSourceType=TRANSFORMSOURCE_Local}
  n5 = Subtract(A=ParticleRandom()) {constB=0.5}
  n6 = AppendVector(A=Multiply(A=ComponentMask(None=@n4) {g=true, r=true}, B=Multiply(A=@n5) {constB=2}), B=@n3.B)
  n7 = ParticleRelativeTime()
  n8 = ScalarParameter() {defaultValue=2, desc=33, group=Fall, parameterName=Splash Speed Mul}
  n9 = Multiply(A=DynamicParameter() {paramNames [Random,Param2,Param3,Param4]}.Random) {constB=0.5}
  n10 = Multiply(
   A=Power(Base=Add(A=@n7), Exp=@n8)
   B=Multiply(A=Add(A=@n9) {constB=0.5}, B=VectorParameter() {group=Fall, parameterName=Gravity, defaultValue=<rgba 0 0 -1 0>}.RGB)
  )
  If(A=@n1, B=Constant() {r=1}, "A > B"=@n2, "A == B"=@n2, "A < B"=LinearInterpolate(B=Subtract(A=@n6, B=@n10), Alpha=Multiply(A=@n7, B=@n8)))
  MaterialFunctionCall(
   Color_Morning=ScalarParameter() {defaultValue=0.7, parameterName=HDRMult_3}
   Color_Day=ScalarParameter() {defaultValue=1, parameterName=HDRMult_2}
   Color_Evening=ScalarParameter() {defaultValue=0.7, parameterName=HDRMult_1}
   Color_Night=ScalarParameter() {defaultValue=0.5, parameterName=HDRMult}
  ) {materialFunction=&Color_Time_of_Day}
  n11 = VectorParameter() {group=Smoke Colors, parameterName=2 - Morning Mids Smoke, sortPriority=1, defaultValue=<rgba 0.445 0.3231 0.248 1>}
  n12 = VectorParameter() {group=Smoke Colors, parameterName=2 - Daytime Mids Smoke, sortPriority=2, defaultValue=<rgba 0.6719 0.7182 0.75 1>}
  n13 = VectorParameter() {group=Smoke Colors, parameterName=2 - Evening Mids Smoke, sortPriority=3, defaultValue=<rgba 0.79 0.4657 0.4032 1>}
  n14 = VectorParameter() {group=Smoke Colors, parameterName=2 - Nighttime Mids Smoke, sortPriority=4, defaultValue=<rgba 0.2808 0.3998 0.4688 1>}
  n15 = DynamicParameter() {paramNames [Random,Param2,Param3,Param4]}
  n16 = VectorParameter() {parameterName=Color, defaultValue=<rgba 0.0938 0.0831 0.0754 1>}
  n17 = Multiply(A=Add(A=Multiply(A=@n15.Random) {constB=0.5}) {constB=0.5}, B=@n16.RGB)
  n18 = TextureSampleParameter2D() {parameterName=Texture, texture=/Game/Effects/Fort_Effects/Textures/Rocks/T_Rock_Stone_Sprite_D}
  MaterialFunctionCall(Emissive=Multiply(
    A=MaterialFunctionCall(Color_Morning=@n11.RGB, Color_Day=@n12.RGB, Color_Evening=@n13.RGB, Color_Night=@n14.RGB) {desc=Mids, materialFunction=&Color_Time_of_Day}.Result
    B=Multiply(A=@n17, B=@n18.RGB)
   )) {materialFunction=&MF_FX_EmissiveAdjustments}
  n19 = StaticComponentMaskParameter(None=@n18.RGBA) {defaultA=true, parameterName=Alpha Channel}
  n20 = Constant() {r=1}
  n21 = LinearInterpolate(A=ScalarParameter() {defaultValue=1, parameterName=DepthFade Max}, Alpha=LinearInterpolate(A=@n19, Alpha=@n1) {constB=0}) {constB=0.5}
  Multiply(A=Multiply(A=Multiply(A=@n16.A, B=@n19), B=DepthFade(Opacity=@n20, FadeDistance=@n21)), B=ParticleColor().A)
  TextureSampleParameter2D() {parameterName=Normal Texture, samplerType=SAMPLERTYPE_Normal, texture=/Game/Effects/Fort_Effects/Textures/Rocks/T_Rock_Stone_Sprite_N}
 ]
}
