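# Additive, unlit, two-sided glow material for Niagara mesh particles on spline-shaped effect meshes.
# SplineThicken takes its base and tip width from the GlowWidth dynamic parameter.
# Particle color feeds an emissive adjustment function, while particle alpha scales a soft centered UV sphere mask and a double-sided Fresnel to form opacity.
# A quality switch skips depth fade on Low and Medium.
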Material /AnnoyedMop/Assets/VFX/Materials/M_AnnoyedMop_GlowSpline
graph {
 properties {
  bUsedWithNiagaraMeshParticles=true
  blendMode=BLEND_Additive
  pixelDepthOffsetMode=PDOM_Legacy
  shadingModel=MSM_Unlit
  twoSided=true
  usedShadingModels=Unlit
 }
 nodes [
  n1 = DynamicParameter() {paramNames [Param1,Param2,GlowWidth,Param4]}
  MaterialFunctionCall(WidthBase=@n1.GlowWidth, WidthTip=@n1.GlowWidth, Expand U or V UV Channel=StaticBool()) {materialFunction=&SplineThicken}
  n2 = ParticleColor()
  MaterialFunctionCall(Emissive=@n2.RGB) {materialFunction=&MF_FX_EmissiveAdjustments}
  n3 = SphereMask(A=TextureCoordinate(), B=Constant() {r=0.5}) {attenuationRadius=0.5, hardnessPercent=0}
  n4 = Multiply(
   A=Multiply(A=Power(Base=@n3), B=@n2.A)
   B=MaterialFunctionCall() {materialFunction=&MF_Fresnel_FX}.Fresnel Double Sided
  )
  QualitySwitch(Default=DepthFade(Opacity=@n4), Low=@n4, Medium=@n4)
 ]
}
