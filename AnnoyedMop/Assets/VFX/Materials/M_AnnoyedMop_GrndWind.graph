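# Translucent unlit material for Niagara mesh particles, apparently a ground-wind effect for AnnoyedMop.
# A panning mask texture, tiled by a dynamic parameter, provides noise that is added to particle color, and particle color also feeds an emissive-adjustment function.
# An inverted-UV smoke disk texture combines with particle alpha and vertex red as an opacity-style mask.
# A separate chain builds a transformed Y-axis vector from UVs, scaled by the dynamic Thickness parameter.
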Material /AnnoyedMop/Assets/VFX/Materials/M_AnnoyedMop_GrndWind
graph {
 properties {
  bUsedWithNiagaraMeshParticles=true
  blendMode=BLEND_Translucent
  pixelDepthOffsetMode=PDOM_Legacy
  shadingModel=MSM_Unlit
  usedShadingModels=Unlit
 }
 nodes [
  n1 = ParticleColor()
  n2 = Multiply(
   A=AppendVector(A=DynamicParameter() {defaultValue=<rgba 20 1 1 1>, paramNames [Thickness,Tiling,Param3,Param4]}.Tiling, B=Constant() {r=1})
   B=TextureCoordinate() {vTiling=0.5}
  )
  n3 = Constant2Vector() {g=1, r=4}
  n4 = TextureSample(UVs=Panner(Coordinate=Multiply(A=@n2, B=@n3)) {speedY=1}) {samplerType=SAMPLERTYPE_Masks, texture=/AnnoyedMop/Assets/VFX/Textures/T_FX_AnnoyedMop_Prj}
  Add(A=@n1.RGB, B=@n4.R)
  MaterialFunctionCall(Emissive=@n1.RGB) {materialFunction=&MF_FX_EmissiveAdjustments}
  n5 = TextureCoordinate()
  n6 = Constant()
  n7 = Multiply(A=AppendVector(A=Subtract(A=@n5, B=@n6), B=Constant() {r=1}), B=Constant3Vector() {constant=<rgba 0 1 0 0>})
  Multiply(
   A=Multiply(A=Transform(None=@n7))
   B=DynamicParameter() {defaultValue=<rgba 20 1 1 1>, paramNames [Thickness,Tiling,Param3,Param4]}.Thickness
  )
  n8 = TextureSample(UVs=Add(A=OneMinus(None=@n2), B=Constant2Vector() {g=0.5})) {texture=/Game/Effects/Fort_Effects/Textures/Smoke/T_Smoke_Bomb_Disk2}
  Multiply(A=Saturate(None=Multiply(A=@n1.A, B=@n8.G)), B=VertexColor().R)
  Multiply(A=NamedRerouteUsage(), B=Constant2Vector() {g=0.1})
  NamedRerouteDeclaration(None=@n4.R) {name=Noise, nodeColor=<rgba 1 0.9844 0 1>}
  VectorParameter() {parameterName=Thickness, defaultValue=<rgba 0 1 1 10>}
 ]
}
