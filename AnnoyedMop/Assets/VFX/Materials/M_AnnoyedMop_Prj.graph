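# Unlit, translucent, two-sided material for Niagara mesh particles, apparently the AnnoyedMop projectile VFX.
# Two panning dissolve-noise samples, cut by vertex red, shape the body and blend glow into a base gradient with hotspots.
# Opacity erodes a mask texture by particle alpha.
# Vertex offsets combine noise wobble along normals with a view-space shift; dynamic parameters drive UV scaling and time offset.
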Material /AnnoyedMop/Assets/VFX/Materials/M_AnnoyedMop_Prj
graph {
 properties {
  bUsedWithNiagaraMeshParticles=true
  blendMode=BLEND_Translucent
  pixelDepthOffsetMode=PDOM_Legacy
  shadingModel=MSM_Unlit
  twoSided=true
  usedShadingModels=Unlit
 }
 nodes [
  n1 = TextureCoordinate() {uTiling=1.25, vTiling=0.5}
  n2 = Add(A=Time(), B=DynamicParameter() {parameterIndex=2}.Param2)
  n3 = Multiply(A=TextureSample(UVs=Panner(Coordinate=@n1, Time=@n2) {speedY=1}) {texture=&T_NoisePack_01}.R) {constB=2}
  n4 = VertexNormalWS()
  n5 = ScalarParameter() {group=Wobble, parameterName=VertexWobble}
  n6 = ComponentMask(None=TextureCoordinate() {vTiling=0.5}) {g=true}
  n7 = Transform(None=VectorParameter() {group=Vertical Offset, parameterName=VertOffset, defaultValue=<rgba 0 0 0 1>}.RGB) {transformSourceType=TRANSFORMSOURCE_View}
  n8 = MaterialFunctionCall(Input=NamedRerouteUsage(), Input Low=Constant() {r=0.5}, Input High=Constant() {r=1}) {materialFunction=&RemapValueRange}
  Add(A=Multiply(A=Add(A=@n3) {constB=-1}, B=Multiply(A=@n4, B=@n5)), B=Multiply(A=Multiply(A=@n6, B=@n7), B=Clamp(None=@n8.Result)))
  n9 = VertexInterpolator(VS=TextureSample(UVs=Time()) {texture=&T_NoisePack_01}.R)
  n10 = VectorParameter() {group=Glow, parameterName=GlowColor, defaultValue=<rgba 0 0 0 1>}
  n11 = Multiply(A=Multiply(A=VertexColor().R, B=Add(A=@n9.PS) {constB=0.5}), B=@n10.A)
  n12 = ParticleColor()
  Add(A=@n11, B=Multiply(A=ScalarParameter() {parameterName=Opacity_FullyErode}, B=OneMinus(None=@n12.A)))
  n13 = TextureCoordinate()
  n14 = LinearInterpolate(
   A=VectorParameter() {parameterName=BaseColor, defaultValue=<rgba 0 0 0 1>}.RGB
   B=VectorParameter() {parameterName=DarkerBaseColor, defaultValue=<rgba 0 0 0 1>}.RGB
   Alpha=SphereMask(A=ComponentMask(None=@n13) {g=true, r=true}, B=Constant2Vector() {r=0.5}, Radius=ScalarParameter() {parameterName=r}) {attenuationRadius=0.5, hardnessPercent=0}
  )
  n15 = VectorParameter() {parameterName=HotspotColor, defaultValue=<rgba 0 0 0 1>}
  n16 = NamedRerouteUsage()
  n17 = MaterialFunctionCall(XY=Multiply(A=TextureCoordinate() {vTiling=0.7}, B=NamedRerouteUsage())) {materialFunction=&Swizzle}
  n18 = NamedRerouteUsage()
  n19 = Multiply(A=TextureCoordinate() {uTiling=0.5, vTiling=0.6}, B=NamedRerouteUsage())
  n20 = Add(
   A=TextureSample(UVs=Panner(Coordinate=@n17.YX, Time=@n18) {bFractionalPart=true, speedX=-1}) {texture=/AnnoyedMop/Assets/VFX/Textures/T_ColdDay_Fire_Trail_Dissolve}.R
   B=TextureSample(UVs=Panner(Coordinate=@n19, Time=@n18) {bFractionalPart=true, speedY=-1}) {texture=&T_ColdDay_Fire_Trail_Dissolve}.R
  )
  n21 = Multiply(A=OneMinus(None=VertexColor().R), B=ScalarParameter() {parameterName=e}) {constB=0.5}
  n22 = Saturate(None=Subtract(A=@n20, B=@n21))
  MaterialFunctionCall(Emissive=LinearInterpolate(A=@n10.RGB, B=LinearInterpolate(A=@n14, B=@n15.RGB, Alpha=@n16), Alpha=@n22)) {materialFunction=&MF_FX_EmissiveAdjustments}
  n23 = OneMinus(None=ParticleColor().A)
  n24 = ScalarParameter() {parameterName=erode_w}
  n25 = Panner(Coordinate=Multiply(A=TextureCoordinate() {uTiling=0.25, vTiling=0.6}, B=NamedRerouteUsage()), Time=NamedRerouteUsage()) {bFractionalPart=true, speedY=-1}
  Multiply(
   A=SmoothStep(
    Min=@n23
    Max=Add(A=@n23, B=@n24)
    Value=TextureSample(UVs=@n25) {samplerType=SAMPLERTYPE_Masks, texture=/AnnoyedMop/Assets/VFX/Textures/T_FX_AnnoyedMop_Prj}.R
   )
   B=Add(A=@n22, B=@n11)
  )
  n26 = Divide(A=@n20)
  n27 = Add(A=@n26, B=NamedRerouteUsage())
  n28 = TextureCoordinate()
  n29 = Power(Base=OneMinus(None=ComponentMask(None=@n28) {g=true}), Exp=ScalarParameter() {group=Highlights, parameterName=LeadingEdge_Exponent})
  NamedRerouteDeclaration(None=SmoothStep(Min=ScalarParameter() {parameterName=hotspot_min}, Max=ScalarParameter() {parameterName=hotspot_max}, Value=Add(A=@n27, B=@n29)) {constMax=0.75, constMin=0.5}) {name=Hotspots, nodeColor=<rgba 0.5234 0 1 1>}
  n30 = DynamicParameter() {parameterIndex=2}
  NamedRerouteDeclaration(None=Add(A=Time(), B=@n30.Param2)) {name=T, nodeColor=<rgba 1 0 0.4922 1>}
  n31 = Subtract(A=Add(A=@n21, B=ScalarParameter() {parameterName=EdgeOffset}), B=@n20)
  n32 = ScalarParameter() {defaultValue=2, parameterName=EdgeExponent}
  NamedRerouteDeclaration(None=Saturate(None=Power(Base=@n31, Exp=@n32))) {name=EdgeMask, nodeColor=<rgba 1 0.0703 0 1>}
  n33 = Transform(None=Constant3Vector() {constant=<rgba 0 1 0 0>}) {transformSourceType=TRANSFORMSOURCE_Local}
  n34 = Normalize(VectorInput=Subtract(
    A=CameraPositionWS()
    B=MaterialFunctionCall() {materialFunction=&ObjectPivotPoint}.Object Pivot Location
   ))
  NamedRerouteDeclaration(None=Saturate(None=DotProduct(A=@n33, B=@n34))) {name=View_FromBehind, nodeColor=<rgba 1 0.9141 0 1>}
  NamedRerouteDeclaration(None=AppendVector(A=@n30.Param1, B=Constant() {r=1})) {name=UVScaling, nodeColor=<rgba 0.1719 0 1 1>}
  NamedRerouteDeclaration(None=@n26) {name=MainNoise, nodeColor=<rgba 0 1 0.8828 1>}
  NamedRerouteUsage()
  ScalarParameter() {parameterName=erode_p}
 ]
}
