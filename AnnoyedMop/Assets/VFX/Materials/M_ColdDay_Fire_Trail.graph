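# Translucent unlit fire trail VFX material for Niagara ribbons and meshes, with dynamic parameters driving motion and dissolve.
# Panned texture channels tint four color layers, while dissolve noise, masks, view-angle and depth fades shape opacity.
# World position offset combines noise waves, optional cosine-modulated SplineThicken and axis bending, alongside a flowmap swirl for UVs.
# Static switch "Sp" selects a sprite variant with alternate UVs, core texture, mask and billboard offset.
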Material /AnnoyedMop/Assets/VFX/Materials/M_ColdDay_Fire_Trail
graph {
 properties {
  bUsedWithMeshDeformer=true
  bUsedWithNiagaraMeshParticles=true
  bUsedWithNiagaraRibbons=true
  bUsedWithSkeletalMesh=true
  blendMode=BLEND_Translucent
  pixelDepthOffsetMode=PDOM_Legacy
  shadingModel=MSM_Unlit
  twoSided=true
  usedShadingModels=Unlit
 }
 nodes [
  n1 = VectorParameter() {parameterName=Core Color, defaultValue=<rgba 1 0.3792 0.0539 1>}
  n2 = TextureCoordinate() {coordinateIndex=1}
  n3 = TextureCoordinate()
  n4 = DynamicParameter() {parameterIndex=2, defaultValue=<rgba 3 0 0.25 0>, paramNames [Sine, Sine Scroll, Bend Amount, Rand]}
  n5 = Multiply(A=VertexColor().R, B=@n4.Rand)
  n6 = Constant()
  n7 = VectorParameter() {parameterName=Dissolve UV, defaultValue=<rgba 0.5 0.4 1 1>}
  n8 = Multiply(A=Add(A=StaticSwitchParameter(True=@n2, False=@n3) {parameterName=Sp}, B=AppendVector(A=@n5, B=@n6)), B=AppendVector(A=@n7.B, B=@n7.B))
  n9 = Constant2Vector() {r=0.2}
  n10 = Add(
   A=Panner(Coordinate=Add(A=@n8, B=@n9), Time=DynamicParameter() {defaultValue=<rgba 40 1 100 1>, paramNames [WPO, Speed, WPO Tip, Dissolve]}.Speed) {speedX=-1}
   B=NamedRerouteUsage()
  )
  n11 = Rotator(Coordinate=Multiply(A=@n10) {constB=0.2}, Time=VertexColor().R)
  n12 = MaterialFunctionCall(Float4=StaticSwitchParameter(
    True=TextureSampleParameter2D(UVs=@n11) {parameterName=Tex_Core, texture=/Game/Effects/FogVolume/noise_flamy_Tile}.RGBA
    False=TextureSampleParameter2D(UVs=@n10) {parameterName=Tex_Default1, texture=/AnnoyedMop/Assets/VFX/Textures/T_ColdDay_Fire_Trail00B}.RGBA
   ) {parameterName=Sp}) {materialFunction=&BreakOutFloat4Components}
  n13 = VectorParameter() {parameterName=Core Color2, defaultValue=<rgba 1 0.8284 0.2855 1>}
  n14 = Multiply(A=@n12.B, B=@n12.B)
  n15 = VectorParameter() {parameterName=Fire Color, defaultValue=<rgba 0.9323 0.0777 0 1>}
  n16 = VectorParameter() {parameterName=Glow Color, defaultValue=<rgba 1 0.0167 0 1>}
  n17 = Add(
   A=Add(A=Multiply(A=@n1.RGB, B=@n12.R), B=Multiply(A=@n13.RGB, B=@n14))
   B=Add(A=Multiply(A=@n12.G, B=@n15.RGB), B=Multiply(A=@n16.RGB, B=@n12.A))
  )
  n18 = Multiply(A=@n17, B=ScalarParameter() {defaultValue=8, parameterName=Core Sprite ColorScale})
  n19 = ParticleColor()
  MaterialFunctionCall(Emissive=Multiply(A=StaticSwitchParameter(True=@n18, False=@n17) {parameterName=Sp}, B=@n19.RGB)) {materialFunction=&MF_FX_EmissiveAdjustments}
  n20 = TextureCoordinate() {coordinateIndex=1}
  n21 = Constant() {r=0.5}
  n22 = Multiply(A=TextureCoordinate()) {constB=0.5}
  n23 = MaterialFunctionCall(Float2=Abs(None=Sine(None=@n22))) {materialFunction=&BreakOutFloat2Components}
  n24 = Saturate(None=Multiply(A=Multiply(A=@n23.R, B=@n23.R)) {constB=2})
  NamedRerouteDeclaration(None=StaticSwitchParameter(True=SphereMask(A=@n20, B=@n21, Radius=@n21) {hardnessPercent=1}, False=Multiply(A=@n24, B=@n24)) {parameterName=Sp}) {name=Square Mask, nodeColor=<rgba 0 1 0.6016 1>}
  n25 = OneMinus(None=StaticSwitchParameter(True=TextureCoordinate() {coordinateIndex=1}, False=TextureCoordinate()) {parameterName=Sp})
  n26 = Time()
  n27 = ScalarParameter() {defaultValue=0.5, parameterName=Swirl Speed}
  n28 = TextureCoordinate() {coordinateIndex=1}
  n29 = TextureCoordinate()
  n30 = Multiply(
   A=AppendVector(
    A=ScalarParameter() {defaultValue=0.9, parameterName=Swirl UV Scale U}
    B=ScalarParameter() {defaultValue=0.9, parameterName=Swirl UV Scale V}
   )
   B=Add(A=Multiply(A=@n26, B=@n27), B=StaticSwitchParameter(True=@n28, False=@n29) {parameterName=Sp})
  )
  n31 = Subtract(A=ComponentMask(None=TextureSample(UVs=@n30) {samplerType=SAMPLERTYPE_LinearColor, texture=/Game/VisualThreatManager/StormVisuals/LargeClouds/T_Swirl_Flowmap}.RGB) {g=true, r=true}) {constB=0.5}
  n32 = DynamicParameter() {parameterIndex=1, defaultValue=<rgba 1 1 20 10>, paramNames [Random Ofs, Swirl Strength, WPO Side Intensity, WPO Wave Intensity]}
  NamedRerouteDeclaration(None=Multiply(A=ComponentMask(None=@n25) {g=true}, B=Multiply(A=@n31, B=@n32.Swirl Strength))) {name=Swirl UV, nodeColor=<rgba 0 1 0.9297 1>}
  n33 = VectorParameter() {parameterName=Rotation Axis, defaultValue=<rgba 1 0 0 1>}
  n34 = DynamicParameter() {parameterIndex=2, defaultValue=<rgba 3 0 0.25 0>, paramNames [Sine, Sine Scroll, Bend Amount, Rand]}
  n35 = ComponentMask(None=TextureCoordinate()) {r=true}
  NamedRerouteDeclaration(None=RotateAboutAxis(
    NormalizedRotationAxis=Normalize(VectorInput=@n33.RGB)
    RotationAngle=Multiply(A=@n34.Bend Amount, B=@n35)
    PivotPoint=ObjectPositionWS()
    Position=WorldPosition() {worldPositionShaderOffset=WPT_ExcludeAllShaderOffsets}.XYZ
   )) {name=Bend, nodeColor=<rgba 0.1953 0 1 1>}
  n36 = WorldPosition()
  n37 = ObjectPositionWS()
  n38 = TextureCoordinate() {coordinateIndex=1}
  n39 = VertexColor()
  n40 = Add(
   A=AppendVector(A=ScalarParameter() {defaultValue=10, parameterName=ScaleX}, B=ScalarParameter() {defaultValue=10, parameterName=ScaleY})
   B=Multiply(A=Subtract(A=@n39.G) {constB=0.5}, B=ScalarParameter() {defaultValue=5, parameterName=RandomScale})
  )
  NamedRerouteDeclaration(None=Add(
    A=Subtract(A=@n36.XYZ, B=@n37)
    B=MaterialFunctionCall!(Center Location, 0-1 UVs=@n38, XY scale=@n40, OverridePivot, "Normals (Optional)") {materialFunction=&Sprite}.World position offset
   )) {name=Sprite, nodeColor=<rgba 0.5 1 0 1>}
  n41 = Multiply(A=@n17, B=@n19.A)
  n42 = Add(A=@n10, B=Multiply(A=TextureCoordinate(), B=ComponentMask(None=@n7.RGB) {g=true, r=true}))
  n43 = Multiply(A=@n42) {constB=0.25}
  n44 = Multiply(
   A=TextureSampleParameter2D(UVs=StaticSwitchParameter(True=@n43, False=@n42) {parameterName=Sp}) {parameterName=Dissolve Noise Tex, texture=/AnnoyedMop/Assets/VFX/Textures/T_ColdDay_Fire_Trail_Dissolve}.R
   B=DynamicParameter() {defaultValue=<rgba 40 1 100 1>, paramNames [WPO, Speed, WPO Tip, Dissolve]}.Dissolve
  )
  n45 = Constant() {r=0.5}
  n46 = Multiply(
   A=Saturate(None=Subtract(A=@n41, B=@n44))
   B=SphereMask(A=TextureCoordinate(), B=@n45, Radius=@n45, Hardness=ScalarParameter() {defaultValue=0.5, parameterName=Sphere Mask hardness}) {attenuationRadius=0.5}
  )
  n47 = NamedRerouteUsage()
  n48 = ScalarParameter() {defaultValue=1, parameterName=Angle Fade}
  n49 = DepthFade(
   Opacity=Multiply(
    A=Multiply(A=@n46, B=@n47)
    B=MaterialFunctionCall("FadeSensitivity (optional)"=@n48) {materialFunction=&FadeBasedOnViewAngle}.Result
   )
   FadeDistance=ScalarParameter() {defaultValue=30, parameterName=Depth Fade}
  )
  StaticSwitchParameter(True=Multiply(A=@n49) {constB=0.2}, False=@n49) {parameterName=Sp}
  n50 = DynamicParameter() {defaultValue=<rgba 40 1 100 1>, paramNames [WPO, Speed, WPO Tip, Dissolve]}
  n51 = DynamicParameter() {parameterIndex=2, defaultValue=<rgba 3 0 0.25 0>, paramNames [Sine, Sine Scroll, Bend Amount, Rand]}
  n52 = Add(A=@n51.Sine Scroll, B=ComponentMask(None=TextureCoordinate()) {r=true})
  n53 = DynamicParameter() {parameterIndex=2, defaultValue=<rgba 3 0 0.25 0>, paramNames [Sine, Sine Scroll, Bend Amount, Rand]}
  n54 = Add(A=Cosine(None=Multiply(A=@n52, B=@n53.Sine) {constB=5}))
  n55 = Add(A=@n50.WPO, B=Multiply(A=Multiply(A=@n54) {constB=0.3}) {constB=70})
  n56 = StaticBool()
  n57 = DynamicParameter() {parameterIndex=1, defaultValue=<rgba 1 1 20 10>, paramNames [Random Ofs, Swirl Strength, WPO Side Intensity, WPO Wave Intensity]}
  n58 = VectorParameter() {group=WPO, parameterName=WPO Side UV, sortPriority=6, defaultValue=<rgba 0.5 0.25 -0.1 0.5>}
  n59 = TextureCoordinate()
  n60 = Multiply(A=@n59, B=AppendVector(A=@n58.R, B=@n58.R))
  n61 = AppendVector(A=@n58.B, B=@n58.B)
  n62 = TextureSample(UVs=Add(A=AppendVector(A=@n57.Random Ofs, B=@n57.Random Ofs), B=Panner(Coordinate=@n60, Speed=@n61))) {samplerType=SAMPLERTYPE_Normal, texture=/Game/Characters/Player/Female/Medium/Bodies/F_MED_Bistro_Astronaut/FX/T_GaussianNoise_Normal}
  n63 = ComponentMask(None=@n62.RGB) {g=true, r=true}
  n64 = DynamicParameter() {parameterIndex=1, defaultValue=<rgba 1 1 20 10>, paramNames [Random Ofs, Swirl Strength, WPO Side Intensity, WPO Wave Intensity]}
  n65 = DynamicParameter() {parameterIndex=1, defaultValue=<rgba 1 1 20 10>, paramNames [Random Ofs, Swirl Strength, WPO Side Intensity, WPO Wave Intensity]}
  n66 = VectorParameter() {group=WPO, parameterName=WPO Wave UV, sortPriority=6, defaultValue=<rgba 1 0.25 0 1>}
  n67 = Multiply(A=@n59, B=AppendVector(A=@n66.R, B=@n66.R))
  n68 = AppendVector(A=@n66.B, B=@n66.B)
  n69 = TextureSample(UVs=Add(A=AppendVector(A=@n65.Random Ofs, B=@n65.Random Ofs), B=Panner(Coordinate=@n67, Speed=@n68))) {texture=&T_Smooth_tile}
  n70 = TextureCoordinate()
  n71 = LinearInterpolate(
   B=AppendVector(A=Multiply(A=@n63, B=@n64.WPO Side Intensity) {constB=10}, B=Multiply(A=@n69.R, B=@n64.WPO Wave Intensity) {constB=10})
   Alpha=OneMinus(None=ComponentMask(None=@n70) {g=true})
  )
  n72 = StaticSwitchParameter(
   True=Add(
    A=MaterialFunctionCall(WidthBase=@n55, WidthTip=@n50.WPO Tip, Expand U or V UV Channel=@n56) {materialFunction=&SplineThicken}.WorldPositionOffset
    B=@n71
   )
   False=@n71
  ) {defaultValue=true, parameterName=USE Thicken}
  n73 = StaticSwitchParameter(True=Add(A=@n72, B=NamedRerouteUsage()), False=@n72) {parameterName=USE_Bending}
  StaticSwitchParameter(True=Add(A=NamedRerouteUsage(), B=@n73), False=@n73) {parameterName=Sp}
 ]
}
