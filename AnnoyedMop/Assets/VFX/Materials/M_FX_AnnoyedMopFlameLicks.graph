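# Translucent unlit Niagara sprite material for the AnnoyedMop flame-lick effect.
# Upward- or downward-scrolling noise distorts UVs into a clamped RGB flame texture and also erodes the resulting mask.
# Dynamic parameter MaskX feeds all three channel weights of the dot product; MaskY and MaskZ are unused.
# Inverted particle alpha drives a SmoothStep dissolve with width w, and particle color tints emissive through MF_FX_EmissiveAdjustments.
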Material /AnnoyedMop/Assets/VFX/Materials/M_FX_AnnoyedMopFlameLicks
graph {
 properties {bUsedWithNiagaraSprites=true, blendMode=BLEND_Translucent, pixelDepthOffsetMode=PDOM_Legacy, shadingModel=MSM_Unlit, usedShadingModels=Unlit}
 nodes [
  n1 = ParticleColor()
  MaterialFunctionCall(Emissive=Multiply(A=NamedRerouteUsage(), B=@n1.RGB)) {materialFunction=&MF_FX_EmissiveAdjustments}
  n2 = OneMinus(None=@n1.A)
  n3 = ScalarParameter() {parameterName=w}
  n4 = TextureSample(UVs=Panner(Coordinate=TextureCoordinate() {uTiling=0.75, vTiling=0.5}) {speedY=1}) {samplerType=SAMPLERTYPE_Masks, texture=/AnnoyedMop/Assets/VFX/Textures/T_FX_AnnoyedMop_Prj}
  n5 = Subtract(A=Multiply(A=@n4.R) {constB=2})
  n6 = Constant2Vector() {g=0.1, r=0.05}
  n7 = TextureSample(UVs=Add(A=TextureCoordinate(), B=Multiply(A=@n5, B=@n6))) {
   samplerSource=SSM_Clamp_WorldGroupSettings
   samplerType=SAMPLERTYPE_Masks
   texture=/AnnoyedMop/Assets/VFX/Textures/T_FX_AnnoyedMop_FlameLicksRGB
  }
  n8 = DynamicParameter() {defaultValue=<rgba 1 0 0 1>, paramNames [MaskX,MaskY,MaskZ,Param4]}
  n9 = Subtract(
   A=DotProduct(
    A=@n7.RGB
    B=MaterialFunctionCall(R=@n8.MaskX, G=@n8.MaskX, B=@n8.MaskX) {materialFunction=&AppendMany}.RGB
   )
   B=Multiply(A=@n4.G) {constB=0.35}
  )
  Multiply(A=SmoothStep(Min=@n2, Max=Add(A=@n2, B=@n3), Value=Saturate(None=@n9))) {constB=0.1}
  NamedRerouteDeclaration(None=@n4.R) {name=MainNoise, nodeColor=<rgba 0 0.2734 1 1>}
  NamedRerouteUsage()
  ScalarParameter() {parameterName=p}
 ]
}
