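# Translucent unlit material for Niagara mesh particles in the AnnoyedMop fan blast effect, drawing a scrolling two-color fire body with glow.
# Panning noise and a distorted fire tile texture drive the color, tinted by particle color.
# A Dynamic Parameter slides a gradient leading-edge reveal, with Fresnel and smoothstep masks shaping edges and depth fade on High and Epic quality.
# Panning texture along normals plus a view-space offset builds vertex displacement.
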Material /AnnoyedMop/Assets/VFX/Materials/M_FX_AnnoyedMop_FanBlast
graph {
 properties {
  bUseEmissiveForDynamicAreaLighting=true
  bUsedWithNiagaraMeshParticles=true
  blendMode=BLEND_Translucent
  pixelDepthOffsetMode=PDOM_Legacy
  shadingModel=MSM_Unlit
  usedShadingModels=Unlit
 }
 nodes [
  n1 = TextureCoordinate()
  n2 = Multiply(
   A=VertexNormalWS()
   B=TextureSample(UVs=Panner(Coordinate=@n1) {speedY=0.3}) {texture=/Game/Effects/Fort_Effects/Textures/Power_Source/T_Power_Source_Sphere_Texture}.G
  )
  n3 = ScalarParameter() {defaultValue=50, parameterName=Bump}
  n4 = VectorParameter() {parameterName=Screenspace_SourceOffset, defaultValue=<rgba 0 0 0 1>}
  n5 = TextureCoordinate()
  Add(
   A=Multiply(A=NamedRerouteUsage(), B=Multiply(A=@n2, B=@n3))
   B=Multiply(A=Transform(None=@n4.RGB) {transformSourceType=TRANSFORMSOURCE_View}, B=ComponentMask(None=@n5) {g=true})
  )
  DynamicParameter() {defaultValue=<rgba 0 1 1 1>, paramNames [Reveal,Param2,Param3,Param4]}
  n6 = VectorParameter() {parameterName=BodyColorA, defaultValue=<rgba 1 1 0 1>}
  n7 = TextureCoordinate() {uTiling=0.2, vTiling=0.5}
  n8 = VectorParameter() {parameterName=NoiseScale, defaultValue=<rgba 1 1 0 1>}
  n9 = ComponentMask(None=@n8.RGB) {g=true, r=true}
  n10 = TextureSample(UVs=Panner(Coordinate=Multiply(A=@n7, B=@n9), Speed=AppendVector(A=@n8.B, B=@n8.B)) {speedY=0.2}) {texture=/Game/Effects/Fort_Effects/Textures/Fire/HF_FireNoisesPacked}
  n11 = Add(A=@n10.R) {constB=0.25}
  n12 = VectorParameter() {parameterName=UV Scale, defaultValue=<rgba 0.5 0.5 0 1>}
  n13 = Multiply(A=Add(A=ParticleRandom(), B=TextureCoordinate() {uTiling=0.7}), B=ComponentMask(None=@n12.RGB) {g=true, r=true})
  n14 = AppendVector(A=@n12.B, B=@n12.B)
  n15 = NamedRerouteUsage()
  n16 = ComponentMask(None=VectorParameter() {parameterName=Distortion, defaultValue=<rgba 0 0 0 1>}.RGB) {g=true, r=true}
  n17 = TextureSample(UVs=Add(A=Panner(Coordinate=@n13, Speed=@n14) {speedY=0.2}, B=Multiply(A=@n15, B=@n16))) {samplerType=SAMPLERTYPE_Masks, texture=/AnnoyedMop/Assets/VFX/Textures/T_AnnoyedMop_FireTile}
  n18 = VectorParameter() {parameterName=GlowColor, defaultValue=<rgba 0 0 0 1>}
  n19 = Add(
   A=LinearInterpolate(A=Multiply(A=@n6.RGB, B=@n11), B=VectorParameter() {parameterName=BodyColorB, defaultValue=<rgba 1 0 0 1>}.RGB, Alpha=@n17.R)
   B=Multiply(A=Multiply(A=@n17.G, B=@n18.A), B=@n18.RGB)
  )
  n20 = ParticleColor()
  MaterialFunctionCall(Emissive=Multiply(A=ParticleColor().RGB, B=Multiply(A=@n19, B=@n20.RGB))) {materialFunction=&MF_FX_EmissiveAdjustments}
  NamedRerouteDeclaration(None=Fresnel() {baseReflectFraction=0, exponent=1}) {name=EdgeFresnel, nodeColor=<rgba 0 1 0.6016 1>}
  NamedRerouteDeclaration(None=Subtract(A=Multiply(A=@n10.R) {constB=2})) {name=MainNoise, nodeColor=<rgba 0 0.9062 1 1>}
  n21 = Add(A=NamedRerouteUsage(), B=ScalarParameter() {parameterName=Edge_Position})
  n22 = ScalarParameter() {defaultValue=-0.2, group=LeadingEdge, parameterName=Edge_Width}
  NamedRerouteDeclaration(None=SmoothStep(Min=@n21, Max=Add(A=@n21, B=@n22), Value=NamedRerouteUsage())) {name=Edge, nodeColor=<rgba 0.7109 1 0 1>}
  n23 = Constant()
  n24 = DynamicParameter() {defaultValue=<rgba 0 1 1 1>, paramNames [Reveal,Param2,Param3,Param4]}
  n25 = TextureSampleParameter2D(UVs=Add(A=TextureCoordinate() {vTiling=0.25}, B=AppendVector(A=@n23, B=@n24.Reveal))) {
   parameterName=T_LeadingEdgeMask
   samplerSource=SSM_Clamp_WorldGroupSettings
   texture=/Game/Effects/Fort_Effects/UI/Transition_IAS/Resources/T_Vert_Gradient_Soft
  }
  n26 = Multiply(A=@n25.R, B=VertexColor().R)
  n27 = NamedRerouteUsage()
  n28 = ScalarParameter() {parameterName=PatternOpacity}
  NamedRerouteDeclaration(None=Add(A=@n26, B=Multiply(A=@n27, B=@n28) {constB=0.125})) {name=LeadingEdge, nodeColor=<rgba 1 0 0.8906 1>}
  n29 = Add(A=NamedRerouteUsage(), B=ScalarParameter() {defaultValue=-0.1, group=LeadingEdge, parameterName=LeadingEdge_Position})
  n30 = ScalarParameter() {defaultValue=0.2, group=LeadingEdge, parameterName=LeadingEdge_Width}
  n31 = NamedRerouteUsage()
  NamedRerouteDeclaration(None=SmoothStep(Min=@n29, Max=Add(A=@n29, B=@n30), Value=Multiply(A=@n31, B=@n26))) {name=LeadingEdgeMask, nodeColor=<rgba 1 0.5625 0 1>}
  NamedRerouteDeclaration(None=@n17.R) {name=MainPattern, nodeColor=<rgba 0.0547 1 0 1>}
  n32 = Saturate(None=Multiply(A=@n20.A, B=NamedRerouteUsage()))
  n33 = DepthFade(Opacity=@n32, FadeDistance=ScalarParameter() {defaultValue=750, parameterName=DepthFadeDistance})
  QualitySwitch(Default=@n32, Low=@n32, High=@n33, Medium=@n32, Epic=@n33)
  TextureSample() {texture=/Game/Effects/Fort_Effects/Textures/Fire/T_FireNoise_02}
 ]
}
