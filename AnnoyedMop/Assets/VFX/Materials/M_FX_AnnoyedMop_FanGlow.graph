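# Unlit additive Niagara mesh particle material for the AnnoyedMop fan glow effect, with particle color fed through an emissive adjustment function.
# A Reveal dynamic parameter slides a soft gradient, masked by vertex color, into a smoothstep leading-edge mask that combines with particle alpha.
# Panning fire noise supplies a named noise signal.
# A vertex offset combines VertexPush along normals with a view-space source offset weighted by UV V.
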Material /AnnoyedMop/Assets/VFX/Materials/M_FX_AnnoyedMop_FanGlow
graph {
 properties {
  bUseEmissiveForDynamicAreaLighting=true
  bUsedWithNiagaraMeshParticles=true
  blendMode=BLEND_Additive
  pixelDepthOffsetMode=PDOM_Legacy
  shadingModel=MSM_Unlit
  usedShadingModels=Unlit
 }
 nodes [
  n1 = Multiply(A=VertexNormalWS(), B=DynamicParameter() {defaultValue=<rgba 0 1 1 1>, paramNames [Reveal,VertexPush,Param3,Param4]}.VertexPush)
  n2 = VectorParameter() {parameterName=Screenspace_SourceOffset, defaultValue=<rgba 0 0 0 1>}
  n3 = TextureCoordinate()
  Add(
   A=Multiply(A=NamedRerouteUsage(), B=Multiply(A=@n1))
   B=Multiply(A=Transform(None=@n2.RGB) {transformSourceType=TRANSFORMSOURCE_View}, B=ComponentMask(None=@n3) {g=true})
  )
  MaterialFunctionCall(Emissive=ParticleColor().RGB) {materialFunction=&MF_FX_EmissiveAdjustments}
  n4 = TextureCoordinate() {uTiling=0.2, vTiling=0.5}
  n5 = VectorParameter() {parameterName=NoiseScale, defaultValue=<rgba 1 1 0 1>}
  n6 = ComponentMask(None=@n5.RGB) {g=true, r=true}
  n7 = TextureSample(UVs=Panner(Coordinate=Multiply(A=@n4, B=@n6), Speed=AppendVector(A=@n5.B, B=@n5.B)) {speedY=0.2}) {texture=/Game/Effects/Fort_Effects/Textures/Fire/HF_FireNoisesPacked}
  NamedRerouteDeclaration(None=Subtract(A=Multiply(A=@n7.R) {constB=2})) {name=MainNoise, nodeColor=<rgba 0 0.9062 1 1>}
  n8 = TextureCoordinate() {vTiling=0.1}
  n9 = AppendVector(A=Constant(), B=DynamicParameter() {defaultValue=<rgba 0 1 1 1>, paramNames [Reveal,VertexPush,Param3,Param4]}.Reveal)
  n10 = Multiply(
   A=TextureSample(UVs=Add(A=@n8, B=@n9)) {samplerSource=SSM_Clamp_WorldGroupSettings, texture=/Game/Effects/Fort_Effects/UI/Transition_IAS/Resources/T_Vert_Gradient_Soft}.R
   B=VertexColor().R
  )
  NamedRerouteDeclaration(None=@n10) {name=LeadingEdge, nodeColor=<rgba 1 0 0.8906 1>}
  n11 = Add(A=NamedRerouteUsage(), B=ScalarParameter() {defaultValue=-0.1, group=LeadingEdge, parameterName=LeadingEdge_Position})
  n12 = ScalarParameter() {defaultValue=0.2, group=LeadingEdge, parameterName=LeadingEdge_Width}
  n13 = NamedRerouteUsage()
  NamedRerouteDeclaration(None=SmoothStep(Min=@n11, Max=Add(A=@n11, B=@n12), Value=Multiply(A=@n13, B=@n10))) {name=LeadingEdgeMask, nodeColor=<rgba 1 0.5625 0 1>}
  Saturate(None=Multiply(A=ParticleColor().A, B=NamedRerouteUsage()))
 ]
}
