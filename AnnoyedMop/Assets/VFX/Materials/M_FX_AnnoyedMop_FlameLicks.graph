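# Unlit translucent Niagara sprite material that makes scrolling flame licks for the AnnoyedMop effects.
# A tiled mask texture pans vertically by the Speed parameter, with a per-particle random UV offset to vary sprites.
# Its green channel is sharpened and confined by a soft radial mask.
# The red channel, scaled by a dynamic parameter, is subtracted to erode the shape, and particle alpha scales the result.
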Material /AnnoyedMop/Assets/VFX/Materials/M_FX_AnnoyedMop_FlameLicks
graph {
 properties {bUsedWithNiagaraSprites=true, blendMode=BLEND_Translucent, pixelDepthOffsetMode=PDOM_Legacy, shadingModel=MSM_Unlit, usedShadingModels=Unlit}
 nodes [
  MaterialFunctionCall() {materialFunction=&Swizzle}
  n1 = Multiply(A=TextureCoordinate() {uTiling=0.35}, B=ScalarParameter() {defaultValue=1, parameterName=Tiling})
  n2 = ParticleRandom()
  n3 = Constant()
  n4 = ScalarParameter() {defaultValue=1, parameterName=Speed}
  n5 = TextureSample(UVs=Panner(Coordinate=Add(A=@n1, B=@n2), Speed=AppendVector(A=@n3, B=@n4)) {speedY=1}) {samplerType=SAMPLERTYPE_Masks, texture=/AnnoyedMop/Assets/VFX/Textures/T_FX_AnnoyedMop_Prj}
  n6 = Power(Base=@n5.G) {constExponent=1.5}
  n7 = SphereMask(A=TextureCoordinate(), B=Constant() {r=0.5}) {attenuationRadius=0.5, hardnessPercent=0}
  n8 = DynamicParameter()
  Multiply(A=ParticleColor().A, B=Subtract(A=Multiply(A=@n6, B=@n7), B=Multiply(A=@n5.R, B=@n8.R)))
  Time()
 ]
}
