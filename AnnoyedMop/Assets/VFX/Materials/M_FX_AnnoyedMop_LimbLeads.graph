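# Translucent unlit Niagara mesh-particle material for AnnoyedMop limb lead effects, built on a vertically panning projectile texture.
# The texture's green channel and UV height blend header color, particle color and highlights through an emissive adjustment function.
# Panning blurred noise scales vertex normals by UV height and a Wobble dynamic parameter to form a displacement.
# A SmoothStep on the red channel, with bounds from Fresnel, inverted UV height and a width parameter, forms a mask.
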Material /AnnoyedMop/Assets/VFX/Materials/M_FX_AnnoyedMop_LimbLeads
graph {
 properties {
  bUsedWithNiagaraMeshParticles=true
  blendMode=BLEND_Translucent
  pixelDepthOffsetMode=PDOM_Legacy
  previewMesh=/Game/Effects/Fort_Effects/Mesh/Shockwave/SM_Shockwave_Partial_Sphere
  shadingModel=MSM_Unlit
  usedShadingModels=Unlit
 }
 nodes [
  Add()
  n1 = VectorParameter() {parameterName=HeaderColor, defaultValue=<rgba 0 0 0 1>}
  n2 = ParticleColor()
  n3 = TextureCoordinate()
  n4 = ComponentMask(None=@n3) {g=true}
  n5 = TextureSample(UVs=Panner(Coordinate=@n3) {speedY=-2}) {samplerType=SAMPLERTYPE_Masks, texture=/AnnoyedMop/Assets/VFX/Textures/T_FX_AnnoyedMop_Prj}
  n6 = Add(A=@n5.G, B=@n5.G)
  n7 = Saturate(None=Add(A=@n4, B=Multiply(A=@n6) {constB=0.5}))
  MaterialFunctionCall(Emissive=LinearInterpolate(
    A=LinearInterpolate(A=@n1.RGB, B=@n2.RGB, Alpha=@n7)
    B=VectorParameter() {parameterName=Highlights, defaultValue=<rgba 0 0 0 1>}.RGB
    Alpha=@n5.G
   )) {materialFunction=&MF_FX_EmissiveAdjustments}
  n8 = VertexNormalWS()
  n9 = TextureSample(UVs=Panner(Coordinate=TextureCoordinate() {vTiling=0.5}) {speedY=-2}) {samplerType=SAMPLERTYPE_Grayscale, texture=&LowResBlurredNoise}
  n10 = Add(A=Subtract(A=Multiply(A=@n9.R) {constB=2})) {constB=0.5}
  n11 = TextureCoordinate()
  Multiply(
   A=DynamicParameter() {paramNames [Wobble,Param2,Param3,Param4]}.Wobble
   B=Multiply(A=Multiply(A=@n8, B=@n10), B=ComponentMask(None=@n11) {g=true})
  )
  n12 = MaterialFunctionCall(Exponent=ScalarParameter() {defaultValue=1, parameterName=FresnelExponent}) {materialFunction=&MF_Fresnel_FX}
  n13 = OneMinus(None=@n4)
  n14 = ScalarParameter() {defaultValue=0.1, parameterName=w}
  SmoothStep(Min=Add(A=Multiply(A=@n12.Fresnel, B=@n13), B=@n14), Max=@n14, Value=@n5.R)
 ]
}
