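# Translucent unlit Niagara mesh-particle material for the AnnoyedMop LrgPrj effect, apparently a fiery large projectile.
# A panning, distorted fire-tile texture, offset per particle, blends two noise-modulated body colors and adds a glow, tinted by particle color.
# The dynamic Reveal parameter scrolls a trail-head mask, whose smoothstepped leading edge combines with a spherical edge mask and particle alpha for opacity.
# The graph also builds noise-driven normal bump, SplineThicken, Fresnel edge and cartoon-flame caustics terms.
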Material /AnnoyedMop/Assets/VFX/Materials/M_FX_AnnoyedMop_LrgPrj
graph {
 properties {
  bUseEmissiveForDynamicAreaLighting=true
  bUsedWithNiagaraMeshParticles=true
  blendMode=BLEND_Translucent
  pixelDepthOffsetMode=PDOM_Legacy
  shadingModel=MSM_Unlit
  usedShadingModels=Unlit
 }
 nodes [
  Add()
  DynamicParameter() {defaultValue=<rgba 0 1 1 1>, paramNames [Reveal,Param2,Param3,Param4]}
  n1 = ScalarParameter() {parameterName=Thickness}
  MaterialFunctionCall(WidthBase=Constant(), WidthTip=@n1) {materialFunction=&SplineThicken}
  n2 = Multiply(A=TextureCoordinate() {uTiling=0.2, vTiling=0.5}, B=ScalarParameter() {parameterName=NoiseScale})
  n3 = NamedRerouteDeclaration(None=TextureSample(UVs=Panner(Coordinate=@n2) {speedY=0.2}) {texture=/Game/Effects/Fort_Effects/Textures/Fire/HF_FireNoisesPacked}.R) {name=MainNoise, nodeColor=<rgba 0 0.9062 1 1>}
  n4 = VectorParameter() {parameterName=UV Scale, defaultValue=<rgba 0.5 0.5 0 1>}
  n5 = Multiply(A=Add(A=ParticleRandom(), B=TextureCoordinate() {uTiling=0.7}), B=ComponentMask(None=@n4.RGB) {g=true, r=true})
  n6 = AppendVector(A=@n4.B, B=@n4.B)
  n7 = NamedRerouteUsage()
  n8 = ComponentMask(None=VectorParameter() {parameterName=Distortion, defaultValue=<rgba 0 0 0 1>}.RGB) {g=true, r=true}
  n9 = TextureSample(UVs=Add(A=Panner(Coordinate=@n5, Speed=@n6) {speedY=0.2}, B=Multiply(A=@n7, B=@n8))) {samplerType=SAMPLERTYPE_Masks, texture=/AnnoyedMop/Assets/VFX/Textures/T_AnnoyedMop_FireTile}
  n10 = LinearInterpolate(
   A=Multiply(A=VectorParameter() {parameterName=BodyColorA, defaultValue=<rgba 1 1 0 1>}.RGB, B=Add(A=@n3) {constB=0.25})
   B=VectorParameter() {parameterName=BodyColorB, defaultValue=<rgba 1 0 0 1>}.RGB
   Alpha=@n9.R
  )
  n11 = VectorParameter() {parameterName=GlowColor, defaultValue=<rgba 0 0 0 1>}
  n12 = Multiply(A=Multiply(A=@n9.G, B=@n11.A), B=@n11.RGB)
  n13 = ParticleColor()
  MaterialFunctionCall(Emissive=Multiply(A=Add(A=@n10, B=@n12), B=@n13.RGB)) {materialFunction=&MF_FX_EmissiveAdjustments}
  n14 = TextureSample(UVs=Panner(Coordinate=TextureCoordinate() {uTiling=2, vTiling=0.5}) {speedX=0.1, speedY=-1}) {texture=/Game/Effects/Fort_Effects/Textures/Tile/T_Tile_Texture_Cartoon_Flames_Layer}
  n15 = TextureSample(UVs=Panner(Coordinate=TextureCoordinate()) {speedX=-0.1, speedY=-3}) {texture=&T_Tile_Texture_Cartoon_Flames_Layer}
  Min(A=@n14.B, B=@n15.B)
  n16 = Panner(Coordinate=TextureCoordinate() {uTiling=0.2, vTiling=0.5}) {speedY=0.3}
  Multiply(
   A=Multiply(A=VertexNormalWS(), B=TextureSample(UVs=@n16) {texture=&HF_FireNoisesPacked}.G)
   B=ScalarParameter() {defaultValue=50, parameterName=Bump}
  )
  n17 = Constant()
  n18 = DynamicParameter() {defaultValue=<rgba 0 1 1 1>, paramNames [Reveal,Param2,Param3,Param4]}
  n19 = TextureSample(UVs=Add(A=TextureCoordinate(), B=AppendVector(A=@n17, B=@n18.Reveal))) {samplerSource=SSM_Clamp_WorldGroupSettings, samplerType=SAMPLERTYPE_Masks, texture=/AnnoyedMop/Assets/VFX/Textures/T_AnnoyedMop_TrailHead}
  Multiply(A=@n19.G, B=@n1)
  NamedRerouteDeclaration(None=Fresnel() {baseReflectFraction=0, exponent=1}) {name=EdgeFresnel, nodeColor=<rgba 0 1 0.6016 1>}
  n20 = Add(A=NamedRerouteUsage(), B=ScalarParameter() {parameterName=Edge_Position})
  n21 = ScalarParameter() {defaultValue=-0.2, group=LeadingEdge, parameterName=Edge_Width}
  NamedRerouteDeclaration(None=SmoothStep(Min=@n20, Max=Add(A=@n20, B=@n21), Value=NamedRerouteUsage())) {name=Edge, nodeColor=<rgba 0.7109 1 0 1>}
  NamedRerouteDeclaration(None=@n19.G) {name=LeadingEdge, nodeColor=<rgba 1 0 0.8906 1>}
  n22 = Max(A=@n14.B, B=@n15.B)
  n23 = ScalarParameter() {parameterName=rrr}
  NamedRerouteDeclaration(None=Saturate(None=Power(Base=@n22, Exp=@n23))) {name=Caustics, nodeColor=<rgba 1 0 0.8437 1>}
  NamedRerouteUsage()
  NamedRerouteUsage()
  NamedRerouteUsage()
  n24 = Add(A=NamedRerouteUsage(), B=ScalarParameter() {defaultValue=-0.1, group=LeadingEdge, parameterName=LeadingEdge_Position})
  n25 = ScalarParameter() {defaultValue=0.2, group=LeadingEdge, parameterName=LeadingEdge_Width}
  n26 = NamedRerouteDeclaration(None=SmoothStep(Min=@n24, Max=Add(A=@n24, B=@n25), Value=NamedRerouteUsage())) {name=LeadingEdgeMask, nodeColor=<rgba 1 0.5625 0 1>}
  n27 = ComponentMask(None=TextureCoordinate()) {g=true, r=true}
  n28 = Constant() {r=0.5}
  n29 = Add(
   A=OneMinus(None=SphereMask(A=@n27, B=@n28) {attenuationRadius=0.5, hardnessPercent=60})
   B=ScalarParameter() {parameterName=Opacity_LeadEdge_Position}
  )
  n30 = ScalarParameter() {parameterName=Opacity_LeadEdge_W}
  n31 = NamedRerouteDeclaration(None=SmoothStep(Min=@n29, Max=Add(A=@n29, B=@n30), Value=NamedRerouteUsage())) {name=OpacityEdge, nodeColor=<rgba 0.3828 1 0 1>}
  Saturate(None=Multiply(A=@n13.A, B=Multiply(A=@n26, B=@n31)))
  TextureSample() {texture=/Game/Effects/Fort_Effects/Textures/Fire/T_FireNoise_02}
  VertexColor()
 ]
}
