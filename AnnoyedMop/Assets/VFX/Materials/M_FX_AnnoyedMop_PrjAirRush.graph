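# Translucent, unlit, two-sided material for the air rush effect of the AnnoyedMop projectile, used on Niagara mesh particles.
# A randomly offset panning mask, scaled by a dynamic parameter, drives a core-to-ends color hotspot along V into emissive adjustments.
# The mask is smoothstepped against an erosion threshold built from particle alpha, a center mask and vertex red.
# Panning noise along vertex normals creates a wobble offset, weighted by the inverse of that center-and-erosion term.
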Material /AnnoyedMop/Assets/VFX/Materials/M_FX_AnnoyedMop_PrjAirRush
graph {
 properties {
  bUsedWithNiagaraMeshParticles=true
  blendMode=BLEND_Translucent
  pixelDepthOffsetMode=PDOM_Legacy
  shadingModel=MSM_Unlit
  twoSided=true
  usedShadingModels=Unlit
 }
 nodes [
  n1 = ComponentMask(None=TextureCoordinate()) {g=true}
  n2 = Constant() {r=0.5}
  n3 = AppendVector(A=DynamicParameter() {paramNames [Scale,Param2,Param3,Param4]}.Scale, B=Constant() {r=1})
  n4 = TextureCoordinate() {vTiling=2}
  n5 = Time()
  n6 = ParticleRandom()
  n7 = TextureSample(UVs=Panner(Coordinate=Multiply(A=@n3, B=@n4), Time=Add(A=@n5, B=@n6)) {speedY=2}) {samplerType=SAMPLERTYPE_Masks, texture=/AnnoyedMop/Assets/VFX/Textures/T_FX_AnnoyedMop_Prj}
  n8 = Multiply(A=ScalarParameter() {parameterName=HotspotRadius}, B=@n7.R)
  MaterialFunctionCall(Emissive=LinearInterpolate(
    A=VectorParameter() {parameterName=EndsColor, defaultValue=<rgba 0 0 0 1>}.RGB
    B=VectorParameter() {parameterName=CoreColor, defaultValue=<rgba 0 0 0 1>}.RGB
    Alpha=SphereMask(A=@n1, B=@n2, Radius=@n8) {attenuationRadius=0.1, hardnessPercent=0}
   )) {materialFunction=&MF_FX_EmissiveAdjustments}
  n9 = TextureCoordinate()
  n10 = Add(
   A=LinearInterpolate(Alpha=NamedRerouteUsage()) {constA=-1, constB=0}
   B=SphereMask(A=ComponentMask(None=@n9) {g=true}, B=Constant() {r=0.5}) {attenuationRadius=0.5, hardnessPercent=0}
  )
  n11 = TextureCoordinate() {uTiling=1.25, vTiling=0.5}
  n12 = Add(A=Time(), B=ParticleRandom())
  n13 = Multiply(
   A=VertexNormalWS()
   B=TextureSample(UVs=Panner(Coordinate=@n11, Time=@n12) {speedY=-1}) {texture=&T_NoisePack_01}.R
  )
  n14 = ScalarParameter() {parameterName=Wobble}
  Multiply(A=OneMinus(None=@n10), B=Multiply(A=Multiply(A=@n13, B=@n14), B=NamedRerouteUsage()))
  n15 = Multiply(A=VertexColor().R, B=@n10)
  n16 = ScalarParameter() {parameterName=w}
  Multiply(A=SmoothStep(Min=@n15, Max=Add(A=@n15, B=@n16), Value=@n7.R), B=@n7.G)
  NamedRerouteDeclaration(None=ParticleColor().A) {name=ErodeT, nodeColor=<rgba 0 1 0.7891 1>}
  ParticleColor()
  ScalarParameter() {parameterName=Erode_p}
 ]
}
