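# Unlit translucent Niagara mesh-particle material for AnnoyedMop effects, apparently projectile edge streaks.
# SplineThicken sets width from a dynamic parameter, adding a time-driven, per-particle sine wobble along local X, scaled along U.
# A panning trail texture distorts a fire texture, which modulates particle-colored emissive, brightened toward the U=0 end.
# That pattern drives opacity through smoothstep erosion by particle alpha and vertex red, faded by double-sided Fresnel using the local X axis.
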Material /AnnoyedMop/Assets/VFX/Materials/M_FX_AnnoyedMop_PrjEdges
graph {
 properties {
  bUsedWithNiagaraMeshParticles=true
  blendMode=BLEND_Translucent
  pixelDepthOffsetMode=PDOM_Legacy
  shadingModel=MSM_Unlit
  twoSided=true
  usedShadingModels=Unlit
 }
 nodes [
  n1 = DynamicParameter() {defaultValue=<rgba 50 1 1 1>, paramNames [SplineThickness, Param2, Param3, Param4]}
  n2 = TextureCoordinate()
  n3 = Add(A=ParticleRandom(), B=Multiply(A=Time()) {constB=0.5})
  n4 = ComponentMask(None=TextureCoordinate() {uTiling=0.5}) {r=true}
  n5 = Multiply(A=Sine(None=Add(A=@n3, B=@n4))) {constB=16}
  n6 = Transform(None=Constant3Vector() {constant=<rgba 1 0 0 0>}) {transformSourceType=TRANSFORMSOURCE_Local}
  Add(
   A=MaterialFunctionCall(WidthBase=@n1.SplineThickness, WidthTip=@n1.SplineThickness, Expand U or V UV Channel=StaticBool()) {materialFunction=&SplineThicken}.WorldPositionOffset
   B=Multiply(A=ComponentMask(None=@n2) {r=true}, B=Multiply(A=@n5, B=@n6))
  )
  n7 = ParticleColor()
  n8 = Multiply(A=@n7.RGB) {constB=25}
  n9 = TextureCoordinate()
  n10 = Saturate(None=Power(Base=ComponentMask(None=@n9) {r=true}) {constExponent=0.05})
  n11 = TextureCoordinate() {uTiling=1.3}
  n12 = ParticleRandom()
  n13 = Constant2Vector() {r=1}
  n14 = Panner(Coordinate=Add(A=@n11, B=Multiply(A=@n12, B=@n13))) {speedX=-2}
  n15 = TextureSample(UVs=@n14) {texture=/AnnoyedMop/Assets/VFX/Textures/T_ColdDay_Fire_Trail00B}
  n16 = Constant2Vector() {r=0.1}
  n17 = TextureSample(UVs=Add(A=@n14, B=Multiply(A=@n15.R, B=@n16))) {texture=/Game/Effects/Fort_Effects/Textures/T_PlantStand_Fire}
  n18 = Divide(A=Add(A=Add(A=@n17.R, B=@n17.R), B=@n17.B)) {constB=3}
  MaterialFunctionCall(Emissive=Multiply(A=LinearInterpolate(A=@n8, B=@n7.RGB, Alpha=@n10), B=@n18)) {materialFunction=&MF_FX_EmissiveAdjustments}
  Multiply(A=@n11)
  n19 = Add(A=OneMinus(None=@n7.A), B=OneMinus(None=VertexColor().R))
  n20 = ScalarParameter() {parameterName=Erode_W}
  n21 = ScalarParameter() {defaultValue=1, parameterName=FadeBrightness}
  n22 = ScalarParameter() {defaultValue=3, parameterName=FadeExponent}
  n23 = Constant3Vector() {constant=<rgba 1 0 0 0>}
  n24 = VertexInterpolator(VS=Normalize(VectorInput=Transform(None=@n23) {transformSourceType=TRANSFORMSOURCE_Local}))
  Multiply(
   A=SmoothStep(Min=@n19, Max=Add(A=@n19, B=@n20), Value=@n18)
   B=Saturate(None=MaterialFunctionCall(Brightness=@n21, Exponent=@n22, Vertex / Pixel Normal=@n24.PS) {materialFunction=&MF_Fresnel_FX}.Fresnel Double Sided)
  )
  ScalarParameter() {parameterName=erode_p}
 ]
}
