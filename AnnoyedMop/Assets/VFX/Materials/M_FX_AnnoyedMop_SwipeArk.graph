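# Unlit, two-sided translucent material for the AnnoyedMop swipe arc VFX, flagged for Niagara mesh particles.
# A panning dissolve texture distorts scrolling trail and fire textures.
# A curve-atlas gradient offset by fire feeds tinted emissive, alongside a layered backing and highlight color blend.
# Dynamic parameters drive SplineThicken width and an erosion mask combining a sphere mask with a U gradient.
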Material /AnnoyedMop/Assets/VFX/Materials/M_FX_AnnoyedMop_SwipeArk
graph {
 properties {
  bUsedWithNiagaraMeshParticles=true
  blendMode=BLEND_Translucent
  pixelDepthOffsetMode=PDOM_Legacy
  shadingModel=MSM_Unlit
  twoSided=true
  usedShadingModels=Unlit
 }
 nodes [
  n1 = TextureSample(UVs=Panner(Coordinate=TextureCoordinate() {uTiling=1.5}) {speedX=1.5}) {texture=/AnnoyedMop/Assets/VFX/Textures/T_ColdDay_Fire_Trail_Dissolve}
  n2 = Subtract(A=Multiply(A=@n1.R) {constB=2})
  n3 = ScalarParameter() {parameterName=Distortion}
  n4 = Panner(Coordinate=Add(A=TextureCoordinate() {uTiling=1.5}, B=Multiply(A=@n2, B=@n3) {constB=0.05})) {speedX=1}
  n5 = TextureSample(UVs=@n4) {texture=/AnnoyedMop/Assets/VFX/Textures/T_ColdDay_Fire_Trail00C}
  n6 = NamedRerouteUsage()
  n7 = Saturate(None=Subtract(A=@n5.RGB, B=@n6))
  Add(
   A=LinearInterpolate(
    A=VectorParameter() {parameterName=BackingColor, defaultValue=<rgba 0 0 0 1>}.RGB
    B=VectorParameter() {parameterName=HighlightsColor, defaultValue=<rgba 0 0 0 1>}.RGB
    Alpha=NamedRerouteUsage()
   )
   B=Multiply(A=ComponentMask(None=@n7) {b=true}, B=VectorParameter() {parameterName=SuperHighlights, defaultValue=<rgba 0 0 0 1>}.RGB)
  )
  ComponentMask(None=@n7) {r=true}
  n8 = DynamicParameter() {defaultValue=<rgba 32 1 1 1>, paramNames [Thickness,Erode,Param3,Param4]}
  MaterialFunctionCall(WidthBase=@n8.Thickness, WidthTip=@n8.Thickness, Expand U or V UV Channel=StaticBool()) {materialFunction=&SplineThicken}
  n9 = TextureCoordinate()
  n10 = TextureSample(UVs=@n4) {texture=/Game/Effects/Fort_Effects/Textures/T_PlantStand_Fire}
  n11 = Divide(A=Add(A=Add(A=@n10.R, B=@n10.R), B=@n10.B)) {constB=3}
  n12 = CurveAtlasRowParameter(CurveTime=Subtract(A=ComponentMask(None=@n9) {r=true}, B=Multiply(A=@n11) {constB=0.5})) {
   atlas=/AnnoyedMop/Assets/VFX/Textures/CA_AnnoyedMopGradients
   curve=/AnnoyedMop/Assets/VFX/Textures/Curve_AnnoyedMopColors1
   parameterName=ColorGradient
  }
  MaterialFunctionCall(Emissive=Multiply(A=@n12, B=VectorParameter() {parameterName=Tint, defaultValue=<rgba 0 0 0 1>}.RGB)) {materialFunction=&MF_FX_EmissiveAdjustments}
  n13 = SphereMask(A=ComponentMask(None=TextureCoordinate()) {g=true, r=true}, B=Constant() {r=0.5}) {attenuationRadius=0.6, hardnessPercent=50}
  n14 = Add(
   A=ComponentMask(None=TextureCoordinate() {uTiling=2}) {r=true}
   B=DynamicParameter() {defaultValue=<rgba 32 1 1 1>, paramNames [Thickness,Erode,Param3,Param4]}.Erode
  )
  NamedRerouteDeclaration(None=Add(A=OneMinus(None=@n13), B=Saturate(None=@n14))) {name=ErodeMask, nodeColor=<rgba 1 0.7266 0 1>}
  NamedRerouteDeclaration(None=ComponentMask(None=@n7) {g=true}) {name=BackingLayer, nodeColor=<rgba 0.0078 0 1 1>}
  NamedRerouteDeclaration(None=@n10.B) {name=HighlightsLayer, nodeColor=<rgba 0.6172 1 0 1>}
  NamedRerouteDeclaration(None=@n5.R) {name=TrailGradient, nodeColor=<rgba 0.0312 1 0 1>}
  NamedRerouteUsage()
  OneMinus(None=@n6)
  SmoothStep(Value=Subtract(A=@n1.R, B=@n6))
 ]
}
