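# Translucent unlit smoke material for Niagara mesh particles, apparently a prototype attack effect.
# Color blends two tints along a powered UV.G gradient, multiplied by particle color.
# A panning ribbon mask with noise-distorted UVs is multiplied by a hardened texture dissolve and particle alpha to form opacity.
# Dynamic parameters drive dissolve, noise strength and SplineThicken width, letting particles animate breakup and thickness.
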Material /AnnoyedMop/Assets/VFX/Materials/M_ProtoAttackSmoke
graph {
 properties {
  bUsedWithNiagaraMeshParticles=true
  blendMode=BLEND_Translucent
  pixelDepthOffsetMode=PDOM_Legacy
  shadingModel=MSM_Unlit
  usedShadingModels=Unlit
 }
 nodes [
  n1 = VectorParameter() {parameterName=Color, defaultValue=<rgba 0.7 0.7 0.7 1>}
  n2 = MaterialFunctionCall(Float2=TextureCoordinate()) {materialFunction=&BreakOutFloat2Components}
  n3 = LinearInterpolate(
   A=VectorParameter() {parameterName=Color2, defaultValue=<rgba 0.2 0.2 0.2 1>}.RGB
   B=@n1.RGB
   Alpha=Power(Base=@n2.G, Exp=ScalarParameter() {defaultValue=4})
  )
  n4 = ParticleColor()
  MaterialFunctionCall(Emissive=Divide(A=Multiply(A=@n3, B=@n4.RGB))) {materialFunction=&MF_FX_EmissiveAdjustments}
  n5 = Multiply(
   A=DynamicParameter() {defaultValue=<rgba 1 0 1 1>, paramNames [Dissolve,Noise,width,Param4]}.width
   B=ScalarParameter() {defaultValue=50, parameterName=Width}
  )
  MaterialFunctionCall(WidthBase=@n5, WidthTip=@n5, Expand U or V UV Channel=StaticBool()) {materialFunction=&SplineThicken}
  Multiply(A=@n3, B=@n1.RGB)
  n6 = TextureObjectParameter() {group=Dissolve, parameterName=Dissolve_Texture, sortPriority=0, texture=/Game/Effects/Fort_Effects/Emotes/Candor/T_Noise_Blur_Texture}
  n7 = VectorParameter() {group=Dissolve, parameterName=Dissolve_UVControl, channelNames {a=VPan, b=UPan, g=VScale, r=UScale}, defaultValue=<rgba 1 2 0 0>}
  n8 = DynamicParameter() {defaultValue=<rgba 1 0 1 1>, paramNames [Dissolve,Noise,width,Param4]}
  n9 = ScalarParameter() {defaultValue=-1.056, parameterName=DissolveTest}
  n10 = Multiply(
   A=Add(
    A=MaterialFunctionCall(Texture=@n6, UVControl_RGB=@n7, UVControl_A=@n7) {materialFunction=&FXCommon_TextureSample}.R
    B=Add(A=@n8.Dissolve, B=@n9)
   )
   B=ScalarParameter() {defaultValue=30, parameterName=DissolveHardness}
  )
  n11 = TextureObjectParameter() {group=MainRibbon, parameterName=MainRibbon_Texture, sortPriority=0, texture=&/T_RibbonMask_B2}
  n12 = VectorParameter() {group=MainRibbon, parameterName=MainRibbon_UVControl, channelNames {a=VPan, b=UPan, g=VScale, r=UScale}, defaultValue=<rgba 2 1 -0.2 0>}
  n13 = TextureObjectParameter() {group=NoiseA, parameterName=NoiseA_Texture, sortPriority=0, texture=/Game/Effects/Fort_Effects/Emotes/UltraLight/Texture/T_SoftNoise_BC7}
  n14 = VectorParameter() {group=NoiseA, parameterName=NoiseA_UVControl, channelNames {a=VPan, b=UPan, g=VScale, r=UScale}, defaultValue=<rgba 0 0.2 0 0>}
  n15 = MaterialFunctionCall(Texture=@n13, UVControl_RGB=@n14, UVControl_A=@n14) {materialFunction=&FXCommon_TextureSample}
  n16 = VectorParameter() {
   group=NoiseA
   parameterName=NoiseA_NoiseControl
   sortPriority=60
   channelNames {a=VStrength, b=UStrength, g=VOffset, r=UOffset}
   defaultValue=<rgba -0.5 0 0 0>
  }
  n17 = Constant()
  n18 = Add(
   A=@n16
   B=MaterialFunctionCall(X=@n17, Y=@n17, Z=DynamicParameter() {defaultValue=<rgba 1 0 1 1>, paramNames [Dissolve,Noise,width,Param4]}.Noise) {materialFunction=&MakeFloat3}.Result
  )
  n19 = MaterialFunctionCall(
   Texture=@n11
   UVControl_RGB=@n12
   UVControl_A=@n12
   "[Override]UVInput"=Add(
    A=TextureCoordinate()
    B=MaterialFunctionCall!(NoiseValue=@n15.RG, Control_RGB=@n18, Control_A=@n16.VStrength) {materialFunction=/Game/Effects/Niagara/MasterMaterials/Functions/Common/FXCommon_NoiseControl}.NoiseValue
   )
  ) {materialFunction=&FXCommon_TextureSample}
  Multiply(A=Multiply(A=Saturate(None=@n10), B=@n19.R), B=@n4.A)
  OneMinus(None=@n2.G)
  n20 = VectorParameter() {parameterName=Refine, defaultValue=<rgba 0.2 1 1.5 0.5>}
  Saturate(None=MaterialFunctionCall(Texture=@n19.R, RGB=@n20.RGB, A=@n20.RGB) {materialFunction=&MF_FX_Refine}.Result)
 ]
}
