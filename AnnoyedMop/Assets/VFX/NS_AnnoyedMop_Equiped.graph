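# Niagara system for the fire effect on the equipped AnnoyedMop, placed and shaped by user Offset and Height parameters.
# Fire_Main emits rising flame sprites with a light, Fire_Mesh_Root001 adds a mesh flame base, and Fire_Root adds base sprites, while Fire_Main_Inherit is disabled.
# Owner speed is remapped to shrink and weaken flames while moving.
# Curve-driven dynamic material parameters animate the flame shading over particle life.
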NiagaraSystem /AnnoyedMop/Assets/VFX/NS_AnnoyedMop_Equiped
dependencies {
 usedDataInterfaces [
  &NiagaraDataInterfaceCurve
  &NiagaraDataInterfaceArrayFloat
  &NiagaraDataInterfaceColorCurve
  &NiagaraDataInterfacePlatformSet
 ]
 usedDynamicInputs [&FN_PickFloatDependOnPlatform]
 usedModules [
  &self:SystemSpawnScript.SystemScriptSource.SystemScriptGraph.NiagaraNodeAssignment_0.SetVariables_#1
  &self:SystemSpawnScript.SystemScriptSource.SystemScriptGraph.NiagaraNodeAssignment_1.SetVariables_#2
  &SystemState
 ]
}
emitters {
 Fire_Main {
  emitterUpdateScript [
   &EmitterState(Delay First Loop Only=true, Life Cycle Mode=::NewEnumerator1, Loop Delay=0.6, Loop Duration=1, UseLoopDelay=true)
   &SpawnRate(SpawnRate=20, Use Spawn Probability=false)
   /SetVariables(Emitter.Light Color:LinearColor=<rgba 20 6.7502 0 1>, Emitter.Light Pos:NiagaraPosition=$User.Offset, Emitter.Light Radius=20) {isSetParameters=true}
  ]
  particleSpawnScript [
   &InitializeParticle(
    Color=<rgba 1 0.8244 0.8142 4>
    Lifetime=&Multiply_Float(
     A=&RandomRangeFloat(Minimum=0.8, Maximum=1)
     B=&MakeCustomFloatFromBool(Bool=&SetBoolByFloatComparison(A=1, B=0), True Float=1, False Float=1)
    )
    Position Offset=&Multiply_VectorByFloat(
     Vector=Linked() {linkedVariable {name=User.Offset, type=NiagaraPosition}}
     Float=&Multiply_Float(
      A=&RemapRange(
       Input Value To Remap=&VectorLength(Vector=$Engine.Owner.Velocity)
       Input Min=300
       Input Max=0
       Output Min=0.85
       Output Max=1
      )
      B=&RandomRangeFloat(Minimum=1, Maximum=1)
     )
    )
    Position Offset Coordinate Space=::Local
    Sprite Rotation Angle=&RandomRangeFloat(Minimum=0, Maximum=1)
    Sprite Rotation Mode=::NewEnumerator1
    Sprite Size Mode=::NewEnumerator1
    Sprite UV Mode=::NewEnumerator3
    Uniform Sprite Size Max=14
    Uniform Sprite Size Min=11
    UsePositionOffset=true
    Write Material Random=true
   ) {module=InitializeParticle001}
   /SetVariables(Particles.Color2=<rgba 4 0.7172 0 0.2>) {isSetParameters=true}
   &ShapeLocation(
    Handle Radius=0
    Large Radius=&Multiply_Float(
     A=&RemapRange(
      Input Value To Remap=&VectorLength(Vector=$Engine.Owner.Velocity)
      Input Min=300
      Input Max=0
      Output Min=0
      Output Max=1
     )
     B=&RandomRangeFloat(Minimum=1, Maximum=1)
    )
    MaskSpawnGroup=false
    Offset=&MakeVector(X=0, Y=0, Z=-9)
    Offset Coordinate Space=::World
    Offset Mode=::NewEnumerator0
    Shape Primitive=::NewEnumerator3
    Torus Distribution Mode:ENiagaraTorusDistributionMode=::NewEnumerator1
    U Position=&ReturnNormalizedExecIndex(Normalized Index Includes One=false)
    UseOwnerScale=true
   ) {module=ShapeLocation001}
   &AddVelocity(
    Rotation Coordinate Space=::World
    Velocity=<0 0 15>
    Velocity Speed Scale=&Multiply_Float(
     A=&Multiply_Float(
      A=&RemapRange(
       Input Value To Remap=&VectorLength(Vector=$Engine.Owner.Velocity)
       Input Min=100
       Input Max=0
       Output Min=0.2
       Output Max=1
      )
      B=&RandomRangeFloat(Minimum=1, Maximum=1)
     )
     B=&ClampFloat(Float=$User.Height, Min=0.3)
    )
   )
   &CameraOffset(Camera Offset Amount=2) {module=CameraOffset001}
   /SetVariables(
    Particles.Scale=&Multiply_VectorByFloat(Vector=$Particles.Scale, Float=&Multiply_Float(A=1, B=1))
    Particles.SpriteSize=&Multiply_Vector2D_ByFloat(Float=1, Vector2D=$Particles.SpriteSize)
   ) {enabled=false, isSetParameters=true, module=SetVariables_#3}
  ]
  particleUpdateScript [
   &ParticleState()
   &Drag(Drag=&RandomRangeFloat(Minimum=4, Maximum=6), Use Linear Drag=true, Use Rotational Drag=false)
   &AccelerationForce(Acceleration=&MakeVector(
     X=0
     Y=0
     Z=&ClampFloat(
      Float=&Multiply_Float(
       A=$User.Height
       B=&Multiply_Float(
        A=&RemapRange(
         Input Value To Remap=&VectorLength(Vector=$Engine.Owner.Velocity)
         Input Min=100
         Input Max=0
         Output Min=0.2
         Output Max=1
        )
        B=&RandomRangeFloat(Minimum=200, Maximum=200)
       )
      )
      Min=20
      Max=140
     )
    )) {module=AccelerationForce001}
   &SpriteRotationRate(Rotation Rate=&Multiply_Float(
     A=&UniformRangedFloat(Minimum=1, Maximum=1)
     B=&FloatFromCurve(
      FloatCurve=DataInterface() {propertyValues {Curve {keys [{arriveTangent=-3.245, interpMode=RCIM_Cubic, leaveTangent=-3.245, tangentMode=RCTM_User, time=0, value=1}, <at 0.2 0.55>, <at 1.0011 0.5>]}}}
      Scale Curve=&SelectFloatFromArray(
       Array Sampling Mode=::NewEnumerator0
       Direct Array Index=?
       Float Selection Array=DataInterface() {propertyValues {FloatData [-100,100]}}
       Random Seed=0
      )
     )
    ))
   /Niagara/Modules/Update/Forces/CurlNoiseForce(
    Noise Frequency:NiagaraFloat=&Divide_Float(A=100, B=100)
    Noise Quality / Cost=::NewEnumerator0
    Noise Strength=&Multiply_Float(
     A=&Multiply_Float(
      A=&RemapRange(
       Input Value To Remap=&VectorLength(Vector=$Engine.Owner.Velocity)
       Input Min=100
       Input Max=0
       Output Min=0.2
       Output Max=1
      )
      B=&RandomRangeFloat(Minimum=1, Maximum=1)
     )
     B=1
    )
    Pan Noise=true
    Pan Noise Field=<0 0 0.1>
   )
   &InheritSourceMovement(
    Applied Position Scale=&Multiply_VectorByFloat(
     Vector=<0.1 0.1 0.01>
     Float=&Multiply_Float(
      A=&RemapRange(
       Input Value To Remap=&VectorLength(Vector=$Engine.Owner.Velocity)
       Input Min=100
       Input Max=0
       Output Min=0
       Output Max=1
      )
      B=&RandomRangeFloat(Minimum=0.1, Maximum=0.1)
     )
    )
    ApplyForce=false
    ApplyPosition=true
    ApplyPositionFalloff=false
   ) {enabled=false}
   &SolveForcesAndVelocity(Clamp Velocity=false, Limit Acceleration=false)
   &ScaleSpriteSize(Scale Factor=&Vector2DFromFloat(Value=&FloatFromCurve(
      FloatCurve=DataInterface() {propertyValues {Curve {keys [
          {arriveTangent=0.5672, interpMode=RCIM_Cubic, leaveTangent=0.5672, tangentMode=RCTM_User, time=0, value=0.85}
          {arriveTangent=0.15, interpMode=RCIM_Cubic, leaveTangent=0.15, time=0.5, value=1.2}
          {interpMode=RCIM_Cubic, leaveTangent=0.155, time=1, value=1}
         ]}}}
      Scale Curve=&Multiply_Float(
       A=&Multiply_Float(
        A=&RemapRange(
         Input Value To Remap=&VectorLength(Vector=$Engine.Owner.Velocity)
         Input Min=500
         Input Max=0
         Output Min=0.4
         Output Max=1
        )
        B=&RandomRangeFloat(Minimum=1, Maximum=1)
       )
       B=&ClampFloat(Float=$User.Height, Min=0.3)
      )
     )))
   &ScaleColor(
    Scale Alpha=&FloatFromCurve(
     FloatCurve=DataInterface() {propertyValues {Curve {keys [
         {arriveTangent=6.9262, interpMode=RCIM_Cubic, leaveTangent=6.9262, tangentMode=RCTM_User, time=0.01, value=0.1}
         {arriveTangent=-0.0768, interpMode=RCIM_Cubic, leaveTangent=-0.0768, tangentMode=RCTM_User, time=0.2, value=0.8}
         {arriveTangent=-0.9986, interpMode=RCIM_Cubic, leaveTangent=-0.9986, time=0.7, value=0.8}
         {arriveTangent=-0.6884, interpMode=RCIM_Cubic, leaveTangent=-0.6884, tangentMode=RCTM_User, time=1.0011, value=0}
        ]}}}
     Scale Curve=&Multiply_Float(
      A=&Multiply_Float(
       A=&RemapRange(
        Input Value To Remap=&VectorLength(Vector=$Engine.Owner.Velocity)
        Input Min=100
        Input Max=0
        Output Min=0.5
        Output Max=1
       )
       B=&RandomRangeFloat(Minimum=1, Maximum=1)
      )
      B=&ClampFloat(Float=$User.Height, Min=0.3)
     )
    )
    Scale RGB=&Lerp_Vector(
     A=<0.15 0.15 0.15>
     B=<0.3 0.3 0.3>
     Alpha=&FloatFromCurve(FloatCurve=DataInterface() {propertyValues {Curve {keys [{tangentMode=RCTM_User, time=0.1, value=1}, {tangentMode=RCTM_User, time=0.35, value=1}]}}})
    )
    ScaleA=true
    ScaleRGB=true
   ) {module=ScaleColor001}
   &DynamicMaterialParameters(
    Index 0 Param 1=&Add_Float(
     A=&RandomRangeFloat(Minimum=0, Maximum=15)
     B=&FloatFromCurve(
      FloatCurve=DataInterface() {propertyValues {Curve {keys [
          {arriveTangent=-2.2561, interpMode=RCIM_Cubic, leaveTangent=-2.2561, tangentMode=RCTM_User, time=0, value=1}
          {arriveTangent=-1.0036, interpMode=RCIM_Cubic, leaveTangent=-1.0036, tangentMode=RCTM_User, time=0.2682, value=0.3}
          <at 0.8 0>
         ]}}}
      Scale Curve=0.5
     )
    )
    Index 0 Param 2=&FloatFromCurve(
     FloatCurve=DataInterface() {propertyValues {Curve {keys [
         {arriveTangent=0.0861, interpMode=RCIM_Cubic, leaveTangent=0.0861, tangentMode=RCTM_User, time=0.1, value=0.3}
         {arriveTangent=1.536, interpMode=RCIM_Cubic, leaveTangent=1.536, tangentMode=RCTM_User, time=1, value=1.25}
        ]}}}
     Scale Curve=1.5
    )
    Index 0 Param 3=1.1
    Index 0 Param 4=1
    Index 1 Param 1=&MaskFloatBySpawnGroup(
     Masked Float=&FloatFromCurve(FloatCurve=DataInterface() {propertyValues {Curve {keys [
          {arriveTangent=8.5612, interpMode=RCIM_Cubic, leaveTangent=8.5612, tangentMode=RCTM_User, time=0, value=-0.5}
          {arriveTangent=0.9995, interpMode=RCIM_Cubic, leaveTangent=0.9995, tangentMode=RCTM_User, time=0.25, value=0.85}
          {arriveTangent=-0.0405, interpMode=RCIM_Cubic, leaveTangent=-0.0405, tangentMode=RCTM_User, time=0.5, value=1}
         ]}}})
     Passthrough Float=&FloatFromCurve(
      FloatCurve=DataInterface() {propertyValues {Curve {keys [
          {arriveTangent=4.1785, interpMode=RCIM_Cubic, leaveTangent=4.1785, tangentMode=RCTM_User, time=0, value=-0.5}
          {arriveTangent=0.9995, interpMode=RCIM_Cubic, leaveTangent=0.9995, tangentMode=RCTM_User, time=0.3, value=0.8}
          {arriveTangent=-0.0405, interpMode=RCIM_Cubic, leaveTangent=-0.0405, tangentMode=RCTM_User, time=0.8, value=1}
         ]}}}
      Scale Curve=2
     )
     Spawn Group Mask=1
    )
    Index 1 Param 2=&Add_Float(
     A=&FloatFromCurve(
      FloatCurve=DataInterface() {propertyValues {Curve {keys [
          {arriveTangent=-12.0154, interpMode=RCIM_Cubic, leaveTangent=-12.0154, tangentMode=RCTM_User, time=0, value=1}
          {interpMode=RCIM_Cubic, tangentMode=RCTM_User, time=0.2, value=0}
         ]}}}
      Scale Curve=-1
     )
     B=&RandomRangeFloat(Minimum=-100, Maximum=100)
    )
    Index 1 Param 3=&FloatFromCurve(
     FloatCurve=DataInterface() {propertyValues {Curve {keys [{interpMode=RCIM_Cubic, tangentMode=RCTM_User, time=0.333, value=1.2}, {interpMode=RCIM_Cubic, tangentMode=RCTM_User, time=1, value=2}]}}}
     Scale Curve=1.5
    )
    Index 1 Param 4=&FloatFromCurve(FloatCurve=DataInterface() {propertyValues {Curve {keys [<at 0 0.6>,<at 1 0.3>]}}})
    Index 2 Param 1=&FloatFromCurve(FloatCurve=DataInterface() {propertyValues {Curve {keys [
         {arriveTangent=0.0048, interpMode=RCIM_Cubic, leaveTangent=0.0048, tangentMode=RCTM_User, time=0, value=1.2}
         {arriveTangent=0.3333, interpMode=RCIM_Cubic, leaveTangent=0.3333, time=0.4, value=0.9}
         {interpMode=RCIM_Cubic, leaveTangent=1.688, time=0.9, value=1.5}
        ]}}})
    Index 2 Param 2=&Add_Float(
     A=&RandomRangeFloat(Minimum=0, Maximum=100)
     B=&FloatFromCurve(
      FloatCurve=DataInterface() {propertyValues {Curve {keys [
          {arriveTangent=0.5501, interpMode=RCIM_Cubic, leaveTangent=0.5501, tangentMode=RCTM_User, time=0, value=0}
          {arriveTangent=1.1924, interpMode=RCIM_Cubic, leaveTangent=1.1924, tangentMode=RCTM_User, time=1, value=0.8}
         ]}}}
      Scale Curve=&RandomRangeFloat(Minimum=-0.3, Maximum=-0.56)
     )
    )
    Index 2 Param 3=&FloatFromCurve(
     FloatCurve=DataInterface() {propertyValues {Curve {keys [
         {arriveTangent=-0.4045, interpMode=RCIM_Cubic, tangentMode=RCTM_User, time=0.4, value=1}
         {arriveTangent=5.0157, interpMode=RCIM_Cubic, leaveTangent=5.0157, tangentMode=RCTM_User, time=1, value=2}
        ]}}}
     Scale Curve=2
    )
    Index 2 Param 4=&FloatFromCurve(
     FloatCurve=DataInterface() {propertyValues {Curve {keys [{arriveTangent=-0.4045, time=0, value=0.75}, {leaveTangent=-0.318, time=1, value=1}]}}}
     Scale Curve=0.4
    )
    Index 3 Param 1=&FloatFromCurve(
     FloatCurve=DataInterface() {propertyValues {Curve {keys [
         {arriveTangent=0.6183, interpMode=RCIM_Cubic, leaveTangent=0.6183, tangentMode=RCTM_User, time=0.1, value=0.02}
         {interpMode=RCIM_Cubic, tangentMode=RCTM_User, time=0.5, value=1}
        ]}}}
     Scale Curve=4
    )
    Index 3 Param 2=&FloatFromCurve(
     FloatCurve=DataInterface() {propertyValues {Curve {keys [<at 0 1>,<at 1 0>]}}}
     Scale Curve=&RandomRangeFloat(Minimum=0.4, Maximum=0.5)
    )
    Index 3 Param 3=&FloatFromCurve(
     FloatCurve=DataInterface() {propertyValues {Curve {keys [
         {arriveTangent=0.4538, interpMode=RCIM_Cubic, leaveTangent=0.4538, tangentMode=RCTM_User, time=0, value=0.5}
         {arriveTangent=0.2675, interpMode=RCIM_Cubic, leaveTangent=0.2675, tangentMode=RCTM_User, time=1, value=0.7}
        ]}}}
     Scale Curve=0.0004
    )
    Index 3 Param 4=1
    Module.Index0Param0Write=true
    Module.Index0Param1Write=true
    Module.Index0Param2Write=true
    Module.Index0Param3Write=false
    Module.Index1Param0Write=false
    Module.Index1Param1Write=false
    Module.Index1Param2Write=true
    Module.Index1Param3Write=true
    Module.Index2Param0Write=true
    Module.Index2Param1Write=true
    Module.Index2Param2Write=true
    Module.Index2Param3Write=true
    Module.Index3Param0Write=true
    Module.Index3Param1Write=true
    Module.Index3Param2Write=true
    Module.Index3Param3Write=true
    Write Parameter Index 1=true
    Write Parameter Index 2=true
    Write Parameter Index 3=true
   ) {module=DynamicMaterialParameters001}
   /SetVariables(
    Particles.Color2=&MultiplyLinearColor(
     Linear Color=&ColorFromCurve(ColorCurve=DataInterface() {propertyValues {
        AlphaCurve {keys [<at 0 0 RCIM_Cubic>,<at 0.1319 0.03>,<at 1 0 RCIM_Cubic>]}
        BlueCurve {keys [<at 0 0 RCIM_Cubic>]}
        GreenCurve {keys [<at 0 0.2698 RCIM_Cubic>]}
        RedCurve {keys [<at 0 1 RCIM_Cubic>]}
       }})
     Scale Factor=&MakeVector4(
      X=1
      Y=1
      Z=1
      W=&Multiply_Float(
       A=&Multiply_Float(
        A=&RemapRange(
         Input Value To Remap=&VectorLength(Vector=$Engine.Owner.Velocity)
         Input Min=400
         Input Max=0
         Output Min=0.2
         Output Max=1
        )
        B=&RandomRangeFloat(Minimum=1, Maximum=1)
       )
       B=&ClampFloat(Float=$User.Height, Min=0.3)
      )
     )
    )
    Particles.Size2:Vector2f=&Multiply_Vector2D_ByFloat(
     Float=&FloatFromCurve(
      FloatCurve=DataInterface() {propertyValues {Curve {keys [
          {arriveTangent=2.1001, interpMode=RCIM_Cubic, leaveTangent=2.1001, tangentMode=RCTM_User, time=0, value=0.6}
          {arriveTangent=0.5263, interpMode=RCIM_Cubic, leaveTangent=0.5263, tangentMode=RCTM_User, time=0.4013, value=0.9169}
          {arriveTangent=0.0224, interpMode=RCIM_Cubic, leaveTangent=0.0224, tangentMode=RCTM_User, time=1, value=1}
         ]}}}
      Scale Curve=&Multiply_Float(
       A=&Multiply_Float(
        A=&RemapRange(
         Input Value To Remap=&VectorLength(Vector=$Engine.Owner.Velocity)
         Input Min=400
         Input Max=0
         Output Min=0.5
         Output Max=1
        )
        B=&RandomRangeFloat(Minimum=1, Maximum=1)
       )
       B=&ClampFloat(Float=$User.Height, Min=0.3)
      )
     )
     Vector2D=<33 33>
    )
   ) {isSetParameters=true}
  ]
  rendererClasses [&NiagaraSpriteRendererProperties, &NiagaraLightRendererProperties]
  renderers [
   {rendererClass=&NiagaraSpriteRendererProperties}
   {rendererClass=&NiagaraLightRendererProperties, rendererIndex=1}
   {rendererClass=&NiagaraSpriteRendererProperties, rendererIndex=2}
  ]
 }
 Fire_Main_Inherit {
  bEnabled=false
  emitterUpdateScript [
   &EmitterState(Delay First Loop Only=true, Life Cycle Mode=::NewEnumerator1, Loop Delay=0.6, Loop Duration=1, UseLoopDelay=true)
   &SpawnPerUnit(
    Max Movement Threshold=500
    Module.Use Max Movement Threshold=true
    Module.Use Movement Tolerance=false
    Spawn Spacing=1
    Use Spawn Probability=false
   )
  ]
  particleSpawnScript [
   &InitializeParticle(
    Color=<rgba 1 0.8244 0.8142 2>
    Lifetime=&Multiply_Float(
     A=&RandomRangeFloat(Minimum=0.1, Maximum=0.2)
     B=&MakeCustomFloatFromBool(Bool=&SetBoolByFloatComparison(A=1, B=0), True Float=1, False Float=1)
    )
    Position Offset=&Multiply_VectorByFloat(Vector=Linked() {linkedVariable {name=User.Offset, type=NiagaraPosition}}, Float=1)
    Position Offset Coordinate Space=::Local
    Sprite Size Mode=::NewEnumerator1
    Sprite UV Mode=::NewEnumerator3
    Uniform Sprite Size Min=9
    UsePositionOffset=true
    Write Material Random=true
   ) {module=InitializeParticle001}
   &ShapeLocation(
    Handle Radius=0
    Large Radius=&Multiply_Float(A=5, B=1)
    MaskSpawnGroup=false
    Offset=<0 0 -2.5>
    Offset Coordinate Space=::World
    Offset Mode=::NewEnumerator0
    Shape Primitive=::NewEnumerator3
    Torus Distribution Mode:ENiagaraTorusDistributionMode=::NewEnumerator1
    U Position=&ReturnNormalizedExecIndex(Normalized Index Includes One=false)
    UseOwnerScale=true
   ) {module=ShapeLocation001}
   &AddVelocity(Rotation Coordinate Space=::World, Velocity=<0 0 15>)
   &CameraOffset(Camera Offset Amount=2) {module=CameraOffset001}
  ]
  particleUpdateScript [
   &ParticleState()
   &Drag(Drag=&RandomRangeFloat(Minimum=4, Maximum=6), Use Linear Drag=false, Use Rotational Drag=false)
   &AccelerationForce(Acceleration=&MakeVector(
     X=0
     Y=0
     Z=&ClampFloat(Float=&Multiply_Float(A=$User.Height, B=200), Min=40, Max=140)
    )) {module=AccelerationForce001}
   &SpriteRotationRate(Rotation Rate=&Multiply_Float(
     A=&UniformRangedFloat(Minimum=1, Maximum=1)
     B=&FloatFromCurve(
      FloatCurve=DataInterface() {propertyValues {Curve {keys [{arriveTangent=-3.245, interpMode=RCIM_Cubic, leaveTangent=-3.245, tangentMode=RCTM_User, time=0, value=1}, <at 0.2 0.55>, <at 1.0011 0.5>]}}}
      Scale Curve=&SelectFloatFromArray(
       Array Sampling Mode=::NewEnumerator0
       Direct Array Index=?
       Float Selection Array=DataInterface() {propertyValues {FloatData [-100,100]}}
       Random Seed=0
      )
     )
    ))
   &CurlNoiseForce(
    Noise Frequency:NiagaraFloat=&Divide_Float(A=1, B=1)
    Noise Quality / Cost=::NewEnumerator0
    Noise Strength=&Multiply_Float(A=100, B=1)
    Pan Noise=true
    Pan Noise Field=<0 0 0.1>
   )
   &InheritSourceMovement(
    Applied Position Scale=&Multiply_VectorByFloat(Vector=<1 1 1>, Float=0.8)
    ApplyForce=false
    ApplyPosition=true
    ApplyPositionFalloff=false
   )
   &SolveForcesAndVelocity(Clamp Velocity=false, Limit Acceleration=false)
   &ScaleSpriteSizeBySpeed(Max Scale Factor=<1.5 2>, Min Scale Factor=<1 1.5>)
   &ScaleSpriteSize(Scale Factor=&Vector2DFromFloat(Value=&FloatFromCurve(
      FloatCurve=DataInterface() {propertyValues {Curve {keys [
          {arriveTangent=0.5672, interpMode=RCIM_Cubic, leaveTangent=0.5672, tangentMode=RCTM_User, time=0, value=1}
          <at 0.2 1.15 RCIM_Cubic>
          {interpMode=RCIM_Cubic, leaveTangent=0.155, time=1, value=1}
         ]}}}
      Scale Curve=&Multiply_Float(A=1, B=&ClampFloat(Float=$User.Height, Min=0.3))
     )))
   &ScaleColor(
    Scale Alpha=&FloatFromCurve(FloatCurve=DataInterface() {propertyValues {Curve {keys [
         {arriveTangent=-1.5703, interpMode=RCIM_Cubic, leaveTangent=-1.5703, tangentMode=RCTM_User, time=0, value=0.8}
         {arriveTangent=-0.6884, interpMode=RCIM_Cubic, leaveTangent=-0.6884, tangentMode=RCTM_User, time=1.0011, value=0}
        ]}}})
    Scale RGB=&Lerp_Vector(
     A=<0.15 0.15 0.15>
     B=<0.3 0.3 0.3>
     Alpha=&FloatFromCurve(FloatCurve=DataInterface() {propertyValues {Curve {keys [{tangentMode=RCTM_User, time=0.1, value=1}, {tangentMode=RCTM_User, time=0.35, value=1}]}}})
    )
    ScaleA=true
    ScaleRGB=true
   ) {module=ScaleColor001}
   &DynamicMaterialParameters(
    Index 0 Param 1=&Add_Float(
     A=&RandomRangeFloat(Minimum=0, Maximum=15)
     B=&FloatFromCurve(
      FloatCurve=DataInterface() {propertyValues {Curve {keys [
          {arriveTangent=-2.2561, interpMode=RCIM_Cubic, leaveTangent=-2.2561, tangentMode=RCTM_User, time=0, value=1}
          {arriveTangent=-1.0036, interpMode=RCIM_Cubic, leaveTangent=-1.0036, tangentMode=RCTM_User, time=0.2682, value=0.3}
          <at 0.8 0>
         ]}}}
      Scale Curve=0.5
     )
    )
    Index 0 Param 2=&FloatFromCurve(
     FloatCurve=DataInterface() {propertyValues {Curve {keys [
         {arriveTangent=0.0861, interpMode=RCIM_Cubic, leaveTangent=0.0861, tangentMode=RCTM_User, time=0.1, value=0.3}
         {arriveTangent=1.536, interpMode=RCIM_Cubic, leaveTangent=1.536, tangentMode=RCTM_User, time=1, value=1.25}
        ]}}}
     Scale Curve=1.5
    )
    Index 0 Param 3=1.1
    Index 0 Param 4=1
    Index 1 Param 1=&MaskFloatBySpawnGroup(
     Masked Float=&FloatFromCurve(FloatCurve=DataInterface() {propertyValues {Curve {keys [
          {arriveTangent=8.5612, interpMode=RCIM_Cubic, leaveTangent=8.5612, tangentMode=RCTM_User, time=0, value=-0.5}
          {arriveTangent=0.9995, interpMode=RCIM_Cubic, leaveTangent=0.9995, tangentMode=RCTM_User, time=0.25, value=0.85}
          {arriveTangent=-0.0405, interpMode=RCIM_Cubic, leaveTangent=-0.0405, tangentMode=RCTM_User, time=0.5, value=1}
         ]}}})
     Passthrough Float=&FloatFromCurve(
      FloatCurve=DataInterface() {propertyValues {Curve {keys [
          {arriveTangent=4.1785, interpMode=RCIM_Cubic, leaveTangent=4.1785, tangentMode=RCTM_User, time=0, value=-0.5}
          {arriveTangent=0.9995, interpMode=RCIM_Cubic, leaveTangent=0.9995, tangentMode=RCTM_User, time=0.3, value=0.8}
          {arriveTangent=-0.0405, interpMode=RCIM_Cubic, leaveTangent=-0.0405, tangentMode=RCTM_User, time=0.8, value=1}
         ]}}}
      Scale Curve=2
     )
     Spawn Group Mask=1
    )
    Index 1 Param 2=&Add_Float(
     A=&FloatFromCurve(
      FloatCurve=DataInterface() {propertyValues {Curve {keys [
          {arriveTangent=-12.0154, interpMode=RCIM_Cubic, leaveTangent=-12.0154, tangentMode=RCTM_User, time=0, value=1}
          {interpMode=RCIM_Cubic, tangentMode=RCTM_User, time=0.2, value=0}
         ]}}}
      Scale Curve=-1
     )
     B=&RandomRangeFloat(Minimum=-100, Maximum=100)
    )
    Index 1 Param 3=&FloatFromCurve(
     FloatCurve=DataInterface() {propertyValues {Curve {keys [{interpMode=RCIM_Cubic, tangentMode=RCTM_User, time=0.333, value=1.4}, {interpMode=RCIM_Cubic, tangentMode=RCTM_User, time=1, value=2}]}}}
     Scale Curve=1.5
    )
    Index 1 Param 4=&FloatFromCurve(FloatCurve=DataInterface() {propertyValues {Curve {keys [<at 0 0.6>,<at 1 0.5>]}}})
    Index 2 Param 1=&FloatFromCurve(FloatCurve=DataInterface() {propertyValues {Curve {keys [
         {arriveTangent=0.0048, interpMode=RCIM_Cubic, leaveTangent=0.0048, tangentMode=RCTM_User, time=0, value=1.2}
         {arriveTangent=0.3333, interpMode=RCIM_Cubic, leaveTangent=0.3333, time=0.4, value=0.9}
         {interpMode=RCIM_Cubic, leaveTangent=1.688, time=0.9, value=1.5}
        ]}}})
    Index 2 Param 2=&Add_Float(
     A=&RandomRangeFloat(Minimum=0, Maximum=100)
     B=&FloatFromCurve(
      FloatCurve=DataInterface() {propertyValues {Curve {keys [
          {arriveTangent=0.5501, interpMode=RCIM_Cubic, leaveTangent=0.5501, tangentMode=RCTM_User, time=0, value=0}
          {arriveTangent=1.1924, interpMode=RCIM_Cubic, leaveTangent=1.1924, tangentMode=RCTM_User, time=1, value=0.8}
         ]}}}
      Scale Curve=&RandomRangeFloat(Minimum=-0.3, Maximum=-0.56)
     )
    )
    Index 2 Param 3=&FloatFromCurve(
     FloatCurve=DataInterface() {propertyValues {Curve {keys [
         {arriveTangent=-0.4045, interpMode=RCIM_Cubic, tangentMode=RCTM_User, time=0.4, value=1.25}
         {arriveTangent=2.4454, interpMode=RCIM_Cubic, leaveTangent=2.4454, tangentMode=RCTM_User, time=1, value=2}
        ]}}}
     Scale Curve=2
    )
    Index 2 Param 4=&FloatFromCurve(
     FloatCurve=DataInterface() {propertyValues {Curve {keys [{arriveTangent=-0.4045, time=0, value=0.75}, {leaveTangent=-0.318, time=1, value=1}]}}}
     Scale Curve=0.4
    )
    Index 3 Param 1=&FloatFromCurve(
     FloatCurve=DataInterface() {propertyValues {Curve {keys [
         {arriveTangent=0.6183, interpMode=RCIM_Cubic, leaveTangent=0.6183, tangentMode=RCTM_User, time=0.1, value=0.02}
         {interpMode=RCIM_Cubic, tangentMode=RCTM_User, time=0.5, value=1}
        ]}}}
     Scale Curve=4
    )
    Index 3 Param 2=&FloatFromCurve(
     FloatCurve=DataInterface() {propertyValues {Curve {keys [<at 0 1>,<at 1 0>]}}}
     Scale Curve=&RandomRangeFloat(Minimum=0.4, Maximum=0.5)
    )
    Index 3 Param 3=&FloatFromCurve(
     FloatCurve=DataInterface() {propertyValues {Curve {keys [
         {arriveTangent=0.4538, interpMode=RCIM_Cubic, leaveTangent=0.4538, tangentMode=RCTM_User, time=0, value=0.5}
         {arriveTangent=0.2675, interpMode=RCIM_Cubic, leaveTangent=0.2675, tangentMode=RCTM_User, time=1, value=0.7}
        ]}}}
     Scale Curve=0.0004
    )
    Index 3 Param 4=1
    Module.Index0Param0Write=true
    Module.Index0Param1Write=true
    Module.Index0Param2Write=true
    Module.Index0Param3Write=false
    Module.Index1Param0Write=false
    Module.Index1Param1Write=false
    Module.Index1Param2Write=true
    Module.Index1Param3Write=true
    Module.Index2Param0Write=true
    Module.Index2Param1Write=true
    Module.Index2Param2Write=true
    Module.Index2Param3Write=true
    Module.Index3Param0Write=true
    Module.Index3Param1Write=true
    Module.Index3Param2Write=true
    Module.Index3Param3Write=true
    Write Parameter Index 1=true
    Write Parameter Index 2=true
    Write Parameter Index 3=true
   ) {module=DynamicMaterialParameters001}
  ]
  rendererClasses [&NiagaraSpriteRendererProperties]
  renderers [{rendererClass=&NiagaraSpriteRendererProperties}]
 }
 Fire_Mesh_Root001 {
  emitterUpdateScript [
   &EmitterState(
    Inactive Response=::NewEnumerator1
    Life Cycle Mode=::NewEnumerator1
    Loop Behavior=::NewEnumerator1
    Loop Delay=0.65
    Loop Duration Mode=::NewEnumerator1
    UseLoopDelay=true
   )
   &SpawnBurst_Instantaneous(Use Spawn Probability=false, UseLoopCountLimit=false)
   &SpawnRate(
    SpawnRate=&RandomRangeFloat(Minimum=8, Maximum=9, Evaluation Type=::NewEnumerator1)
    Use Spawn Probability=false
   ) {enabled=false}
  ]
  particleSpawnScript [
   &InitializeParticle(
    Color=<rgba 15 1.8465 0 1>
    Lifetime=&RandomRangeFloat(Minimum=0.5, Maximum=0.5)
    Mesh Scale Mode=::NewEnumerator3
    Mesh Uniform Scale=&Multiply_Float(
     A=&Multiply_Float(A=0.045, B=&ClampFloat(Float=$User.Height, Min=0.5))
     B=&Multiply_Float(A=1, B=&ClampFloat(Float=$User.Height, Min=0.3))
    )
    Position Offset=&Multiply_Vector(A=Linked() {linkedVariable {name=User.Offset, type=NiagaraPosition}}, B=<1 0.3 1>)
    Position Offset Coordinate Space=::Local
    UsePositionOffset=true
    Write Material Random=true
   )
   &InitialMeshOrientation(
    Mesh Axis To Orient=$User.Orientation
    Mesh Orientation Mode=::NewEnumerator2
    Orientation Vector=<0 0 1>
    Rotation=$User.Rotation
    Use Rotation Vector=true
   )
   &CameraOffset(Camera Offset Amount=10) {enabled=false}
   &ShapeLocation(
    MaskSpawnGroup=false
    Non Uniform Scale=<0.1 0.1 0.1>
    Offset=<0 0 -6.5>
    Offset Coordinate Space=::World
    Offset Mode=::NewEnumerator0
    Override Seed=false
    Ring Radius=0
    Shape Primitive=::NewEnumerator4
    UseOwnerScale=true
   )
  ]
  particleUpdateScript [
   &ParticleState(Kill Particles When Lifetime Has Elapsed=false, Let Infinitely Lived Particles Die When Emitter Deactivates=true)
   &UpdateMeshOrientation(
    Rotation Rate=&RandomRangeFloat(Minimum=0.1, Maximum=0.2)
    Rotation Vector=&RandomRangeVector(Minimum=<-0.2 1 0>, Maximum=<0.2 1 0>)
   ) {enabled=false}
   &Timeline(
    Play Rate=&FloatFromCurve(FloatCurve=DataInterface() {propertyValues {Curve {keys [<at 0 1 RCIM_Cubic>,<at 1 0.35 RCIM_Cubic>]}}}, Scale Curve=2.5)
    Timeline End=10000000
   ) {enabled=false}
   &ScaleMeshSize(Scale Factor=&Multiply_VectorByFloat(
     Vector=&MakeVector(X=1, Y=1, Z=1)
     Float=&Multiply_Float(
      A=&RemapRange(
       Input Value To Remap=&VectorLength(Vector=$Engine.Owner.Velocity)
       Input Min=250
       Input Max=0
       Output Min=0.6
       Output Max=1
      )
      B=&RandomRangeFloat(Minimum=1, Maximum=1)
     )
    ))
   &DynamicMaterialParameters(
    Index 0 Param 1=&Multiply_Float(
     A=27
     B=&Multiply_Float(
      A=&RemapRange(
       Input Value To Remap=&VectorLength(Vector=$Engine.Owner.Velocity)
       Input Min=100
       Input Max=0
       Output Min=0.5
       Output Max=1
      )
      B=&RandomRangeFloat(Minimum=1, Maximum=1)
     )
    )
    Index 0 Param 2=1
    Index 0 Param 3=1
    Index 0 Param 4=0.1
    Index 1 Param 1=0.3
    Index 2 Param 1=&RandomRangeFloat(Minimum=0, Maximum=1)
    Index 2 Param 2=1
    Index 2 Param 3=&FloatFromCurve(FloatCurve=DataInterface() {propertyValues {Curve {keys [<at 0 1>,<at 1 0>]}}}, Scale Curve=0)
    Module.Index0Param0Write=true
    Module.Index0Param1Write=true
    Module.Index0Param2Write=true
    Module.Index0Param3Write=true
    Module.Index1Param0Write=true
    Module.Index1Param1Write=false
    Module.Index1Param2Write=false
    Module.Index1Param3Write=false
    Module.Index2Param0Write=true
    Module.Index2Param1Write=true
    Module.Index2Param2Write=false
    Module.Index2Param3Write=false
    Write Parameter Index 1=true
    Write Parameter Index 2=true
   )
   &ScaleColor(
    Scale Alpha=&Multiply_Float(
     A=1
     B=&MakeFloatFromVector(
      Vector=&Multiply_VectorByFloat(
       Vector=Linked() {linkedVariable {name=User.Offset, type=NiagaraPosition}}
       Float=&Multiply_Float(
        A=&RemapRange(
         Input Value To Remap=&VectorLength(Vector=$Engine.Owner.Velocity)
         Input Min=200
         Input Max=0
         Output Min=-0.025
         Output Max=-1
        )
        B=&RandomRangeFloat(Minimum=1, Maximum=1.1)
       )
      )
      Channel=::NewEnumerator0
     )
    )
    Scale RGB=&Multiply_VectorByFloat(Vector=<1 1 1>, Float=1)
    ScaleA=true
    ScaleRGB=false
   )
   &KillParticles(Kill Particles=&SetBoolByFloatComparison(A=$User.Height, B=0.5, Comparison Type=::NewEnumerator6)) {enabled=false}
   &CameraOffset(Camera Offset Amount=2) {module=CameraOffset001}
  ]
  rendererClasses [&NiagaraMeshRendererProperties]
  renderers [{rendererClass=&NiagaraMeshRendererProperties, rendererIndex=0}]
 }
 Fire_Root {
  emitterUpdateScript [
   &EmitterState(Delay First Loop Only=true, Life Cycle Mode=::NewEnumerator1, Loop Delay=0.6, Loop Duration=1, UseLoopDelay=true)
   &SpawnRate(SpawnRate=10, Use Spawn Probability=false)
   &SpawnBurst_Instantaneous(Spawn Count=5, Use Spawn Probability=true, UseLoopCountLimit=false) {enabled=false, module=SpawnBurst_Instantaneous002}
  ]
  particleSpawnScript [
   &InitializeParticle(
    Color=&MakeLinearColor(R:NiagaraFloat=1, G:NiagaraFloat=1, B:NiagaraFloat=1, A:NiagaraFloat=1)
    Lifetime Max=0.6
    Lifetime Min=0.5
    Lifetime Mode=::NewEnumerator1
    Position Offset=&Multiply_Vector(A=Linked() {linkedVariable {name=User.Offset, type=NiagaraPosition}}, B=<0.8 0.35 1>)
    Position Offset Coordinate Space=::Local
    Sprite Size Mode=::NewEnumerator1
    Sprite UV Mode=::NewEnumerator3
    Uniform Sprite Size Min=9
    UsePositionOffset=true
    Write Material Random=true
   ) {module=InitializeParticle001}
   &ShapeLocation(
    Handle Radius=0
    Large Radius=&Multiply_Float(A=2, B=1)
    MaskSpawnGroup=false
    Offset=<0 0 -13>
    Offset Coordinate Space=::World
    Offset Mode=::NewEnumerator0
    Shape Primitive=::NewEnumerator3
    Torus Distribution Mode:ENiagaraTorusDistributionMode=::NewEnumerator1
    U Position=&ReturnNormalizedExecIndex(Normalized Index Includes One=false)
    UseOwnerScale=true
   ) {module=ShapeLocation001}
   &AddVelocity(Rotation Coordinate Space=::World, Velocity=<0 0 15>) {enabled=false}
   &AddVelocity(
    Rotation Coordinate Space=::Simulation
    Velocity=&MakeVector(
     X=0
     Y=&Multiply_Float(
      A=&MakeFloatFromVector(Vector=Linked() {linkedVariable {name=User.Offset, type=NiagaraPosition}}, Channel=::NewEnumerator1)
      B=6
     )
     Z=0
    )
   ) {module=AddVelocity001}
   &CameraOffset(Camera Offset Amount=2) {module=CameraOffset001}
   /SetVariables(
    Particles.Scale=&Multiply_VectorByFloat(Vector=$Particles.Scale, Float=&Multiply_Float(A=1, B=1))
    Particles.SpriteSize=&Multiply_Vector2D_ByFloat(Float=1, Vector2D=$Particles.SpriteSize)
   ) {enabled=false, isSetParameters=true, module=SetVariables_#3}
  ]
  particleUpdateScript [
   &ParticleState()
   &Drag(Drag=&RandomRangeFloat(Minimum=4, Maximum=6), Use Linear Drag=true, Use Rotational Drag=false)
   &AccelerationForce(Acceleration=&MakeVector(
     X=0
     Y=0
     Z=&ClampFloat(
      Float=&Multiply_Float(A=$User.Height, B=200)
      Min=40
      Max=&FloatFromCurve(
       FloatCurve=DataInterface() {propertyValues {Curve {keys [{tangentMode=RCTM_User, time=0.15, value=0}, {tangentMode=RCTM_User, time=0.2, value=1}]}}}
       Scale Curve=100
      )
     )
    )) {module=AccelerationForce001}
   &SpriteRotationRate(Rotation Rate=&Multiply_Float(
     A=&UniformRangedFloat(Minimum=1, Maximum=1)
     B=&FloatFromCurve(
      FloatCurve=DataInterface() {propertyValues {Curve {keys [{arriveTangent=-3.245, interpMode=RCIM_Cubic, leaveTangent=-3.245, tangentMode=RCTM_User, time=0, value=1}, <at 0.2 0.55>, <at 1.0011 0.5>]}}}
      Scale Curve=&SelectFloatFromArray(
       Array Sampling Mode=::NewEnumerator0
       Direct Array Index=?
       Float Selection Array=DataInterface() {propertyValues {FloatData [-100,100]}}
       Random Seed=0
      )
     )
    ))
   &CurlNoiseForce(
    Noise Frequency:NiagaraFloat=&Divide_Float(A=1, B=1)
    Noise Quality / Cost=::NewEnumerator0
    Noise Strength=&Multiply_Float(A=100, B=1)
    Pan Noise=true
    Pan Noise Field=<0 0 0.1>
   )
   &InheritSourceMovement(Applied Position Scale=<0.01 0.01 0.03>, ApplyForce=false, ApplyPosition=true, ApplyPositionFalloff=false) {enabled=false}
   &SolveForcesAndVelocity(Clamp Velocity=false, Limit Acceleration=false)
   &ScaleSpriteSizeBySpeed(Max Scale Factor=<0.8 1.7>, Min Scale Factor=<0.9 1.7>, Velocity Threshold=500)
   &ScaleSpriteSize(Scale Factor=&Vector2DFromFloat(Value=&FloatFromCurve(
      FloatCurve=DataInterface() {propertyValues {Curve {keys [
          {arriveTangent=0.5672, interpMode=RCIM_Cubic, leaveTangent=0.5672, tangentMode=RCTM_User, time=0, value=0.85}
          {arriveTangent=0.15, interpMode=RCIM_Cubic, leaveTangent=0.15, time=0.5, value=1.2}
          {interpMode=RCIM_Cubic, leaveTangent=0.155, time=1, value=1}
         ]}}}
      Scale Curve=&Multiply_Float(
       A=&Multiply_Float(
        A=&RemapRange(
         Input Value To Remap=&VectorLength(Vector=$Engine.Owner.Velocity)
         Input Min=100
         Input Max=0
         Output Min=0.5
         Output Max=1
        )
        B=&RandomRangeFloat(Minimum=1, Maximum=1)
       )
       B=&ClampFloat(Float=$User.Height, Min=0.3)
      )
     )))
   &ScaleColor(
    Scale Alpha=&FloatFromCurve(
     FloatCurve=DataInterface() {propertyValues {Curve {keys [
         {arriveTangent=5.8797, interpMode=RCIM_Cubic, leaveTangent=5.8797, tangentMode=RCTM_User, time=0, value=0}
         {arriveTangent=1.9105, interpMode=RCIM_Cubic, leaveTangent=1.9105, tangentMode=RCTM_User, time=0.3, value=0.75}
         {arriveTangent=-1.0697, interpMode=RCIM_Cubic, leaveTangent=-1.0697, time=0.7, value=0.8}
         {arriveTangent=-0.6884, interpMode=RCIM_Cubic, leaveTangent=-0.6884, tangentMode=RCTM_User, time=1.0011, value=0}
        ]}}}
     Scale Curve=1.15
    )
    Scale RGB=&Lerp_Vector(
     A=<0.15 0.15 0.15>
     B=<0.3 0.3 0.3>
     Alpha=&FloatFromCurve(FloatCurve=DataInterface() {propertyValues {Curve {keys [{tangentMode=RCTM_User, time=0.1, value=1}, {tangentMode=RCTM_User, time=0.35, value=1}]}}})
    )
    ScaleA=true
    ScaleRGB=true
   ) {module=ScaleColor001}
   &DynamicMaterialParameters(
    Index 0 Param 1=&Add_Float(
     A=&RandomRangeFloat(Minimum=0, Maximum=15)
     B=&FloatFromCurve(
      FloatCurve=DataInterface() {propertyValues {Curve {keys [
          {arriveTangent=-2.2561, interpMode=RCIM_Cubic, leaveTangent=-2.2561, tangentMode=RCTM_User, time=0, value=1}
          {arriveTangent=-1.0036, interpMode=RCIM_Cubic, leaveTangent=-1.0036, tangentMode=RCTM_User, time=0.2682, value=0.3}
          <at 0.8 0>
         ]}}}
      Scale Curve=0.25
     )
    )
    Index 0 Param 2=&FloatFromCurve(
     FloatCurve=DataInterface() {propertyValues {Curve {keys [
         {arriveTangent=0.0861, interpMode=RCIM_Cubic, leaveTangent=0.0861, tangentMode=RCTM_User, time=0.1, value=0.3}
         {arriveTangent=1.536, interpMode=RCIM_Cubic, leaveTangent=1.536, tangentMode=RCTM_User, time=1, value=1.05}
        ]}}}
     Scale Curve=1.5
    )
    Index 0 Param 3=1.1
    Index 0 Param 4=1
    Index 1 Param 1=&MaskFloatBySpawnGroup(
     Masked Float=&FloatFromCurve(FloatCurve=DataInterface() {propertyValues {Curve {keys [
          {arriveTangent=8.5612, interpMode=RCIM_Cubic, leaveTangent=8.5612, tangentMode=RCTM_User, time=0, value=-0.5}
          {arriveTangent=0.9995, interpMode=RCIM_Cubic, leaveTangent=0.9995, tangentMode=RCTM_User, time=0.25, value=0.85}
          {arriveTangent=-0.0405, interpMode=RCIM_Cubic, leaveTangent=-0.0405, tangentMode=RCTM_User, time=0.5, value=1}
         ]}}})
     Passthrough Float=&FloatFromCurve(
      FloatCurve=DataInterface() {propertyValues {Curve {keys [
          {arriveTangent=4.1785, interpMode=RCIM_Cubic, leaveTangent=4.1785, tangentMode=RCTM_User, time=0, value=-0.5}
          {arriveTangent=0.9995, interpMode=RCIM_Cubic, leaveTangent=0.9995, tangentMode=RCTM_User, time=0.3, value=0.8}
          {arriveTangent=-0.0405, interpMode=RCIM_Cubic, leaveTangent=-0.0405, tangentMode=RCTM_User, time=0.8, value=1}
         ]}}}
      Scale Curve=2
     )
     Spawn Group Mask=1
    )
    Index 1 Param 2=&Add_Float(
     A=&FloatFromCurve(
      FloatCurve=DataInterface() {propertyValues {Curve {keys [
          {arriveTangent=-12.0154, interpMode=RCIM_Cubic, leaveTangent=-12.0154, tangentMode=RCTM_User, time=0, value=1}
          {interpMode=RCIM_Cubic, tangentMode=RCTM_User, time=0.2, value=0}
         ]}}}
      Scale Curve=-1
     )
     B=&RandomRangeFloat(Minimum=-100, Maximum=100)
    )
    Index 1 Param 3=&FloatFromCurve(
     FloatCurve=DataInterface() {propertyValues {Curve {keys [{interpMode=RCIM_Cubic, tangentMode=RCTM_User, time=0.333, value=1.2}, {interpMode=RCIM_Cubic, tangentMode=RCTM_User, time=1, value=2}]}}}
     Scale Curve=1.5
    )
    Index 1 Param 4=&FloatFromCurve(FloatCurve=DataInterface() {propertyValues {Curve {keys [<at 0 0.6>,<at 1 0.3>]}}}, Scale Curve=0.4)
    Index 2 Param 1=&FloatFromCurve(FloatCurve=DataInterface() {propertyValues {Curve {keys [
         {arriveTangent=0.0048, interpMode=RCIM_Cubic, leaveTangent=0.0048, tangentMode=RCTM_User, time=0, value=1.2}
         {arriveTangent=0.3333, interpMode=RCIM_Cubic, leaveTangent=0.3333, time=0.4, value=0.9}
         {interpMode=RCIM_Cubic, leaveTangent=1.688, time=0.9, value=1.5}
        ]}}})
    Index 2 Param 2=&Add_Float(
     A=&RandomRangeFloat(Minimum=0, Maximum=100)
     B=&FloatFromCurve(
      FloatCurve=DataInterface() {propertyValues {Curve {keys [
          {arriveTangent=0.5501, interpMode=RCIM_Cubic, leaveTangent=0.5501, tangentMode=RCTM_User, time=0, value=0}
          {arriveTangent=1.1924, interpMode=RCIM_Cubic, leaveTangent=1.1924, tangentMode=RCTM_User, time=1, value=0.8}
         ]}}}
      Scale Curve=&RandomRangeFloat(Minimum=-0.15, Maximum=-0.2)
     )
    )
    Index 2 Param 3=&FloatFromCurve(
     FloatCurve=DataInterface() {propertyValues {Curve {keys [
         {arriveTangent=-0.4045, interpMode=RCIM_Cubic, tangentMode=RCTM_User, time=0.4, value=1}
         {arriveTangent=5.0157, interpMode=RCIM_Cubic, leaveTangent=5.0157, tangentMode=RCTM_User, time=1, value=2}
        ]}}}
     Scale Curve=2
    )
    Index 2 Param 4=&FloatFromCurve(
     FloatCurve=DataInterface() {propertyValues {Curve {keys [{arriveTangent=-0.4045, time=0, value=0.75}, {leaveTangent=-0.318, time=1, value=1}]}}}
     Scale Curve=0.4
    )
    Index 3 Param 1=&FloatFromCurve(
     FloatCurve=DataInterface() {propertyValues {Curve {keys [
         {arriveTangent=0.6183, interpMode=RCIM_Cubic, leaveTangent=0.6183, tangentMode=RCTM_User, time=0.1, value=0.02}
         {interpMode=RCIM_Cubic, tangentMode=RCTM_User, time=0.5, value=1}
        ]}}}
     Scale Curve=4
    )
    Index 3 Param 2=&FloatFromCurve(
     FloatCurve=DataInterface() {propertyValues {Curve {keys [<at 0 1>,<at 1 0>]}}}
     Scale Curve=&RandomRangeFloat(Minimum=0.4, Maximum=0.5)
    )
    Index 3 Param 3=&FloatFromCurve(
     FloatCurve=DataInterface() {propertyValues {Curve {keys [
         {arriveTangent=0.9433, interpMode=RCIM_Cubic, leaveTangent=0.9433, tangentMode=RCTM_User, time=0, value=0.5}
         {arriveTangent=0.3564, interpMode=RCIM_Cubic, leaveTangent=0.3564, tangentMode=RCTM_User, time=1, value=0.9}
        ]}}}
     Scale Curve=0.0004
    )
    Index 3 Param 4=1
    Module.Index0Param0Write=true
    Module.Index0Param1Write=true
    Module.Index0Param2Write=true
    Module.Index0Param3Write=false
    Module.Index1Param0Write=false
    Module.Index1Param1Write=false
    Module.Index1Param2Write=true
    Module.Index1Param3Write=true
    Module.Index2Param0Write=true
    Module.Index2Param1Write=true
    Module.Index2Param2Write=true
    Module.Index2Param3Write=true
    Module.Index3Param0Write=true
    Module.Index3Param1Write=true
    Module.Index3Param2Write=true
    Module.Index3Param3Write=true
    Write Parameter Index 1=true
    Write Parameter Index 2=true
    Write Parameter Index 3=true
   ) {module=DynamicMaterialParameters001}
  ]
  rendererClasses [&NiagaraSpriteRendererProperties]
  renderers [{rendererClass=&NiagaraSpriteRendererProperties}]
 }
 SegmentEmbers {
  simTarget=GPUComputeSim
  emitterUpdateScript [
   &EmitterState()
   &SpawnPerUnit(Module.Use Max Movement Threshold=true, Module.Use Movement Tolerance=false, Use Spawn Probability=false) {enabled=false}
   &SpawnRate(SpawnRate=0, Use Spawn Probability=false) {enabled=false, module=SpawnRate001}
   &SpawnRate(
    SpawnRate=&Multiply_Float(
     A=&Waveform("Waveform [1]"=::NewEnumerator0, "[1] Frequency"=1, "[1] Amplitude Min/Max"=<0 1>, "[1] Scale"=10, "[1] Bias"=0.5)
     B=&FloatFromCurve(
      FloatCurve=DataInterface() {propertyValues {Curve {keys [<at 0 1>,<at 1 0>]}}}
      CurveIndex=&RemapRange(
       Input Value To Remap=&VectorLength(Vector=$Engine.Owner.Velocity)
       Input Min=100
       Input Max=0
       Output Min=0.5
       Output Max=1
      )
      Scale Curve=15
     )
    )
    Use Spawn Probability=false
   )
  ]
  particleSpawnScript [
   &InitializeParticle(
    Color=<rgba 2 0.9024 0 1>
    Lifetime=&Multiply_Float(
     A=&Multiply_Float(
      A=&RemapRange(
       Input Value To Remap=&VectorLength(Vector=$Engine.Owner.Velocity)
       Input Min=100
       Input Max=0
       Output Min=0.2
       Output Max=1
      )
      B=&RandomRangeFloat(Minimum=0.34, Maximum=0.8)
     )
     B=&ClampFloat(Float=$User.Height, Min=0.2)
    )
    Position Offset=&Multiply_VectorByFloat(Vector=Linked() {linkedVariable {name=User.Offset, type=NiagaraPosition}}, Float=1)
    Position Offset Coordinate Space=::Local
    Sprite Size Mode=::NewEnumerator3
    Uniform Sprite Size=&RandomRangeFloat(Minimum=1.5, Maximum=2.5)
    UsePositionOffset=true
    Write Material Random=true
   )
   &ShapeLocation(
    Cylinder Height=7
    Cylinder Height Midpoint=0
    Cylinder Radius=7
    MaskSpawnGroup=false
    Offset=<0 0 -6>
    Offset Coordinate Space=::World
    Offset Mode=::NewEnumerator0
    Override Seed=false
    Rotation Mode=::NewEnumerator1
    Shape Primitive=::NewEnumerator1
    Surface Only=false
    UseOwnerScale=true
   ) {module=ShapeLocation001}
   &AddVelocity(
    Module.Use Falloff Distance=false
    Origin Offset=<0 0 -5>
    Origin Offset Coordinate Space=::World
    Override Seed=false
    Velocity Mode=::NewEnumerator1
    Velocity Speed=&Multiply_Float(
     A=&RemapRange(
      Input Value To Remap=&VectorLength(Vector=$Engine.Owner.Velocity)
      Input Min=100
      Input Max=0
      Output Min=0
      Output Max=1
     )
     B=&RandomRangeFloat(Minimum=25, Maximum=40)
    )
   )
   &AddVelocity(
    Rotation Coordinate Space=::World
    Velocity=<0 0 1>
    Velocity Speed Scale=&Multiply_Float(
     A=&RemapRange(
      Input Value To Remap=&VectorLength(Vector=$Engine.Owner.Velocity)
      Input Min=100
      Input Max=0
      Output Min=0
      Output Max=1
     )
     B=&RandomRangeFloat(Minimum=10, Maximum=25)
    )
   ) {module=AddVelocity001}
  ]
  particleUpdateScript [
   &ParticleState()
   &Drag(Drag=3, Use Linear Drag=true, Use Rotational Drag=true)
   /Niagara/Modules/Update/Forces/V2/CurlNoiseForce(
    Noise Frequency=80
    Noise Quality / Cost=::NewEnumerator0
    Noise Strength=&FloatFromCurve(
     FloatCurve=DataInterface() {propertyValues {Curve {keys [{tangentMode=RCTM_User, time=0, value=0}, {tangentMode=RCTM_User, time=1, value=1}]}}}
     Scale Curve=1000
    )
    Pan Noise=true
    Pan Noise Field=<0 2 1>
   )
   &AccelerationForce(Acceleration=&MakeVector(
     X=0
     Y=0
     Z=&Multiply_Float(
      A=&RemapRange(
       Input Value To Remap=&VectorLength(Vector=$Engine.Owner.Velocity)
       Input Min=100
       Input Max=0
       Output Min=0
       Output Max=1
      )
      B=&RandomRangeFloat(Minimum=200, Maximum=250)
     )
    ))
   &VortexForce(UseOriginOffset=false, Vortex Axis=<1 0 0>, Vortex Axis Coordinate Space:ENiagaraCoordinateSpace=::Local, Vortex Force Amount=50)
   &InheritSourceMovement(
    Applied Position Scale=&Multiply_Vector(
     A=<1 1 1>
     B=&VectorFromFloat(Value=&FloatFromCurve(
       FloatCurve=DataInterface() {propertyValues {Curve {keys [<at 0 1 RCIM_Cubic>, {arriveTangent=-1.4286, interpMode=RCIM_Cubic, leaveTangent=-1.4286, time=0.4, value=0.7}, <at 0.7 0 RCIM_Cubic>]}}}
       Scale Curve=0.5
      ))
    )
    ApplyForce=false
    ApplyPosition=true
    ApplyPositionFalloff=false
   )
   &SolveForcesAndVelocity(Clamp Velocity=false, Limit Acceleration=false)
   &ScaleSpriteSizeBySpeed(Max Scale Factor=<0.9 6>, Velocity Threshold=300)
   &ScaleColor(
    Scale RGB=&VectorFromFloat(Value=&FloatFromCurve(
      FloatCurve=DataInterface() {propertyValues {Curve {keys [
          {arriveTangent=9.0725, interpMode=RCIM_Cubic, leaveTangent=9.0725, tangentMode=RCTM_User, time=0, value=0}
          {arriveTangent=-2.5249, interpMode=RCIM_Cubic, leaveTangent=-2.5249, tangentMode=RCTM_User, time=0.1408, value=0.9847}
          {arriveTangent=-0.0405, interpMode=RCIM_Cubic, leaveTangent=-0.0405, tangentMode=RCTM_User, time=1, value=0}
         ]}}}
      Scale Curve=5
     ))
    ScaleA=true
    ScaleRGB=true
   )
  ]
  rendererClasses [&NiagaraSpriteRendererProperties]
  renderers [{rendererClass=&NiagaraSpriteRendererProperties}]
 }
 SegmentEmbers001 {
  bEnabled=false
  simTarget=GPUComputeSim
  emitterUpdateScript [
   &EmitterState()
   &SpawnPerUnit(
    Module.Use Max Movement Threshold=true
    Module.Use Movement Tolerance=true
    Movement Tolerance=10
    Spawn Spacing=5
    Use Spawn Probability=false
   )
   &SpawnRate(SpawnRate=0, Use Spawn Probability=false) {enabled=false, module=SpawnRate001}
   &SpawnRate(
    SpawnRate=&FloatFromCurve(
     FloatCurve=DataInterface() {propertyValues {Curve {keys [<at 0 1>,<at 1 0>]}}}
     CurveIndex=&RemapRange(
      Input Value To Remap=&VectorLength(Vector=$Engine.Owner.Velocity)
      Input Min=100
      Input Max=0
      Output Min=0.5
      Output Max=1
     )
     Scale Curve=60
    )
    Use Spawn Probability=false
   ) {enabled=false}
  ]
  particleSpawnScript [
   &InitializeParticle(
    Color=<rgba 2 2 2 0.5>
    Lifetime=&Multiply_Float(
     A=&Multiply_Float(
      A=&RemapRange(
       Input Value To Remap=&VectorLength(Vector=$Engine.Owner.Velocity)
       Input Min=100
       Input Max=0
       Output Min=0.5
       Output Max=1
      )
      B=&RandomRangeFloat(Minimum=0.34, Maximum=0.8)
     )
     B=&ClampFloat(Float=$User.Height, Min=0.2)
    )
    Position Offset=&Multiply_VectorByFloat(
     Vector=Linked() {linkedVariable {name=User.Offset, type=NiagaraPosition}}
     Float=&Multiply_Float(
      A=&RemapRange(
       Input Value To Remap=&VectorLength(Vector=$Engine.Owner.Velocity)
       Input Min=100
       Input Max=0
       Output Min=0.75
       Output Max=1
      )
      B=&RandomRangeFloat(Minimum=1, Maximum=1.05)
     )
    )
    Position Offset Coordinate Space=::Local
    Sprite Size Mode=::NewEnumerator3
    Uniform Sprite Size=&RandomRangeFloat(Minimum=5, Maximum=10)
    UsePositionOffset=true
    Write Material Random=true
   )
   &ShapeLocation(
    Cylinder Height=2
    Cylinder Height Midpoint=0
    Cylinder Radius=3
    MaskSpawnGroup=false
    Offset Coordinate Space=::World
    Offset Mode=::NewEnumerator0
    Override Seed=false
    Rotation Mode=::NewEnumerator1
    Shape Primitive=::NewEnumerator1
    Surface Only=false
    UseOwnerScale=true
   ) {module=ShapeLocation001}
   &AddVelocity(
    Module.Use Falloff Distance=false
    Origin Offset=<0 0 -5>
    Origin Offset Coordinate Space=::World
    Override Seed=false
    Velocity Mode=::NewEnumerator1
    Velocity Speed=&Multiply_Float(
     A=&RemapRange(
      Input Value To Remap=&VectorLength(Vector=$Engine.Owner.Velocity)
      Input Min=100
      Input Max=0
      Output Min=0
      Output Max=1
     )
     B=&RandomRangeFloat(Minimum=25, Maximum=40)
    )
   )
   &AddVelocity(
    Rotation Coordinate Space=::World
    Velocity=<0 0 1>
    Velocity Speed Scale=&Multiply_Float(
     A=&RemapRange(
      Input Value To Remap=&VectorLength(Vector=$Engine.Owner.Velocity)
      Input Min=100
      Input Max=0
      Output Min=0
      Output Max=1
     )
     B=&RandomRangeFloat(Minimum=10, Maximum=25)
    )
   ) {module=AddVelocity001}
  ]
  particleUpdateScript [
   &ParticleState()
   &Drag(Drag=3, Use Linear Drag=true, Use Rotational Drag=true)
   &CurlNoiseForce(
    Noise Frequency=80
    Noise Quality / Cost=::NewEnumerator0
    Noise Strength=&FloatFromCurve(
     FloatCurve=DataInterface() {propertyValues {Curve {keys [{tangentMode=RCTM_User, time=0, value=0}, {tangentMode=RCTM_User, time=1, value=1}]}}}
     Scale Curve=1000
    )
    Pan Noise=true
    Pan Noise Field=<0 2 1>
   )
   &AccelerationForce(Acceleration=&MakeVector(
     X=0
     Y=0
     Z=&Multiply_Float(
      A=&RemapRange(
       Input Value To Remap=&VectorLength(Vector=$Engine.Owner.Velocity)
       Input Min=100
       Input Max=0
       Output Min=0
       Output Max=1
      )
      B=&RandomRangeFloat(Minimum=200, Maximum=250)
     )
    ))
   &VortexForce(UseOriginOffset=false, Vortex Axis=<1 0 0>, Vortex Axis Coordinate Space:ENiagaraCoordinateSpace=::Local, Vortex Force Amount=50) {enabled=false}
   &InheritSourceMovement(
    Applied Position Scale=&Multiply_Vector(
     A=<0.99 0.99 1.05>
     B=&VectorFromFloat(Value=&FloatFromCurve(FloatCurve=DataInterface() {propertyValues {Curve {keys [<at 0.1 1.001 RCIM_Cubic>,<at 0.9 0 RCIM_Cubic>]}}}))
    )
    ApplyForce=false
    ApplyPosition=true
    ApplyPositionFalloff=false
   )
   &SolveForcesAndVelocity(Clamp Velocity=false, Limit Acceleration=false)
   &DynamicMaterialParameters(
    Index 0 Param 1=&Add_Float(
     A=&RandomRangeFloat(Minimum=0, Maximum=15)
     B=&FloatFromCurve(
      FloatCurve=DataInterface() {propertyValues {Curve {keys [
          {arriveTangent=-2.2561, interpMode=RCIM_Cubic, leaveTangent=-2.2561, tangentMode=RCTM_User, time=0, value=1}
          {arriveTangent=-1.0036, interpMode=RCIM_Cubic, leaveTangent=-1.0036, tangentMode=RCTM_User, time=0.2682, value=0.3}
          <at 0.8 0>
         ]}}}
      Scale Curve=0.1
     )
    )
    Index 0 Param 2=&FloatFromCurve(FloatCurve=DataInterface() {propertyValues {Curve {keys [
         {arriveTangent=0.086, interpMode=RCIM_Cubic, leaveTangent=0.086, tangentMode=RCTM_User, time=0.1, value=0.3}
         {arriveTangent=1.536, interpMode=RCIM_Cubic, leaveTangent=1.536, tangentMode=RCTM_User, time=1, value=1.25}
        ]}}})
    Index 0 Param 3=1.1
    Index 0 Param 4=1
    Index 1 Param 1=&MaskFloatBySpawnGroup(
     Masked Float=&FloatFromCurve(FloatCurve=DataInterface() {propertyValues {Curve {keys [
          {arriveTangent=8.5612, interpMode=RCIM_Cubic, leaveTangent=8.5612, tangentMode=RCTM_User, time=0, value=-0.5}
          {arriveTangent=0.9995, interpMode=RCIM_Cubic, leaveTangent=0.9995, tangentMode=RCTM_User, time=0.25, value=0.85}
          {arriveTangent=-0.0405, interpMode=RCIM_Cubic, leaveTangent=-0.0405, tangentMode=RCTM_User, time=0.5, value=1}
         ]}}})
     Passthrough Float=&FloatFromCurve(
      FloatCurve=DataInterface() {propertyValues {Curve {keys [
          {arriveTangent=4.1785, interpMode=RCIM_Cubic, leaveTangent=4.1785, tangentMode=RCTM_User, time=0, value=-0.5}
          {arriveTangent=0.9995, interpMode=RCIM_Cubic, leaveTangent=0.9995, tangentMode=RCTM_User, time=0.3, value=0.8}
          {arriveTangent=-0.0405, interpMode=RCIM_Cubic, leaveTangent=-0.0405, tangentMode=RCTM_User, time=0.8, value=1}
         ]}}}
      Scale Curve=2
     )
     Spawn Group Mask=1
    )
    Index 1 Param 2=&Add_Float(
     A=&FloatFromCurve(
      FloatCurve=DataInterface() {propertyValues {Curve {keys [
          {arriveTangent=-12.0154, interpMode=RCIM_Cubic, leaveTangent=-12.0154, tangentMode=RCTM_User, time=0, value=1}
          {interpMode=RCIM_Cubic, tangentMode=RCTM_User, time=0.2, value=0}
         ]}}}
      Scale Curve=-1
     )
     B=&RandomRangeFloat(Minimum=-100, Maximum=100)
    )
    Index 1 Param 3=&FloatFromCurve(
     FloatCurve=DataInterface() {propertyValues {Curve {keys [{interpMode=RCIM_Cubic, tangentMode=RCTM_User, time=0.333, value=1.2}, {interpMode=RCIM_Cubic, tangentMode=RCTM_User, time=1, value=2}]}}}
     Scale Curve=2
    )
    Index 1 Param 4=&FloatFromCurve(FloatCurve=DataInterface() {propertyValues {Curve {keys [<at 0 0.6>,<at 1 0.3>]}}}, Scale Curve=0.1)
    Index 2 Param 1=&FloatFromCurve(FloatCurve=DataInterface() {propertyValues {Curve {keys [
         {arriveTangent=0.0048, interpMode=RCIM_Cubic, leaveTangent=0.0048, tangentMode=RCTM_User, time=0, value=1.2}
         {arriveTangent=0.3333, interpMode=RCIM_Cubic, leaveTangent=0.3333, time=0.4, value=0.9}
         {interpMode=RCIM_Cubic, leaveTangent=1.688, time=0.9, value=1.5}
        ]}}})
    Index 2 Param 2=&Add_Float(
     A=&RandomRangeFloat(Minimum=0, Maximum=100)
     B=&FloatFromCurve(
      FloatCurve=DataInterface() {propertyValues {Curve {keys [
          {arriveTangent=0.5501, interpMode=RCIM_Cubic, leaveTangent=0.5501, tangentMode=RCTM_User, time=0, value=0}
          {arriveTangent=1.1924, interpMode=RCIM_Cubic, leaveTangent=1.1924, tangentMode=RCTM_User, time=1, value=0.8}
         ]}}}
      Scale Curve=&RandomRangeFloat(Minimum=-0.3, Maximum=-0.56)
     )
    )
    Index 2 Param 3=&FloatFromCurve(
     FloatCurve=DataInterface() {propertyValues {Curve {keys [
         {arriveTangent=-0.4045, interpMode=RCIM_Cubic, tangentMode=RCTM_User, time=0.4, value=1}
         {arriveTangent=5.0157, interpMode=RCIM_Cubic, leaveTangent=5.0157, tangentMode=RCTM_User, time=1, value=2}
        ]}}}
     Scale Curve=2
    )
    Index 2 Param 4=&FloatFromCurve(
     FloatCurve=DataInterface() {propertyValues {Curve {keys [{arriveTangent=-0.4045, time=0, value=0.75}, {leaveTangent=-0.318, time=1, value=1}]}}}
     Scale Curve=0.4
    )
    Index 3 Param 1=&FloatFromCurve(
     FloatCurve=DataInterface() {propertyValues {Curve {keys [
         {arriveTangent=0.6183, interpMode=RCIM_Cubic, leaveTangent=0.6183, tangentMode=RCTM_User, time=0.1, value=0.02}
         {interpMode=RCIM_Cubic, tangentMode=RCTM_User, time=0.5, value=1}
        ]}}}
     Scale Curve=2
    )
    Index 3 Param 2=&FloatFromCurve(
     FloatCurve=DataInterface() {propertyValues {Curve {keys [<at 0 1>,<at 1 0>]}}}
     Scale Curve=&RandomRangeFloat(Minimum=0.1, Maximum=0.15)
    )
    Index 3 Param 3=&FloatFromCurve(
     FloatCurve=DataInterface() {propertyValues {Curve {keys [
         {arriveTangent=0.4538, interpMode=RCIM_Cubic, leaveTangent=0.4538, tangentMode=RCTM_User, time=0, value=0.5}
         {arriveTangent=0.2675, interpMode=RCIM_Cubic, leaveTangent=0.2675, tangentMode=RCTM_User, time=1, value=0.7}
        ]}}}
     Scale Curve=0.0002
    )
    Index 3 Param 4=1
    Module.Index0Param0Write=true
    Module.Index0Param1Write=true
    Module.Index0Param2Write=true
    Module.Index0Param3Write=false
    Module.Index1Param0Write=false
    Module.Index1Param1Write=false
    Module.Index1Param2Write=true
    Module.Index1Param3Write=true
    Module.Index2Param0Write=true
    Module.Index2Param1Write=true
    Module.Index2Param2Write=true
    Module.Index2Param3Write=true
    Module.Index3Param0Write=true
    Module.Index3Param1Write=true
    Module.Index3Param2Write=true
    Module.Index3Param3Write=true
    Write Parameter Index 1=true
    Write Parameter Index 2=true
    Write Parameter Index 3=true
   ) {module=DynamicMaterialParameters001}
   &ScaleSpriteSizeBySpeed(Max Scale Factor=<0.9 6>, Velocity Threshold=300) {enabled=false}
   &ScaleColor(
    Scale RGB=&VectorFromFloat(Value=&FloatFromCurve(
      FloatCurve=DataInterface() {propertyValues {Curve {keys [
          {arriveTangent=9.0725, interpMode=RCIM_Cubic, leaveTangent=9.0725, tangentMode=RCTM_User, time=0, value=0}
          {arriveTangent=-2.5249, interpMode=RCIM_Cubic, leaveTangent=-2.5249, tangentMode=RCTM_User, time=0.1408, value=0.9847}
          {arriveTangent=-0.0405, interpMode=RCIM_Cubic, leaveTangent=-0.0405, tangentMode=RCTM_User, time=1, value=0}
         ]}}}
      Scale Curve=5
     ))
    ScaleA=true
    ScaleRGB=true
   )
  ]
  rendererClasses [&NiagaraSpriteRendererProperties]
  renderers [{rendererClass=&NiagaraSpriteRendererProperties}]
 }
}
summary {
 systemName=NS_AnnoyedMop_Equiped
 emitters [
  {emitterName=SegmentEmbers, simTarget=GPUComputeSim, rendererClasses [&NiagaraSpriteRendererProperties]}
  {emitterName=Fire_Main, rendererClasses [&NiagaraSpriteRendererProperties, &NiagaraLightRendererProperties]}
  {emitterName=Fire_Root, rendererClasses [&NiagaraSpriteRendererProperties]}
  {emitterName=Fire_Mesh_Root001, rendererClasses [&NiagaraMeshRendererProperties]}
  {bEnabled=false, emitterName=Fire_Main_Inherit, rendererClasses [&NiagaraSpriteRendererProperties]}
  {bEnabled=false, emitterName=SegmentEmbers001, simTarget=GPUComputeSim, rendererClasses [&NiagaraSpriteRendererProperties]}
 ]
 userVariables [
  User.ActiveArray {
   type=NiagaraDataInterfaceArrayBool
   description=""
   defaultValue {
    struct=NiagaraExt_VariableValue_DataInterface
    value {
     dataInterface=&self:NiagaraDataInterfaceArrayBool_0
     dataInterfaceClass=/Script/Niagara.NiagaraDataInterfaceArrayBool
    }
   }
  }
  User.GlobalScale {type=NiagaraFloat, defaultValue=3, description=""}
  User.Height {type=NiagaraFloat, defaultValue=1, description=""}
  User.NumSegments {type=NiagaraInt32, defaultValue=5}
  User.Offset {type=NiagaraPosition, description="", defaultValue {}}
  User.Orientation {type=Vector3f, description="", defaultValue {struct=Vector3f, value=<0 0 1>}}
  User.RootScale {type=Vector3f, description="", defaultValue {struct=Vector3f, value=<1 1 1>}}
  User.Rotation {type=Vector3f, description="", defaultValue {struct=Vector3f, value=<0 0 0>}}
  User.SegmentPositions {
   type=NiagaraDataInterfaceArrayFloat3
   description=""
   defaultValue {
    struct=NiagaraExt_VariableValue_DataInterface
    value {
     dataInterface=&self:NiagaraDataInterfaceArrayFloat3_0
     dataInterfaceClass=&NiagaraDataInterfaceArrayFloat3
    }
   }
  }
  User.SegmentRotations {
   type=NiagaraDataInterfaceArrayQuat
   description=""
   defaultValue {
    struct=NiagaraExt_VariableValue_DataInterface
    value {
     dataInterface=&self:NiagaraDataInterfaceArrayQuat_0
     dataInterfaceClass=/Script/Niagara.NiagaraDataInterfaceArrayQuat
    }
   }
  }
 ]
}
